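# Niagara cosmetic effect for the TweakSmirk character body, a brief pink-magenta reactivity burst.
# A growing glow feeds three sprite renderers, with curve-driven ShadowColor and CoronaColor particle attributes and a camera offset that differs between gameplay and front end.
# A shorter, wide, thin streak flashes and fades in front of it.
# A few small random pink-to-violet sprites spawn on a ring, drift under curl noise and fade out.
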NiagaraSystem /BRCosmetics/Characters/Player/Female/Medium/Bodies/F_MED_TweakSmirk/FX/NS_Character_TweakSmirk_Reactivity
dependencies {
 usedDataInterfaces [&NiagaraDataInterfaceCurve, &NiagaraDataInterfaceColorCurve, &NiagaraDataInterfaceVector2DCurve]
 usedDynamicInputs []
 usedModules [&SystemState]
}
emitters {
 Minimal {
  emitterUpdateScript [
   &EmitterState()
   &SpawnBurst_Instantaneous(Use Spawn Probability=false, UseLoopCountLimit=false)
  ]
  particleSpawnScript [
   &InitializeParticle(
    Color=<rgba 2 0.4292 1.7704 0.6>
    Lifetime=0.7
    Position Offset=<0 4 -5>
    Sprite Size Mode=::NewEnumerator3
    Uniform Sprite Size=85
    UsePositionOffset=true
    Write Material Random=true
   )
   &CameraOffset(Camera Offset Amount=&FN_FrontEndFloatModifier(Gameplay=4, Front End=1))
  ]
  particleUpdateScript [
   &ParticleState()
   &ScaleSpriteSize(Uniform Curve Sprite Scale=DataInterface() {propertyValues {Curve {
       defaultValue=0
       keys [
        {arriveTangent=4.1785, interpMode=RCIM_Cubic, leaveTangent=4.1785, tangentMode=RCTM_User, time=0, value=0}
        {arriveTangent=1.1179, interpMode=RCIM_Cubic, leaveTangent=1.1179, tangentMode=RCTM_User, time=0.2042, value=0.6228}
        {arriveTangent=0.0224, interpMode=RCIM_Cubic, leaveTangent=0.0224, tangentMode=RCTM_User, time=1, value=1}
       ]
      }}})
   &ScaleColor(
    Scale Alpha=&FloatFromCurve(FloatCurve=DataInterface() {propertyValues {Curve {
        defaultValue=1
        keys [
         {arriveTangent=9.0725, interpMode=RCIM_Cubic, leaveTangent=9.0725, tangentMode=RCTM_User, time=0, value=0}
         {arriveTangent=-2.5249, interpMode=RCIM_Cubic, leaveTangent=-2.5249, tangentMode=RCTM_User, time=0.1408, value=0.9847}
         {arriveTangent=-0.0405, interpMode=RCIM_Cubic, leaveTangent=-0.0405, tangentMode=RCTM_User, time=1, value=0}
        ]
       }}})
    ScaleA=true
    ScaleRGB=true
   )
   /SetVariables(Particles.ShadowColor:LinearColor=&ColorFromCurve(ColorCurve=DataInterface() {propertyValues {
       AlphaCurve {keys [
         <at 0.0063 0>
         {arriveTangent=0.7154, interpMode=RCIM_Cubic, leaveTangent=0.7154, time=0.1342, value=0.9}
         <at 0.5654 0.4>
         <at 0.9416 0 RCIM_Cubic>
        ]}
       BlueCurve {keys [<at 0.0828 0.6949 RCIM_Cubic>,<at 0.5075 0.094 RCIM_Cubic>]}
       GreenCurve {keys [<at 0.0828 0.2785 RCIM_Cubic>,<at 0.5075 0.0246 RCIM_Cubic>]}
       RedCurve {keys [<at 0.0828 1 RCIM_Cubic>,<at 0.5075 0.1337 RCIM_Cubic>]}
      }})) {isSetParameters=true}
   /SetVariables(Particles.CoronaColor:LinearColor=&ColorFromCurve(ColorCurve=DataInterface() {propertyValues {
       AlphaCurve {keys [<at 0.0156 0>,<at 0.1379 0.6 RCIM_Cubic>,<at 0.557 0 RCIM_Cubic>]}
       BlueCurve {keys [<at 0.08 1 RCIM_Cubic>,<at 0.585 0.8976 RCIM_Cubic>]}
       GreenCurve {keys [<at 0.08 0.4468 RCIM_Cubic>,<at 0.585 0.5969 RCIM_Cubic>]}
       RedCurve {keys [<at 0.08 0.7374 RCIM_Cubic>,<at 0.585 0.8138 RCIM_Cubic>]}
      }})) {isSetParameters=true}
  ]
  rendererClasses [&NiagaraSpriteRendererProperties]
  renderers [
   {rendererClass=&NiagaraSpriteRendererProperties}
   {rendererClass=&NiagaraSpriteRendererProperties, rendererIndex=1}
   {rendererClass=&NiagaraSpriteRendererProperties, rendererIndex=2}
  ]
 }
 Minimal001 {
  emitterUpdateScript [
   &EmitterState()
   &SpawnBurst_Instantaneous(Use Spawn Probability=false, UseLoopCountLimit=false)
  ]
  particleSpawnScript [
   &InitializeParticle(
    Color=<rgba 4 0.8527 3.4859 2>
    Lifetime=0.3
    Position Offset=<0 4 -5>
    Sprite Size=<155 15>
    Sprite Size Mode=::NewEnumerator4
    UsePositionOffset=true
    Write Material Random=true
   )
   &CameraOffset(Camera Offset Amount=8)
  ]
  particleUpdateScript [
   &ParticleState()
   &ScaleSpriteSize(
    Non-Uniform Curve Sprite Scale=DataInterface() {propertyValues {
      XCurve {
       defaultValue=0
       keys [
        {arriveTangent=0.4684, interpMode=RCIM_Cubic, leaveTangent=0.4684, tangentMode=RCTM_User, time=0, value=0}
        {arriveTangent=0.0224, interpMode=RCIM_Cubic, leaveTangent=0.0224, tangentMode=RCTM_User, time=0.9, value=1}
       ]
      }
      YCurve {keys [
        {tangentMode=RCTM_User, time=0.0016, value=0.8426}
        {interpMode=RCIM_Cubic, leaveTangent=-1.8307, tangentMode=RCTM_Break, time=0.367, value=0.9991}
        {arriveTangent=-0.948, tangentMode=RCTM_User, time=1.0024, value=0.0066}
       ]}
     }}
    Scale Sprite Size Mode=::NewEnumerator3
   )
   &ScaleColor(
    Scale Alpha=&FloatFromCurve(FloatCurve=DataInterface() {propertyValues {Curve {
        defaultValue=1
        keys [
         {arriveTangent=-1.2576, interpMode=RCIM_Cubic, leaveTangent=-1.2576, tangentMode=RCTM_User, time=0, value=0}
         {arriveTangent=0.262, interpMode=RCIM_Cubic, leaveTangent=0.262, tangentMode=RCTM_User, time=0.1408, value=0.9847}
         {arriveTangent=-1.1861, interpMode=RCIM_Cubic, leaveTangent=-1.1861, tangentMode=RCTM_User, time=1, value=0}
        ]
       }}})
    ScaleA=true
    ScaleRGB=true
   )
  ]
  rendererClasses [&NiagaraSpriteRendererProperties]
  renderers [{rendererClass=&NiagaraSpriteRendererProperties}]
 }
 Minimal002 {
  emitterUpdateScript [
   &EmitterState()
   &SpawnBurst_Instantaneous(
    Spawn Count=&RandomRangeInt(Minimum=3, Maximum=5)
    Use Spawn Probability=false
    UseLoopCountLimit=false
   )
  ]
  particleSpawnScript [
   &InitializeParticle(
    Color Maximum=<rgba 1 0.1494 0.7956 0.1>
    Color Minimum=<rgba 0.6979 0.1623 1 0.1>
    Color Mode=::NewEnumerator2
    Lifetime Max=1.1
    Lifetime Min=0.6
    Lifetime Mode=::NewEnumerator1
    Sprite Size Mode=::NewEnumerator1
    Uniform Sprite Size Max=18
    Uniform Sprite Size Min=8
    UsePositionOffset=false
    Write Material Random=true
   )
   &ShapeLocation(
    Disc Coverage=0.3135
    MaskSpawnGroup=false
    Offset=<0 0 -4>
    Offset Mode=::NewEnumerator0
    Override Seed=false
    Ring Radius=12
    Rotation Mode=::NewEnumerator2
    Shape Primitive=::NewEnumerator4
    UseOwnerScale=true
    Yaw / Pitch / Roll=<0 0 0.25>
   )
   &CameraOffset(Camera Offset Amount=12)
  ]
  particleUpdateScript [
   &ParticleState()
   &CurlNoiseForce(Bias Noise=false, Noise Strength=12, Pan Noise=false)
   &SolveForcesAndVelocity(Clamp Velocity=false, Limit Acceleration=false)
   &ScaleColor(
    Scale Alpha=&FloatFromCurve(FloatCurve=DataInterface() {propertyValues {Curve {keys [<at 0.6 1>,<at 1 0>]}}})
    ScaleA=true
    ScaleRGB=true
   )
  ]
  rendererClasses [&NiagaraSpriteRendererProperties]
  renderers [{rendererClass=&NiagaraSpriteRendererProperties}]
 }
}
summary {
 systemName=NS_Character_TweakSmirk_Reactivity
 emitters [
  {emitterName=Minimal, rendererClasses [&NiagaraSpriteRendererProperties]}
  {emitterName=Minimal001, rendererClasses [&NiagaraSpriteRendererProperties]}
  {emitterName=Minimal002, rendererClasses [&NiagaraSpriteRendererProperties]}
 ]
 userVariables [
  User.bIsFrontEnd {type=NiagaraBool}
  User.bIsFrontEndPreview {type=NiagaraBool}
 ]
}
