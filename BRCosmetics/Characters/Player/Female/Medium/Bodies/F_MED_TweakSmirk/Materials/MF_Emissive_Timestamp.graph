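# Material function for the F_MED_TweakSmirk character body that makes an input emissive flare in response to two timestamp parameters.
# Each timestamp is compared to Time to build a ramp-in, hold and fade-out envelope, and the stronger envelope drives the Timestamp mask.
# The DebugTimestamp static switch substitutes a slow saturated sine.
# The masked emissive feeds Color_Time_of_Day, with separate intensity multipliers for evening and night.
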MaterialFunction /BRCosmetics/Characters/Player/Female/Medium/Bodies/F_MED_TweakSmirk/Materials/MF_Emissive_Timestamp
graph {nodes [
  n1 = Multiply(A=FunctionInput(), B=NamedRerouteUsage())
  n2 = ScalarParameter() {defaultValue=2, parameterName=Emissive_ToD_Evening_Intensity}
  n3 = ScalarParameter() {defaultValue=2, parameterName=Emissive_ToD_Night_Intensity}
  FunctionOutput(None=MaterialFunctionCall(Color_Morning=@n1, Color_Day=@n1, Color_Evening=Multiply(A=@n1, B=@n2) {constB=2}, Color_Night=Multiply(A=@n1, B=@n3)) {materialFunction=&Color_Time_of_Day}.Result)
  n4 = Sine(None=Multiply(A=Time()) {constB=0.2})
  n5 = Time()
  n6 = ScalarParameter() {defaultValue=999, group=EmissiveTimestamp, parameterName=Timestamp_0}
  n7 = ScalarParameter() {defaultValue=0.75, group=EmissiveTimestamp, parameterName=TimestampIn}
  n8 = Divide(A=Subtract(A=@n5, B=@n6), B=@n7) {desc=InLength}
  n9 = Add(
   A=@n6
   B=Add(A=@n7, B=ScalarParameter() {defaultValue=0.2, group=EmissiveTimestamp, parameterName=TimestampStay}) {constB=3, desc=eXTRAdURATION}
  )
  n10 = Saturate(None=Divide(A=Subtract(A=@n5, B=@n9), B=ScalarParameter() {defaultValue=3, group=EmissiveTimestamp, parameterName=TimestampOut}) {constB=3, desc=OutLength})
  n11 = Multiply(
   A=NamedRerouteDeclaration(None=Saturate(None=@n8)) {name=TimestampIn, nodeColor=<rgba 0 0.1328 1 1>}
   B=NamedRerouteDeclaration(None=OneMinus(None=@n10)) {name=TimestampOut, nodeColor=<rgba 0 0.0859 1 1>}
  )
  n12 = Time()
  n13 = ScalarParameter() {defaultValue=999, group=EmissiveTimestamp, parameterName=Timestamp_1}
  n14 = ScalarParameter() {defaultValue=0.75, group=EmissiveTimestamp, parameterName=TimestampIn}
  n15 = Saturate(None=Divide(A=Subtract(A=@n12, B=@n13), B=@n14) {desc=InLength})
  n16 = Add(
   A=@n13
   B=Add(A=@n14, B=ScalarParameter() {defaultValue=0.2, group=EmissiveTimestamp, parameterName=TimestampStay}) {constB=3, desc=eXTRAdURATION}
  )
  n17 = Saturate(None=Divide(A=Subtract(A=@n12, B=@n16), B=ScalarParameter() {defaultValue=3, group=EmissiveTimestamp, parameterName=TimestampOut}) {constB=3, desc=OutLength})
  n18 = Multiply(A=@n15, B=NamedRerouteDeclaration(None=OneMinus(None=@n17)) {name=TimestampOut_1, nodeColor=<rgba 0 0.0859 1 1>})
  NamedRerouteDeclaration(None=StaticSwitchParameter(True=Saturate(None=@n4), False=Max(A=@n11, B=@n18)) {group=EmissiveTimestamp, parameterName=DebugTimestamp}) {name=Timestamp, nodeColor=<rgba 1 0 0.9141 1>}
  NamedRerouteDeclaration(None=@n15) {name=TimestampIn_1, nodeColor=<rgba 0 0.1328 1 1>}
 ]}
