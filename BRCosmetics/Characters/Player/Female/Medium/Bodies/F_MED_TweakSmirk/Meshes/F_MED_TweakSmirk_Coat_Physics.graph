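# Physics asset for the coat mesh of the female medium TweakSmirk player cosmetic, covering dynamic coat and spine-kerchief bones.
# Capsule bodies form five coat chains (back, sides, front left and right) and a kerchief chain, rooted at spine_02, with master bones leading the front and kerchief chains.
# Upper attachments swing freely, while lower links have angular limits with little or no twist.
# Pelvis, spine, thighs, right forearm and a large offset root capsule supply body shapes.
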PhysicsAsset /BRCosmetics/Characters/Player/Female/Medium/Bodies/F_MED_TweakSmirk/Meshes/F_MED_TweakSmirk_Coat_Physics
bodies {
 dyn_coat_01_fL [{
   shapeType=Capsule
   length=3.6277
   radius=1.1388
   shapeName=dyn_coat_01_fL_capsule
   center=<5.3419 -0.5611 -0.5446>
   rotation=<rot 0 -167.0827 -77.7878>
  }]
 dyn_coat_01_fR [{
   shapeType=Capsule
   length=3.6277
   radius=1.1376
   shapeName=dyn_coat_01_fR_capsule
   center=<5.3388 0.5779 -0.5615>
   rotation=<rot 0 -12.7465 -78.2896>
  }]
 dyn_coat_02_L [{
   shapeType=Capsule
   length=0
   radius=3.6781
   shapeName=dyn_coat_02_L_capsule
   center=<5.5749 -0.4701 -1.0249>
   rotation=<rot 0 96.5323 -89.5385>
  }]
 dyn_coat_02_R [{
   shapeType=Capsule
   length=0
   radius=3.7614
   shapeName=dyn_coat_02_R_capsule
   center=<5.7155 -0.5971 -1.0388>
   rotation=<rot 0 -99.9157 -88.6945>
  }]
 dyn_coat_02_fL [{
   shapeType=Capsule
   length=9.4614
   radius=5.8636
   shapeName=dyn_coat_03_fL_capsule
   center=<9.9529 -3.4285 2.9498>
   rotation=<rot -2.8841 -174.5043 -66.623>
  }]
 dyn_coat_02_fR [{
   shapeType=Capsule
   length=9.4614
   radius=5.8636
   shapeName=dyn_coat_03_fL_capsule
   center=<9.7825 0.0051 3.0104>
   rotation=<rot 11.1757 166.1635 -100.3872>
  }]
 dyn_coat_03_L [{
   shapeType=Capsule
   length=11.3623
   radius=7.5494
   shapeName=dyn_coat_03_L_capsule
   center=<8.7581 -1.1015 2.7244>
   rotation=<rot 0.0225 0.8685 -89.9417>
  }]
 dyn_coat_03_R [{
   shapeType=Capsule
   length=11.3623
   radius=7.5494
   shapeName=dyn_coat_03_L_capsule
   center=<8.7581 -1.1015 2.7244>
   rotation=<rot 0.019 4.2793 -89.9404>
  }]
 dyn_coat_03_fL [{
   shapeType=Capsule
   length=8.9614
   radius=6.3636
   shapeName=dyn_coat_03_fL_capsule
   center=<8.6325 -4.2854 0.7435>
   rotation=<rot 1.4027 176.5453 -46.098>
  }]
 dyn_coat_03_fR [{
   shapeType=Capsule
   length=9.4614
   radius=5.8636
   shapeName=dyn_coat_03_fL_capsule
   center=<10.8677 3.46 2.3535>
   rotation=<rot 11.1757 166.1636 -118.7618>
  }]
 dyn_coat_04_L [{
   shapeType=Capsule
   length=10.9777
   radius=8.47
   shapeName=dyn_coat_04_L_capsule
   center=<5.1269 0.4844 -0.2564>
   rotation=<rot 0 -5.8618 -86.9103>
  }]
 dyn_coat_04_R [{
   shapeType=Capsule
   length=10.9777
   radius=8.47
   shapeName=dyn_coat_04_L_capsule
   center=<5.1269 0.4844 -0.2564>
   rotation=<rot -0.2284 -1.6263 -86.9187>
  }]
 dyn_coat_04_fL [{
   shapeType=Capsule
   length=9.5354
   radius=7.4116
   shapeName=dyn_coat_04_fL_capsule
   center=<6.2893 -3.6514 -1.4858>
   rotation=<rot 3.1368 -171.0704 -46.3092>
  }]
 dyn_coat_04_fR [{
   shapeType=Capsule
   length=9.5354
   radius=7.4116
   shapeName=dyn_coat_04_fL_capsule
   center=<6.691 4.0202 -0.6374>
   rotation=<rot 12.8429 178.4453 -119.295>
  }]
 dyn_coat_MSTR_fL [{shapeType=Capsule, length=0.505, radius=0.505, shapeName=dyn_coat_MSTR_fL_capsule, rotation=<rot 0 -90 0>}]
 dyn_coat_MSTR_fR [{shapeType=Capsule, length=0.505, radius=0.505, shapeName=dyn_coat_MSTR_fR_capsule, rotation=<rot 0 -90 0>}]
 dyn_coat_bk_02 [{
   shapeType=Capsule
   length=0
   radius=4.2466
   shapeName=dyn_coat_bk_02_capsule
   center=<4.9381 0.1707 -0.5708>
   rotation=<rot 0 86.6089 -89.1574>
  }]
 dyn_coat_bk_03 [{
   shapeType=Capsule
   length=11.5744
   radius=8.575
   shapeName=dyn_coat_bk_03_capsule
   center=<10.4681 0.1718 2.4267>
   rotation=<rot -3.1523 0.3028 -90.0015>
  }]
 dyn_coat_bk_04 [{
   shapeType=Capsule
   length=10.5081
   radius=9.2127
   shapeName=dyn_coat_bk_04_capsule
   center=<6.3008 0.2185 -0.7171>
   rotation=<rot 0 0.3514 -88.6356>
  }]
 dyn_spinekerchief_01_fL [{
   shapeType=Capsule
   length=6.0026
   radius=5.2507
   shapeName=dyn_spinekerchief_01_fL_capsule
   center=<4.2459 0.3206 3.1611>
   rotation=<rot 0 85.6117 -84.4147>
  }]
 dyn_spinekerchief_02_fL [{
   shapeType=Capsule
   length=4.0522
   radius=3.3128
   shapeName=dyn_spinekerchief_02_fL_capsule
   center=<4.9038 0.3626 1.9006>
   rotation=<rot -0.0907 -89.8608 -89.3328>
  }]
 dyn_spinekerchief_MSTR_fL [{shapeType=Capsule, length=0.505, radius=0.505, shapeName=dyn_spinekerchief_MSTR_fL_capsule, rotation=<rot 0 -90 0>}]
 lowerarm_r [{shapeType=Capsule, length=31.0989, radius=6.0998, center=<-14.7532 -1.2494 -0.092>, rotation=<rot 0 88.7368 -91.2727>}]
 pelvis [{shapeType=Capsule, length=13.8253, radius=9.3183, shapeName=pelvis_capsule, center=<-3.6587 -0.5058 0>}]
 root [{
   shapeType=Capsule
   length=33.8253
   radius=71.8183
   shapeName=pelvis_capsule
   center=<-2.9811 -26.0965 -66.2095>
   rotation=<rot 89.4136 -179.9941 -179.9941>
  }]
 spine_02 [{shapeType=Capsule, length=4.6242, radius=10.452, shapeName=spine_04_capsule, center=<-5.0689 -2.626 -0.4062>}]
 thigh_l [{
   shapeType=Capsule
   length=26.2206
   radius=7.5299
   shapeName=thigh_l_capsule
   center=<-19.0365 -0.0381 0.2152>
   rotation=<rot 0 -86.5889 -87.0058>
  }]
 thigh_r [{
   shapeType=Capsule
   length=25.7495
   radius=7.4963
   shapeName=thigh_r_capsule
   center=<17.941 0.2516 0.3669>
   rotation=<rot 0 -91.6433 -89.8488>
  }]
}
constraints [
 {bone1Name=lowerarm_r, bone2Name=spine_02, swing1LimitDegrees=45, swing2LimitDegrees=45, twistLimitDegrees=45}
 {bone1Name=dyn_coat_bk_04, bone2Name=dyn_coat_bk_03, swing1LimitDegrees=20, swing2LimitDegrees=45, twistLimitDegrees=1}
 {bone1Name=dyn_coat_bk_03, bone2Name=dyn_coat_bk_02, swing1LimitDegrees=5, swing2LimitDegrees=20, twistLimitDegrees=1}
 {bone1Name=dyn_coat_04_fR, bone2Name=dyn_coat_03_fR, swing1LimitDegrees=5, swing2LimitDegrees=45, twistLimitDegrees=1}
 {bone1Name=dyn_coat_03_fR, bone2Name=dyn_coat_02_fR, swing1LimitDegrees=5, swing2LimitDegrees=20, twistLimitDegrees=1}
 {bone1Name=dyn_coat_04_fL, bone2Name=dyn_coat_03_fL, swing1LimitDegrees=5, swing2LimitDegrees=45, twistLimitDegrees=1}
 {bone1Name=dyn_coat_03_fL, bone2Name=dyn_coat_02_fL, swing1LimitDegrees=5, swing2LimitDegrees=20, twistLimitDegrees=1}
 {bone1Name=dyn_coat_04_R, bone2Name=dyn_coat_03_R, swing1LimitDegrees=20, swing2LimitDegrees=45, twistLimitDegrees=1}
 {bone1Name=dyn_coat_03_R, bone2Name=dyn_coat_02_R, swing1LimitDegrees=5, swing2LimitDegrees=20, twistLimitDegrees=1}
 {bone1Name=dyn_coat_04_L, bone2Name=dyn_coat_03_L, swing1LimitDegrees=20, swing2LimitDegrees=45, twistLimitDegrees=1}
 {bone1Name=dyn_coat_03_L, bone2Name=dyn_coat_02_L, swing1LimitDegrees=5, swing2LimitDegrees=20, twistLimitDegrees=1}
 {
  bone1Name=thigh_l
  bone2Name=pelvis
  swing1LimitDegrees=45
  swing1Motion=Free
  swing2LimitDegrees=45
  swing2Motion=Free
  twistLimitDegrees=45
  twistMotion=Free
 }
 {
  bone1Name=thigh_r
  bone2Name=pelvis
  swing1LimitDegrees=45
  swing1Motion=Free
  swing2LimitDegrees=45
  swing2Motion=Free
  twistLimitDegrees=45
  twistMotion=Free
 }
 {
  bone1Name=spine_02
  bone2Name=pelvis
  swing1LimitDegrees=45
  swing1Motion=Free
  swing2LimitDegrees=45
  swing2Motion=Free
  twistLimitDegrees=45
  twistMotion=Free
 }
 {
  bone1Name=dyn_spinekerchief_MSTR_fL
  bone2Name=spine_02
  swing1LimitDegrees=45
  swing1Motion=Free
  swing2LimitDegrees=45
  swing2Motion=Free
  twistLimitDegrees=45
  twistMotion=Free
 }
 {bone1Name=dyn_spinekerchief_01_fL, bone2Name=dyn_spinekerchief_MSTR_fL, swing1LimitDegrees=25, swing2LimitDegrees=55, twistLimitDegrees=5}
 {bone1Name=dyn_spinekerchief_02_fL, bone2Name=dyn_spinekerchief_01_fL, swing1LimitDegrees=25, swing2LimitDegrees=55, twistLimitDegrees=5}
 {
  bone1Name=dyn_coat_bk_02
  bone2Name=spine_02
  swing1LimitDegrees=45
  swing1Motion=Free
  swing2LimitDegrees=45
  swing2Motion=Free
  twistLimitDegrees=45
  twistMotion=Free
 }
 {bone1Name=dyn_coat_02_fR, bone2Name=dyn_coat_01_fR, swing1LimitDegrees=1, swing2LimitDegrees=10, twistLimitDegrees=0}
 {bone1Name=dyn_coat_02_fL, bone2Name=dyn_coat_01_fL, swing1LimitDegrees=1, swing2LimitDegrees=10, twistLimitDegrees=0}
 {
  bone1Name=dyn_coat_02_R
  bone2Name=spine_02
  swing1LimitDegrees=45
  swing1Motion=Free
  swing2LimitDegrees=45
  swing2Motion=Free
  twistLimitDegrees=45
  twistMotion=Free
 }
 {
  bone1Name=dyn_coat_02_L
  bone2Name=spine_02
  swing1LimitDegrees=45
  swing1Motion=Free
  swing2LimitDegrees=45
  swing2Motion=Free
  twistLimitDegrees=45
  twistMotion=Free
 }
 {
  bone1Name=dyn_coat_01_fL
  bone2Name=dyn_coat_MSTR_fL
  swing1LimitDegrees=45
  swing1Motion=Free
  swing2LimitDegrees=45
  swing2Motion=Free
  twistLimitDegrees=45
  twistMotion=Free
 }
 {
  bone1Name=dyn_coat_01_fR
  bone2Name=dyn_coat_MSTR_fR
  swing1LimitDegrees=45
  swing1Motion=Free
  swing2LimitDegrees=45
  swing2Motion=Free
  twistLimitDegrees=45
  twistMotion=Free
 }
 {
  bone1Name=dyn_coat_MSTR_fR
  bone2Name=spine_02
  swing1LimitDegrees=45
  swing1Motion=Free
  swing2LimitDegrees=45
  swing2Motion=Free
  twistLimitDegrees=45
  twistMotion=Free
 }
 {
  bone1Name=dyn_coat_MSTR_fL
  bone2Name=spine_02
  swing1LimitDegrees=45
  swing1Motion=Free
  swing2LimitDegrees=45
  swing2Motion=Free
  twistLimitDegrees=45
  twistMotion=Free
 }
 {
  bone1Name=pelvis
  bone2Name=root
  swing1LimitDegrees=45
  swing1Motion=Free
  swing2LimitDegrees=45
  swing2Motion=Free
  twistLimitDegrees=45
  twistMotion=Free
 }
]
