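# Control Rig for the F_MED_TweakSmirk cosmetic body that procedurally corrects coat, collar and sleeve bones.
# After a shared DeformRig function, spherical pose readers on upper arms and aim-constrained thigh controls store alpha variables.
# Those alphas weight parent constraints and transform offsets on sleeves, collar and coat, while local hand readers shift cuff bones.
# Sliding, baseball-slide, DBNO and carry flags select additional fixed offsets.
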ControlRigBlueprint /BRCosmetics/Characters/Player/Female/Medium/Bodies/F_MED_TweakSmirk/Meshes/F_MED_TweakSmirk_CtrlRig
graphs [
 New Function_ContainedGraph {nodes [
   n1 = RigVMFunctionEntryNode()
   SequenceAggregate = &RigVMFunction_Sequence(ExecuteContext=@n1.ExecuteContext)
   n2 = &RigUnit_SphericalPoseReader(
    ActiveRegionScaleFactors=(PositiveWidth=2,NegativeWidth=2,PositiveHeight=2,NegativeHeight=2)
    ActiveRegionSize=0.1
    Debug=(bDrawDebug=false,DebugScale=25,DebugSegments=20,DebugThickness=0.25)
    DriverAxis=(X=1,Y=0,Z=0)
    DriverItem=(Type=Bone,Name="upperarm_l")
    FalloffRegionScaleFactors=(PositiveWidth=0.5,NegativeWidth=0.5,PositiveHeight=0.5,NegativeHeight=0.5)
    FalloffSize=0.5
    OptionalParentItem=(Type=Bone,Name="spine_05")
    RotationOffset=(X=0,Y=10,Z=0)
    ExecutePin=@SequenceAggregate.A
   )
   > RigVMVariableNode(Value=@n2.OutputParam) {variable=spr_upperarm_l_up_alpha}
   > n3 = &RigUnit_SphericalPoseReader(
    ActiveRegionScaleFactors=(PositiveWidth=2,NegativeWidth=2,PositiveHeight=2,NegativeHeight=2)
    ActiveRegionSize=0.1
    Debug=(bDrawDebug=false,DebugScale=25,DebugSegments=20,DebugThickness=0.25)
    DriverAxis=(X=-1,Y=0,Z=0)
    DriverItem=(Type=Bone,Name="upperarm_r")
    FalloffRegionScaleFactors=(PositiveWidth=0.5,NegativeWidth=0.5,PositiveHeight=0.5,NegativeHeight=0.5)
    FalloffSize=0.5
    FlipHeightScaling=true
    FlipWidthScaling=false
    OptionalParentItem=(Type=Bone,Name="spine_05")
    RotationOffset=(X=180,Y=-10,Z=180)
   )
   > RigVMVariableNode(Value=@n3.OutputParam) {variable=spr_upperarm_r_up_alpha}
   > n4 = &RigUnit_SphericalPoseReader(
    ActiveRegionScaleFactors=(PositiveWidth=0.1,NegativeWidth=1,PositiveHeight=0.1,NegativeHeight=1)
    ActiveRegionSize=0.1
    Debug=(bDrawDebug=false,DebugScale=25,DebugSegments=20,DebugThickness=0.25)
    DriverAxis=(X=1,Y=0,Z=0)
    DriverItem=(Type=Bone,Name="upperarm_l")
    FalloffRegionScaleFactors=(PositiveWidth=0.4,NegativeWidth=1,PositiveHeight=0.8,NegativeHeight=0.5)
    FalloffSize=0.5
    OptionalParentItem=(Type=Bone,Name="spine_05")
    RotationOffset=(X=85,Y=30,Z=-10)
   )
   > RigVMVariableNode(Value=@n4.OutputParam) {variable=spr_upperarm_l_f_alpha}
   > n5 = &RigUnit_SphericalPoseReader(
    ActiveRegionScaleFactors=(PositiveWidth=0.1,NegativeWidth=1,PositiveHeight=0.1,NegativeHeight=1)
    ActiveRegionSize=0.1
    Debug=(bDrawDebug=false,DebugScale=25,DebugSegments=20,DebugThickness=0.25)
    DriverAxis=(X=-1,Y=0,Z=0)
    DriverItem=(Type=Bone,Name="upperarm_r")
    FalloffRegionScaleFactors=(PositiveWidth=0.4,NegativeWidth=1,PositiveHeight=0.8,NegativeHeight=0.5)
    FalloffSize=0.5
    FlipHeightScaling=true
    FlipWidthScaling=false
    OptionalParentItem=(Type=Bone,Name="spine_05")
    RotationOffset=(X=95,Y=-30,Z=170)
   )
   > RigVMVariableNode(Value=@n5.OutputParam) {variable=spr_upperarm_r_f_alpha}
   > n6 = &RigUnit_SphericalPoseReader(
    ActiveRegionScaleFactors=(PositiveWidth=0.1,NegativeWidth=2,PositiveHeight=2,NegativeHeight=2)
    ActiveRegionSize=0.1
    Debug=(bDrawDebug=false,DebugScale=25,DebugSegments=20,DebugThickness=0.25)
    DriverAxis=(X=1,Y=0,Z=0)
    DriverItem=(Type=Bone,Name="upperarm_l")
    FalloffRegionScaleFactors=(PositiveWidth=0.7,NegativeWidth=2,PositiveHeight=2,NegativeHeight=2.5)
    FalloffSize=0.5
    FlipHeightScaling=false
    FlipWidthScaling=false
    OptionalParentItem=(Type=Bone,Name="spine_05")
    RotationOffset=(X=180,Y=-10,Z=-40)
   )
   > RigVMVariableNode(Value=@n6.OutputParam) {variable=spr_upperarm_l_dn_alpha}
   > n7 = &RigUnit_SphericalPoseReader(
    ActiveRegionScaleFactors=(PositiveWidth=0.1,NegativeWidth=2,PositiveHeight=2,NegativeHeight=2)
    ActiveRegionSize=0.1
    Debug=(bDrawDebug=false,DebugScale=25,DebugSegments=20,DebugThickness=0.25)
    DriverAxis=(X=-1,Y=0,Z=0)
    DriverItem=(Type=Bone,Name="upperarm_r")
    FalloffRegionScaleFactors=(PositiveWidth=0.7,NegativeWidth=2,PositiveHeight=2,NegativeHeight=2.5)
    FalloffSize=0.5
    FlipHeightScaling=true
    FlipWidthScaling=false
    OptionalParentItem=(Type=Bone,Name="spine_05")
    RotationOffset=(X=0,Y=10,Z=140)
   )
   > RigVMVariableNode(Value=@n7.OutputParam) {variable=spr_upperarm_r_dn_alpha}
   > n8 = &RigUnit_SphericalPoseReader(
    ActiveRegionScaleFactors=(PositiveWidth=0.5,NegativeWidth=1,PositiveHeight=1,NegativeHeight=1)
    ActiveRegionSize=0.1
    Debug=(bDrawDebug=false,DebugScale=25,DebugSegments=20,DebugThickness=0.25)
    DriverAxis=(X=1,Y=0,Z=0)
    DriverItem=(Type=Bone,Name="upperarm_l")
    FalloffRegionScaleFactors=(PositiveWidth=1,NegativeWidth=1,PositiveHeight=0.7,NegativeHeight=1.5)
    FalloffSize=0.55
    FlipHeightScaling=false
    FlipWidthScaling=true
    OptionalParentItem=(Type=Bone,Name="spine_05")
    RotationOffset=(X=-40,Y=-90,Z=0)
   )
   > RigVMVariableNode(Value=@n8.OutputParam) {variable=spr_upperarm_l_bk_alpha}
   > n9 = &RigUnit_SphericalPoseReader(
    ActiveRegionScaleFactors=(PositiveWidth=0.5,NegativeWidth=1,PositiveHeight=1,NegativeHeight=1)
    ActiveRegionSize=0.1
    Debug=(bDrawDebug=false,DebugScale=25,DebugSegments=20,DebugThickness=0.25)
    DriverAxis=(X=-1,Y=0,Z=0)
    DriverItem=(Type=Bone,Name="upperarm_r")
    FalloffRegionScaleFactors=(PositiveWidth=1,NegativeWidth=1,PositiveHeight=0.7,NegativeHeight=1.5)
    FalloffSize=0.55
    FlipHeightScaling=false
    FlipWidthScaling=false
    OptionalParentItem=(Type=Bone,Name="spine_05")
    RotationOffset=(X=140,Y=-90,Z=0)
   )
   > RigVMVariableNode(Value=@n9.OutputParam) {variable=spr_upperarm_r_bk_alpha}
   > &RigUnit_AimConstraintLocalSpaceOffset(
    AdvancedSettings=(DebugSettings=(bEnabled=false,Scale=10,WorldOffset=(Rotation=(X=0,Y=0,Z=0,W=1),Translation=(X=0,Y=0,Z=0),Scale3D=(X=1,Y=1,Z=1))),RotationOrderForFilter=XZY)
    Child=(Type=Control,Name="thigh_l_ctrl")
    Filter=(bX=True,bY=True,bZ=True)
    Parents=((Item=(Type=Control,Name="calf_l_ctrl"),Weight=1))
    Weight=1
    WorldUp=(Target=(X=0,Y=0,Z=1),Kind=Direction,Space=(Type=None,Name="None"))
    bMaintainOffset=true
   )
   > n10 = &RigUnit_SphericalPoseReader(
    ActiveRegionScaleFactors=(PositiveWidth=2,NegativeWidth=2,PositiveHeight=1.5,NegativeHeight=2)
    ActiveRegionSize=0.1
    Debug=(bDrawDebug=false,DebugScale=25,DebugSegments=20,DebugThickness=0.25)
    DriverAxis=(X=-1,Y=0,Z=0)
    DriverItem=(Type=Control,Name="thigh_l_ctrl")
    FalloffRegionScaleFactors=(PositiveWidth=1,NegativeWidth=1,PositiveHeight=0.7,NegativeHeight=1)
    FalloffSize=0.4
    FlipHeightScaling=true
    FlipWidthScaling=true
    OptionalParentItem=(Type=Bone,Name="pelvis")
    RotationOffset=(X=170,Y=-10,Z=0)
   )
   > RigVMVariableNode(Value=@n10.OutputParam) {variable=spr_thigh_l_f_alpha}
   > &RigUnit_AimConstraintLocalSpaceOffset(
    AdvancedSettings=(DebugSettings=(bEnabled=false,Scale=10,WorldOffset=(Rotation=(X=0,Y=0,Z=0,W=1),Translation=(X=0,Y=0,Z=0),Scale3D=(X=1,Y=1,Z=1))),RotationOrderForFilter=XZY)
    Child=(Type=Control,Name="thigh_r_ctrl")
    Filter=(bX=True,bY=True,bZ=True)
    Parents=((Item=(Type=Control,Name="calf_r_ctrl"),Weight=1))
    Weight=1
    WorldUp=(Target=(X=0,Y=0,Z=1),Kind=Direction,Space=(Type=None,Name="None"))
    bMaintainOffset=true
   )
   > n11 = &RigUnit_SphericalPoseReader(
    ActiveRegionScaleFactors=(PositiveWidth=2,NegativeWidth=2,PositiveHeight=1.5,NegativeHeight=2)
    ActiveRegionSize=0.1
    Debug=(bDrawDebug=false,DebugScale=25,DebugSegments=20,DebugThickness=0.25)
    DriverAxis=(X=1,Y=0,Z=0)
    DriverItem=(Type=Control,Name="thigh_r_ctrl")
    FalloffRegionScaleFactors=(PositiveWidth=1,NegativeWidth=1,PositiveHeight=0.7,NegativeHeight=1)
    FalloffSize=0.4
    FlipHeightScaling=false
    FlipWidthScaling=false
    OptionalParentItem=(Type=Bone,Name="spine_01")
    RotationOffset=(X=10,Y=-10,Z=0)
   )
   > RigVMVariableNode(Value=@n11.OutputParam) {variable=spr_thigh_r_f_alpha}
   RigVMFunctionReturnNode(ExecuteContext=@SequenceAggregate.B)
  ]}
 New Function_ContainedGraph {nodes [
   RigVMFunctionEntryNode()
   > SequenceAggregate = RigVMAggregateNode()
   &RigUnit_ParentConstraint(
    AdvancedSettings=(InterpolationType=Average,RotationOrderForFilter=XZY)
    Child=(Type=Bone,Name="dyn_coat_sleeves_mid_MSTR_L")
    Filter=(TranslationFilter=(bX=true,bY=true,bZ=true),RotationFilter=(bX=true,bY=true,bZ=true),ScaleFilter=(bX=True,bY=True,bZ=True))
    Parents=((Item=(Type=Bone,Name="spine_05"),Weight=1),(Item=(Type=Bone,Name="clavicle_l"),Weight=0),(Item=(Type=Bone,Name="upperarm_twist_01_l"),Weight=0))
    Weight=1
    ExecutePin=@SequenceAggregate.A
   )
   > &RigUnit_ParentConstraint(
    AdvancedSettings=(InterpolationType=Average,RotationOrderForFilter=XZY)
    Child=(Type=Bone,Name="dyn_coat_collar_MSTR_fL")
    Filter=(TranslationFilter=(bX=true,bY=true,bZ=true),RotationFilter=(bX=true,bY=true,bZ=true),ScaleFilter=(bX=True,bY=True,bZ=True))
    Parents=((Item=(Type=Bone,Name="spine_05"),Weight=0.4),(Item=(Type=Bone,Name="clavicle_l"),Weight=0.3),(Item=(Type=Bone,Name="spine_04"),Weight=0.3))
    Weight=1
   )
   > &RigUnit_ParentConstraint(
    AdvancedSettings=(InterpolationType=Average,RotationOrderForFilter=XZY)
    Child=(Type=Bone,Name="dyn_coat_collar01_MSTR_fL")
    Filter=(TranslationFilter=(bX=true,bY=true,bZ=true),RotationFilter=(bX=true,bY=true,bZ=true),ScaleFilter=(bX=True,bY=True,bZ=True))
    Parents=((Item=(Type=Bone,Name="spine_05"),Weight=0.4),(Item=(Type=Bone,Name="clavicle_l"),Weight=0.3),(Item=(Type=Bone,Name="spine_04"),Weight=0.3))
    Weight=1
   )
   n1 = RigVMVariableNode() {variable=spr_upperarm_l_f_alpha}
   > &RigUnit_ParentConstraint(
    AdvancedSettings=(InterpolationType=Average,RotationOrderForFilter=XZY)
    Child=(Type=Bone,Name="dyn_coat_sleeves_mid_MSTR_L")
    Filter=(TranslationFilter=(bX=true,bY=true,bZ=true),RotationFilter=(bX=true,bY=false,bZ=false),ScaleFilter=(bX=True,bY=True,bZ=True))
    Parents=((Item=(Type=Bone,Name="spine_05"),Weight=0.2),(Item=(Type=Bone,Name="clavicle_l"),Weight=0),(Item=(Type=Bone,Name="upperarm_twist_01_l"),Weight=0.8))
    Weight=@n1.Value
   )
   > &RigUnit_ModifyTransforms(
    ItemToModify="((Item=(Type=Bone,Name=\"dyn_coat_sleeves_mid_MSTR_L\"),Transform=(Rotation=(X=-0.1574,Y=-0.0734,Z=0.4162,W=0.8925),Translation=(X=0,Y=3,Z=5),Scale3D=(X=1,Y=1,Z=1))),(Item=(Type=Bone,Name=\"dyn_coat_collar_01_fL\"),Transform=(Rotation=(X=-0.5,Y=0,Z=0,W=0.866), ...(truncated)"
    Weight=@n1.Value
   )
   > &RigUnit_ModifyTransforms(
    ItemToModify="((Item=(Type=Bone,Name=\"dyn_coat_sleeves_mid_MSTR_L\"),Transform=(Rotation=(X=0,Y=0,Z=0.1736,W=0.9848),Translation=(X=-5,Y=3,Z=-2),Scale3D=(X=1,Y=1,Z=1))),(Item=(Type=Bone,Name=\"deform_upperarm_bk_l\"),Transform=(Rotation=(X=0,Y=0,Z=0,W=1),Translation=(X=0,Y ...(truncated)"
    Weight=RigVMVariableNode() {variable=spr_upperarm_l_up_alpha}.Value
   )
   > &RigUnit_ModifyTransforms(
    ItemToModify="((Item=(Type=Bone,Name=\"dyn_coat_sleeves_mid_MSTR_L\"),Transform=(Rotation=(X=-0.5824,Y=0.0088,Z=-0.305,W=0.7534),Translation=(X=-1,Y=0,Z=-8),Scale3D=(X=1,Y=1,Z=1))),(Item=(Type=Bone,Name=\"dyn_coat_collar_01_fL\"),Transform=(Rotation=(X=-0.7071,Y=0,Z=0,W=0.7 ...(truncated)"
    Weight=RigVMVariableNode() {variable=spr_upperarm_l_dn_alpha}.Value
   )
   n2 = RigVMVariableNode() {variable=spr_upperarm_l_bk_alpha}
   > &RigUnit_ParentConstraint(
    AdvancedSettings=(InterpolationType=Average,RotationOrderForFilter=XZY)
    Child=(Type=Bone,Name="dyn_coat_sleeves_mid_MSTR_L")
    Filter=(TranslationFilter=(bX=true,bY=true,bZ=true),RotationFilter=(bX=true,bY=true,bZ=true),ScaleFilter=(bX=True,bY=True,bZ=True))
    Parents=((Item=(Type=Bone,Name="spine_05"),Weight=0.5),(Item=(Type=Bone,Name="clavicle_l"),Weight=0),(Item=(Type=Bone,Name="upperarm_twist_01_l"),Weight=0.5))
    Weight=@n2.Value
   )
   > &RigUnit_ModifyTransforms(
    ItemToModify="((Item=(Type=Bone,Name=\"dyn_coat_sleeves_mid_MSTR_L\"),Transform=(Rotation=(X=-0.1736,Y=0,Z=0,W=0.9848),Translation=(X=5,Y=-5,Z=3),Scale3D=(X=1,Y=1,Z=1))),(Item=(Type=Bone,Name=\"deform_upperarm_bk_l\"),Transform=(Rotation=(X=0,Y=0,Z=0,W=1),Translation=(X=0,Y ...(truncated)"
    Weight=@n2.Value
   )
   > n3 = &RigUnit_SphericalPoseReader(
    ActiveRegionScaleFactors=(PositiveWidth=2,NegativeWidth=2,PositiveHeight=2,NegativeHeight=2)
    ActiveRegionSize=0.1
    Debug=(bDrawDebug=false,DebugScale=25,DebugSegments=20,DebugThickness=0.25)
    DriverAxis=(X=-1,Y=0,Z=0)
    DriverItem=(Type=Bone,Name="hand_r")
    FalloffRegionScaleFactors=(PositiveWidth=1,NegativeWidth=1,PositiveHeight=1,NegativeHeight=0.6)
    FalloffSize=0.6
    FlipHeightScaling=false
    FlipWidthScaling=false
    OptionalParentItem=(Type=Bone,Name="lowerarm_r")
    RotationOffset=(X=-40,Y=0,Z=0)
   )
   > &RigUnit_ModifyTransforms(
    ItemToModify="((Item=(Type=Bone,Name=\"base_sleeveFront_r\"),Transform=(Rotation=(X=0,Y=0,Z=0,W=1),Translation=(X=3,Y=1,Z=0),Scale3D=(X=1,Y=1,Z=1))),(Item=(Type=Bone,Name=\"base_sleeveTop_r\"),Transform=(Rotation=(X=0,Y=0,Z=0,W=1),Translation=(X=0,Y=0,Z=0),Scale3D=(X=1,Y=1, ...(truncated)"
    Weight=@n3.OutputParam
   )
   > n4 = &RigUnit_SphericalPoseReader(
    ActiveRegionScaleFactors=(PositiveWidth=2,NegativeWidth=2,PositiveHeight=2,NegativeHeight=2)
    ActiveRegionSize=0.1
    Debug=(bDrawDebug=false,DebugScale=25,DebugSegments=20,DebugThickness=0.25)
    DriverAxis=(X=-1,Y=0,Z=0)
    DriverItem=(Type=Bone,Name="hand_r")
    FalloffRegionScaleFactors=(PositiveWidth=1,NegativeWidth=1,PositiveHeight=1,NegativeHeight=0.6)
    FalloffSize=0.6
    FlipHeightScaling=false
    FlipWidthScaling=false
    OptionalParentItem=(Type=Bone,Name="lowerarm_r")
    RotationOffset=(X=180,Y=0,Z=0)
   )
   > &RigUnit_ModifyTransforms(
    ItemToModify=((Item=(Type=Bone,Name="base_sleeveBack_r"),Transform=(Rotation=(X=0.0019,Y=0.0436,Z=-0.0436,W=0.9981),Translation=(X=5,Y=2,Z=0),Scale3D=(X=1,Y=1,Z=1))))
    Weight=@n4.OutputParam
   )
   > n5 = &RigUnit_SphericalPoseReader(
    ActiveRegionScaleFactors=(PositiveWidth=2,NegativeWidth=2,PositiveHeight=2,NegativeHeight=2)
    ActiveRegionSize=0.1
    Debug=(bDrawDebug=false,DebugScale=25,DebugSegments=20,DebugThickness=0.25)
    DriverAxis=(X=1,Y=0,Z=0)
    DriverItem=(Type=Bone,Name="hand_l")
    FalloffRegionScaleFactors=(PositiveWidth=1,NegativeWidth=0.6,PositiveHeight=0.6,NegativeHeight=0.6)
    FalloffSize=0.6
    FlipHeightScaling=false
    FlipWidthScaling=true
    OptionalParentItem=(Type=Bone,Name="lowerarm_l")
    RotationOffset=(X=140,Y=10,Z=0)
   )
   > &RigUnit_ModifyTransforms(
    ItemToModify=((Item=(Type=Bone,Name="base_sleeveFront_l"),Transform=(Rotation=(X=0,Y=0,Z=0,W=1),Translation=(X=-3,Y=0,Z=0),Scale3D=(X=1,Y=1,Z=1))))
    Weight=@n5.OutputParam
   )
   > n6 = &RigUnit_SphericalPoseReader(
    ActiveRegionScaleFactors=(PositiveWidth=2,NegativeWidth=2,PositiveHeight=2,NegativeHeight=2)
    ActiveRegionSize=0.1
    Debug=(bDrawDebug=false,DebugScale=25,DebugSegments=20,DebugThickness=0.25)
    DriverAxis=(X=1,Y=0,Z=0)
    DriverItem=(Type=Bone,Name="hand_l")
    FalloffRegionScaleFactors=(PositiveWidth=1,NegativeWidth=0.6,PositiveHeight=0.6,NegativeHeight=0.6)
    FalloffSize=0.6
    FlipHeightScaling=false
    FlipWidthScaling=true
    OptionalParentItem=(Type=Bone,Name="lowerarm_l")
    RotationOffset=(X=0,Y=0,Z=0)
   )
   > &RigUnit_ModifyTransforms(
    ItemToModify=((Item=(Type=Bone,Name="base_sleeveBack_l"),Transform=(Rotation=(X=0,Y=0,Z=0,W=1),Translation=(X=-5,Y=0,Z=0),Scale3D=(X=1,Y=1,Z=1))))
    Weight=@n6.OutputParam
   )
   &RigUnit_ParentConstraint(
    AdvancedSettings=(InterpolationType=Average,RotationOrderForFilter=XZY)
    Child=(Type=Bone,Name="dyn_coat_sleeves_mid_MSTR_R")
    Filter=(TranslationFilter=(bX=true,bY=true,bZ=true),RotationFilter=(bX=true,bY=true,bZ=true),ScaleFilter=(bX=True,bY=True,bZ=True))
    Parents=((Item=(Type=Bone,Name="spine_05"),Weight=1),(Item=(Type=Bone,Name="clavicle_r"),Weight=0),(Item=(Type=Bone,Name="upperarm_twist_01_r"),Weight=0))
    Weight=1
    ExecutePin=@SequenceAggregate.B
   )
   > &RigUnit_ParentConstraint(
    AdvancedSettings=(InterpolationType=Average,RotationOrderForFilter=XZY)
    Child=(Type=Bone,Name="dyn_coat_collar_MSTR_fR")
    Filter=(TranslationFilter=(bX=true,bY=true,bZ=true),RotationFilter=(bX=true,bY=true,bZ=true),ScaleFilter=(bX=True,bY=True,bZ=True))
    Parents=((Item=(Type=Bone,Name="spine_05"),Weight=0.4),(Item=(Type=Bone,Name="clavicle_r"),Weight=0.3),(Item=(Type=Bone,Name="spine_04"),Weight=0.3))
    Weight=1
   )
   > &RigUnit_ParentConstraint(
    AdvancedSettings=(InterpolationType=Average,RotationOrderForFilter=XZY)
    Child=(Type=Bone,Name="dyn_coat_collar01_MSTR_fR")
    Filter=(TranslationFilter=(bX=true,bY=true,bZ=true),RotationFilter=(bX=true,bY=true,bZ=true),ScaleFilter=(bX=True,bY=True,bZ=True))
    Parents=((Item=(Type=Bone,Name="spine_05"),Weight=0.4),(Item=(Type=Bone,Name="clavicle_r"),Weight=0.3),(Item=(Type=Bone,Name="spine_04"),Weight=0.3))
    Weight=1
   )
   n7 = RigVMVariableNode() {variable=spr_upperarm_r_f_alpha}
   > &RigUnit_ParentConstraint(
    AdvancedSettings=(InterpolationType=Average,RotationOrderForFilter=XZY)
    Child=(Type=Bone,Name="dyn_coat_sleeves_mid_MSTR_R")
    Filter=(TranslationFilter=(bX=true,bY=true,bZ=true),RotationFilter=(bX=true,bY=false,bZ=false),ScaleFilter=(bX=True,bY=True,bZ=True))
    Parents=((Item=(Type=Bone,Name="spine_05"),Weight=0.2),(Item=(Type=Bone,Name="clavicle_r"),Weight=0),(Item=(Type=Bone,Name="upperarm_twist_01_r"),Weight=0.8))
    Weight=@n7.Value
   )
   > &RigUnit_ModifyTransforms(
    ItemToModify="((Item=(Type=Bone,Name=\"dyn_coat_sleeves_mid_MSTR_R\"),Transform=(Rotation=(X=-0.2597,Y=-0.1601,Z=-0.4427,W=0.8431),Translation=(X=0,Y=-3,Z=5),Scale3D=(X=1,Y=1,Z=1))),(Item=(Type=Bone,Name=\"dyn_coat_collar_01_fR\"),Transform=(Rotation=(X=0,Y=-0.4226,Z=0,W=0. ...(truncated)"
    Weight=@n7.Value
   )
   > &RigUnit_ModifyTransforms(
    ItemToModify="((Item=(Type=Bone,Name=\"dyn_coat_sleeves_mid_MSTR_R\"),Transform=(Rotation=(X=0,Y=0,Z=-0.1736,W=0.9848),Translation=(X=-5,Y=-3,Z=-2),Scale3D=(X=1,Y=1,Z=1))),(Item=(Type=Bone,Name=\"dyn_coat_collar01_01_fR\"),Transform=(Rotation=(X=0.2588,Y=0,Z=0,W=0.9659),Tra ...(truncated)"
    Weight=RigVMVariableNode() {variable=spr_upperarm_r_up_alpha}.Value
   )
   > &RigUnit_ModifyTransforms(
    ItemToModify="((Item=(Type=Bone,Name=\"dyn_coat_sleeves_mid_MSTR_R\"),Transform=(Rotation=(X=0.5824,Y=0.0088,Z=0.305,W=0.7534),Translation=(X=-1,Y=5,Z=-8),Scale3D=(X=1,Y=1,Z=1))),(Item=(Type=Bone,Name=\"dyn_coat_collar_01_fR\"),Transform=(Rotation=(X=0.8529,Y=0.0868,Z=-0.15 ...(truncated)"
    Weight=RigVMVariableNode() {variable=spr_upperarm_r_dn_alpha}.Value
   )
   n8 = RigVMVariableNode() {variable=spr_upperarm_r_bk_alpha}
   > &RigUnit_ParentConstraint(
    AdvancedSettings=(InterpolationType=Average,RotationOrderForFilter=XZY)
    Child=(Type=Bone,Name="dyn_coat_sleeves_mid_MSTR_R")
    Filter=(TranslationFilter=(bX=true,bY=true,bZ=true),RotationFilter=(bX=true,bY=true,bZ=true),ScaleFilter=(bX=True,bY=True,bZ=True))
    Parents=((Item=(Type=Bone,Name="spine_05"),Weight=0.5),(Item=(Type=Bone,Name="clavicle_r"),Weight=0),(Item=(Type=Bone,Name="upperarm_twist_01_r"),Weight=0.5))
    Weight=@n8.Value
   )
   > &RigUnit_ModifyTransforms(
    ItemToModify="((Item=(Type=Bone,Name=\"dyn_coat_sleeves_mid_MSTR_R\"),Transform=(Rotation=(X=0.1736,Y=0,Z=0,W=0.9848),Translation=(X=5,Y=5,Z=3),Scale3D=(X=1,Y=1,Z=1))),(Item=(Type=Bone,Name=\"dyn_coat_ribbon_MSTR_R\"),Transform=(Rotation=(X=0,Y=0.0785,Z=0,W=0.9969),Translat ...(truncated)"
    Weight=@n8.Value
   )
   RigVMFunctionReturnNode(ExecuteContext=@SequenceAggregate.D)
  ]}
 New Function_ContainedGraph {nodes [
   RigVMFunctionEntryNode()
   > SequenceAggregate = RigVMAggregateNode()
   &RigUnit_ParentConstraint(
    AdvancedSettings=(RotationOrderForFilter=XZY)
    Child=(Type=Bone,Name="dyn_coat_MSTR_fL")
    Filter=(TranslationFilter=(bX=false,bY=false,bZ=false),RotationFilter=(bX=True,bY=True,bZ=True),ScaleFilter=(bX=True,bY=True,bZ=True))
    Parents=((Item=(Type=Bone,Name="spine_04"),Weight=0.3),(Item=(Type=Bone,Name="pelvis"),Weight=0.7))
    Weight=1
    ExecutePin=@SequenceAggregate.A
   )
   > &RigUnit_ParentConstraint(
    AdvancedSettings=(RotationOrderForFilter=XZY)
    Child=(Type=Bone,Name="dyn_coat_MSTR_fR")
    Filter=(TranslationFilter=(bX=false,bY=false,bZ=false),RotationFilter=(bX=True,bY=True,bZ=True),ScaleFilter=(bX=True,bY=True,bZ=True))
    Parents=((Item=(Type=Bone,Name="spine_04"),Weight=0.3),(Item=(Type=Bone,Name="pelvis"),Weight=0.7))
    Weight=1
   )
   > &RigUnit_ParentConstraint(
    AdvancedSettings=(RotationOrderForFilter=XZY)
    Child=(Type=Bone,Name="dyn_coat_MSTR_L")
    Filter=(TranslationFilter=(bX=true,bY=true,bZ=true),RotationFilter=(bX=True,bY=True,bZ=True),ScaleFilter=(bX=True,bY=True,bZ=True))
    Parents=((Item=(Type=Bone,Name="spine_04"),Weight=0.7),(Item=(Type=Bone,Name="pelvis"),Weight=0.3))
    Weight=1
   )
   > &RigUnit_ParentConstraint(
    AdvancedSettings=(RotationOrderForFilter=XZY)
    Child=(Type=Bone,Name="dyn_coat_bk_MSTR")
    Filter=(TranslationFilter=(bX=true,bY=true,bZ=true),RotationFilter=(bX=True,bY=True,bZ=True),ScaleFilter=(bX=True,bY=True,bZ=True))
    Parents=((Item=(Type=Bone,Name="spine_04"),Weight=0.7),(Item=(Type=Bone,Name="pelvis"),Weight=0.3))
    Weight=1
   )
   > &RigUnit_ParentConstraint(
    AdvancedSettings=(RotationOrderForFilter=XZY)
    Child=(Type=Bone,Name="dyn_coat_MSTR_R")
    Filter=(TranslationFilter=(bX=true,bY=true,bZ=true),RotationFilter=(bX=True,bY=True,bZ=True),ScaleFilter=(bX=True,bY=True,bZ=True))
    Parents=((Item=(Type=Bone,Name="spine_04"),Weight=0.7),(Item=(Type=Bone,Name="pelvis"),Weight=0.3))
    Weight=1
   )
   > &RigUnit_ParentConstraint(
    AdvancedSettings=(RotationOrderForFilter=XZY)
    Child=(Type=Bone,Name="dyn_coat_02_L")
    Filter=(TranslationFilter=(bX=true,bY=true,bZ=true),RotationFilter=(bX=True,bY=True,bZ=True),ScaleFilter=(bX=True,bY=True,bZ=True))
    Parents=((Item=(Type=Bone,Name="spine_02"),Weight=0.7),(Item=(Type=Bone,Name="pelvis"),Weight=0.3))
    Weight=1
   )
   > &RigUnit_ParentConstraint(
    AdvancedSettings=(RotationOrderForFilter=XZY)
    Child=(Type=Bone,Name="dyn_coat_bk_02")
    Filter=(TranslationFilter=(bX=true,bY=true,bZ=true),RotationFilter=(bX=True,bY=True,bZ=True),ScaleFilter=(bX=True,bY=True,bZ=True))
    Parents=((Item=(Type=Bone,Name="spine_02"),Weight=0.7),(Item=(Type=Bone,Name="pelvis"),Weight=0.3))
    Weight=1
   )
   > &RigUnit_ParentConstraint(
    AdvancedSettings=(RotationOrderForFilter=XZY)
    Child=(Type=Bone,Name="dyn_coat_02_R")
    Filter=(TranslationFilter=(bX=true,bY=true,bZ=true),RotationFilter=(bX=True,bY=True,bZ=True),ScaleFilter=(bX=True,bY=True,bZ=True))
    Parents=((Item=(Type=Bone,Name="spine_02"),Weight=0.7),(Item=(Type=Bone,Name="pelvis"),Weight=0.3))
    Weight=1
   )
   n1 = RigVMVariableNode() {variable=spr_thigh_l_f_alpha}
   > &RigUnit_ParentConstraint(
    AdvancedSettings=(RotationOrderForFilter=XZY)
    Child=(Type=Bone,Name="dyn_coat_MSTR_fL")
    Filter=(TranslationFilter=(bX=false,bY=false,bZ=false),RotationFilter=(bX=True,bY=True,bZ=True),ScaleFilter=(bX=True,bY=True,bZ=True))
    Parents=((Item=(Type=Bone,Name="thigh_twist_01_l"),Weight=0.3),(Item=(Type=Bone,Name="pelvis"),Weight=0.7))
    Weight=@n1.Value
   )
   > &RigUnit_ModifyTransforms(
    ItemToModify="((Item=(Type=Bone,Name=\"dyn_coat_MSTR_fL\"),Transform=(Rotation=(X=0.0519,Y=-0.1113,Z=-0.3323,W=0.9352),Translation=(X=0,Y=0,Z=0),Scale3D=(X=1,Y=1,Z=1))),(Item=(Type=Bone,Name=\"dyn_spinekerchief_MSTR_fL\"),Transform=(Rotation=(X=-0.2346,Y=-0.4082,Z=-0.1094,W ...(truncated)"
    Weight=@n1.Value
   )
   n2 = RigVMVariableNode() {variable=spr_thigh_r_f_alpha}
   > &RigUnit_ParentConstraint(
    AdvancedSettings=(RotationOrderForFilter=XZY)
    Child=(Type=Bone,Name="dyn_coat_MSTR_fR")
    Filter=(TranslationFilter=(bX=false,bY=false,bZ=false),RotationFilter=(bX=True,bY=True,bZ=True),ScaleFilter=(bX=True,bY=True,bZ=True))
    Parents=((Item=(Type=Bone,Name="thigh_twist_01_r"),Weight=0.3),(Item=(Type=Bone,Name="pelvis"),Weight=0.7))
    Weight=@n2.Value
   )
   > &RigUnit_ModifyTransforms(
    ItemToModify=((Item=(Type=Bone,Name="dyn_coat_MSTR_fR"),Transform=(Rotation=(X=-0.0519,Y=-0.1113,Z=0.3323,W=0.9352),Translation=(X=0,Y=0,Z=0),Scale3D=(X=1,Y=1,Z=1))))
    Weight=@n2.Value
   )
   > SequenceAggregate_1 = RigVMAggregateNode()
   RigVMFunction_ControlFlowBranch = &RigVMFunction_ControlFlowBranch(ExecuteContext=@SequenceAggregate_1.A, Condition=RigVMVariableNode() {variable=isSliding}.Value)
   n3 = &RigVMFunction_ControlFlowBranch(ExecuteContext=@RigVMFunction_ControlFlowBranch.True, Condition=RigVMVariableNode() {variable=isBaseballSliding}.Value)
   &RigUnit_ModifyTransforms(
    ItemToModify="((Item=(Type=Bone,Name=\"dyn_coat_bk_MSTR\"),Transform=(Rotation=(X=0,Y=-0.342,Z=0,W=0.9397),Translation=(X=0,Y=0,Z=0),Scale3D=(X=1,Y=1,Z=1))),(Item=(Type=Bone,Name=\"dyn_coat_MSTR_L\"),Transform=(Rotation=(X=0.067,Y=-0.25,Z=0.25,W=0.933),Translation=(X=0,Y=0, ...(truncated)"
    Weight=1
    ExecutePin=@n3.False
   )
   &RigUnit_ModifyTransforms(
    ItemToModify="((Item=(Type=Bone,Name=\"dyn_coat_bk_MSTR\"),Transform=(Rotation=(X=0,Y=-0.0436,Z=0,W=0.999),Translation=(X=0,Y=0,Z=0),Scale3D=(X=1,Y=1,Z=1))),(Item=(Type=Bone,Name=\"dyn_coat_MSTR_L\"),Transform=(Rotation=(X=0.0226,Y=-0.0842,Z=0.2578,W=0.9623),Translation=(X= ...(truncated)"
    Weight=1
    ExecutePin=@n3.True
   )
   n4 = &RigVMFunction_ControlFlowBranch(ExecuteContext=@SequenceAggregate_1.B, Condition=RigVMVariableNode() {variable=DBNO}.Value)
   n5 = &RigVMFunction_ControlFlowBranch(ExecuteContext=@n4.True, Condition=RigVMVariableNode() {variable=DBNO_Carry}.Value)
   &RigUnit_ModifyTransforms(
    ItemToModify="((Item=(Type=Bone,Name=\"dyn_coat_sleeves_mid_MSTR_L\"),Transform=(Rotation=(X=-0.067,Y=0.25,Z=0.25,W=0.933),Translation=(X=0,Y=0,Z=0),Scale3D=(X=1,Y=1,Z=1))),(Item=(Type=Bone,Name=\"dyn_coat_sleeves_mid_MSTR_R\"),Transform=(Rotation=(X=0.4427,Y=0.1601,Z=-0.25 ...(truncated)"
    Weight=1
    ExecutePin=@n5.False
   )
   &RigUnit_ParentConstraint(
    AdvancedSettings=(RotationOrderForFilter=XZY)
    Child=(Type=Bone,Name="dyn_coat_sleeves_mid_MSTR_R")
    Filter=(TranslationFilter=(bX=false,bY=false,bZ=false),RotationFilter=(bX=True,bY=True,bZ=True),ScaleFilter=(bX=True,bY=True,bZ=True))
    Parents=((Item=(Type=Bone,Name="spine_05"),Weight=15),(Item=(Type=Bone,Name="dyn_coat_sleeves_mid_MSTR_R"),Weight=5))
    Weight=1
    ExecutePin=@n5.True
   )
   > &RigUnit_ModifyTransforms(
    ItemToModify="((Item=(Type=Bone,Name=\"dyn_coat_sleeves_mid_MSTR_L\"),Transform=(Rotation=(X=-0.067,Y=0.25,Z=0.25,W=0.933),Translation=(X=0,Y=0,Z=0),Scale3D=(X=1,Y=1,Z=1))),(Item=(Type=Bone,Name=\"dyn_coat_sleeves_mid_MSTR_R\"),Transform=(Rotation=(X=-0.0368,Y=0.421,Z=0.079 ...(truncated)"
    Weight=1
   )
   RigVMFunctionReturnNode(ExecuteContext=@SequenceAggregate.C)
  ]}
 RigVMFunctionLibrary {nodes [
   Coat = RigVMCollapseNode(ExecuteContext=())
   Coat Sleeves = RigVMCollapseNode(ExecuteContext=())
   Spherical Pose = RigVMCollapseNode(ExecuteContext=())
  ]}
 RigVMModel {nodes [
   RigUnit_BeginExecution = &RigUnit_BeginExecution()
   FN_DeformRig_FNC = RigVMFunctionReferenceNode(
    DeformRig_ Curve="(deformRig_l_upperarm_up_alpha_72_#1=false,deformRig_r_upperarm_up_alpha_73_#2=False,deformRig_l_upperarm_fr_alpha_74_#3=False,deformRig_r_upperarm_fr_alpha_75_CB51B8 ...(truncated)"
    DeformRig_Bool="(upperarm_1_#4=True,elbow_3_#5=True,wrist_7_#6=True,groin_8_#7=True,glute_17_#8=True,knee_9_40C408324CA261 ...(truncated)"
    DeformRig_Debug=(upperarm_debug_18_#4=False,wrist_debug_20_#6=False,lat_debug_26_#9=False,groinGlute_debug_27_#7=False)
    DeformRig_Value="(upperarm_up_18_#10=(Rotation=(X=0.0,Y=-0.0,Z=0.0,W=1.0),Translation=(X=0,Y=0,Z=3),Scale3D=(X=0.70,Y=1.0,Z=1.0)),upperarm_fwd_fr_19_#11=(Rotation=(X=0,Y=0,Z=0,W=1),Translation=(X=-20,Y=15,Z=-20),Sca ...(truncated)"
    ExecuteContext=@RigUnit_BeginExecution.ExecutePin
   )
   > Spherical Pose = RigVMFunctionReferenceNode()
   > Coat Sleeves = RigVMFunctionReferenceNode()
   > Coat = RigVMFunctionReferenceNode()
  ]}
 SequenceAggregate_1_ContainedGraph {nodes [
   n1 = RigVMFunctionEntryNode()
   SequenceAggregate = &RigVMFunction_Sequence(ExecuteContext=@n1.ExecuteContext)
   SequenceAggregate_1 = &RigVMFunction_Sequence(ExecuteContext=@SequenceAggregate.B)
   SequenceAggregate_2 = &RigVMFunction_Sequence(ExecuteContext=@SequenceAggregate_1.B)
   RigVMFunctionReturnNode(A=@SequenceAggregate.A, B=@SequenceAggregate_1.A, C=@SequenceAggregate_2.A, D=@SequenceAggregate_2.B)
  ]}
 SequenceAggregate_1_ContainedGraph {nodes [
   n1 = RigVMFunctionEntryNode()
   SequenceAggregate = &RigVMFunction_Sequence(ExecuteContext=@n1.ExecuteContext)
   SequenceAggregate_1 = &RigVMFunction_Sequence(ExecuteContext=@SequenceAggregate.B)
   RigVMFunctionReturnNode(A=@SequenceAggregate.A, B=@SequenceAggregate_1.A, C=@SequenceAggregate_1.B)
  ]}
 SequenceAggregate_1_ContainedGraph {nodes [
   n1 = RigVMFunctionEntryNode()
   SequenceAggregate = &RigVMFunction_Sequence(ExecuteContext=@n1.ExecuteContext)
   SequenceAggregate_1 = &RigVMFunction_Sequence(ExecuteContext=@SequenceAggregate.B)
   RigVMFunctionReturnNode(A=@SequenceAggregate.A, B=@SequenceAggregate_1.A, C=@SequenceAggregate_1.B)
  ]}
]
hierarchy [
 root {loc=<0 0 0>, rot=<0 0 0 1>}
 pelvis {parent=root, loc=<0 -0.5301 98.0571>, rot=<0.0127 0.707 -0.0127 -0.707>}
 spine_01 {parent=pelvis, loc=<6.1208 -0.9328 0>, rot=<0 0 0.034 -0.9994>}
 spine_02 {parent=spine_01, loc=<6.5997 -0.2095 0>, rot=<0 0 0.0238 -0.9997>}
 spine_03 {parent=spine_02, loc=<6.4914 -0.1237 0>, rot=<0 0 0.0149 -0.9999>}
 spine_04 {parent=spine_03, loc=<6.9598 0.0794 0>, rot=<0 0 -0.0132 -0.9999>}
 spine_05 {parent=spine_04, loc=<6.9987 0 0>, rot=<0 0 -0.0015 -1>}
 neck_01 {parent=spine_05, loc=<11.3138 0.1529 0>, rot=<0 0 -0.2003 -0.9797>}
 neck_02 {parent=neck_01, loc=<4.5627 0 0>, rot=<0 0 0.0504 -0.9987>}
 head {parent=neck_02, loc=<5.3914 0 0>, rot=<0 0 0.115 -0.9934>}
 clavicle_l {parent=spine_05, loc=<5.4831 -0.1053 -3.2946>, rot=<0.0465 -0.7289 -0.009 -0.6829>}
 upperarm_l {parent=clavicle_l, loc=<9.8658 -1.2136 -0.4105>, rot=<0.0164 -0.3245 -0.0269 -0.9454>}
 lowerarm_l {parent=upperarm_l, loc=<26.465 -0.8345 0.259>, rot=<0.0036 0.0161 -0.2298 -0.9731>}
 hand_l {parent=lowerarm_l, loc=<22.7355 -0.3203 -0.6105>, rot=<-0.6784 -0.0223 0.0735 -0.7307>}
 thumb_01_l {parent=hand_l, loc=<1.3118 -0.3191 -2.459>, rot=<0.6546 -0.4042 0.0664 -0.6354>}
 thumb_02_l {parent=thumb_01_l, loc=<5.9336 -0.0448 0.0401>, rot=<-0.0037 0.0213 0.298 -0.9543>}
 thumb_03_l {parent=thumb_02_l, loc=<2.2223 0.3682 -0.0907>, rot=<0.0118 -0.013 0.0066 -0.9998>}
 index_metacarpal_l {parent=hand_l, loc=<1.6535 0.0183 -2.96>, rot=<0.0158 -0.0124 -0.0173 -0.9996>}
 index_01_l {parent=index_metacarpal_l, loc=<8.3008 0.0357 -0.0524>, rot=<-0.0044 0.0485 0.0924 -0.9945>}
 index_02_l {parent=index_01_l, loc=<3.3587 -0.3004 0.0452>, rot=<-0.002 0.0027 0.3225 -0.9466>}
 index_03_l {parent=index_02_l, loc=<2.5578 -0.123 -0.0252>, rot=<-0.002 0.0067 0.2558 -0.9667>}
 middle_metacarpal_l {parent=hand_l, loc=<2.1856 -0.1725 -0.5984>, rot=<0.006 0.0168 -0.0329 -0.9993>}
 middle_01_l {parent=middle_metacarpal_l, loc=<7.3275 -0.2447 0.0129>, rot=<-0.0208 0.0042 0.0858 -0.9961>}
 middle_02_l {parent=middle_01_l, loc=<4.0123 0.1401 0.1175>, rot=<-0.0218 0.0258 0.3512 -0.9357>}
 middle_03_l {parent=middle_02_l, loc=<2.6077 0.0534 0.1373>, rot=<0.0082 0.0188 0.1708 -0.9851>}
 ring_metacarpal_l {parent=hand_l, loc=<2.2074 -0.3625 1.3638>, rot=<-0.0284 0.0722 -0.0458 -0.9959>}
 ring_01_l {parent=ring_metacarpal_l, loc=<7.0295 -0.6197 -0.1446>, rot=<-0.0489 -0.0252 0.1892 -0.9804>}
 ring_02_l {parent=ring_01_l, loc=<3.6174 0.4647 -0.1027>, rot=<0 0 0.2806 -0.9598>}
 ring_03_l {parent=ring_02_l, loc=<2.4152 0.114 0.0199>, rot=<0 0 0.1189 -0.9929>}
 pinky_metacarpal_l {parent=hand_l, loc=<1.7616 -0.9753 3.3814>, rot=<-0.0397 0.0814 -0.069 -0.9935>}
 pinky_01_l {parent=pinky_metacarpal_l, loc=<7.1709 -0.4831 -0.1094>, rot=<-0.0654 -0.0035 0.2242 -0.9723>}
 pinky_02_l {parent=pinky_01_l, loc=<2.8274 0.0893 -0.1043>, rot=<0.0037 0.0011 0.2854 -0.9584>}
 pinky_03_l {parent=pinky_02_l, loc=<2.2046 0.0307 -0.0078>, rot=<-0.0002 0.0098 0.1445 -0.9895>}
 lowerarm_twist_02_l {parent=lowerarm_l, loc=<10.6707 0 0>, rot=<0 0 0 1>}
 lowerarm_twist_01_l {parent=lowerarm_l, loc=<18.316 0 0>, rot=<0 0 0 1>}
 base_sleeveTop_l {parent=lowerarm_twist_01_l, loc=<-1.7715 -0.6638 2.3435>, rot=<0.6955 -0.0473 0.0332 -0.7162>}
 dyn_sleeveTop_l {parent=base_sleeveTop_l, loc=<0 0 0>, rot=<0 0 0 1>}
 base_sleeveBottom_l {parent=lowerarm_twist_01_l, loc=<-2.3465 0.3224 -2.5589>, rot=<-0.7552 0.0204 0.0848 -0.6497>}
 dyn_sleeveBottom_l {parent=base_sleeveBottom_l, loc=<0 0 0>, rot=<0 0 0 1>}
 base_sleeveBack_l {parent=lowerarm_twist_01_l, loc=<-2.2975 -2.828 -1.1131>, rot=<0.0314 -0.0308 0.0923 -0.9948>}
 dyn_sleeveBack_l {parent=base_sleeveBack_l, loc=<0 0 0>, rot=<0 0 0 1>}
 base_sleeveFront_l {parent=lowerarm_twist_01_l, loc=<-2.1476 2.448 -0.4896>, rot=<0.9964 -0.0601 -0.0269 -0.0538>}
 dyn_sleeveFront_l {parent=base_sleeveFront_l, loc=<0 0 0>, rot=<0 0 0 1>}
 deform_elbow_in_base_l {parent=lowerarm_l, loc=<1.5474 3.4661 -1.7551>, rot=<0.0285 -0.0335 0.0024 -0.999>}
 deform_elbow_in_l {parent=deform_elbow_in_base_l, loc=<0 0 0>, rot=<0 0 0 1>}
 deform_elbow_out_base_l {parent=lowerarm_l, loc=<-0.1062 -3.1717 1.2591>, rot=<0.0285 -0.0335 0.0024 -0.999>}
 deform_elbow_out_l {parent=deform_elbow_out_base_l, loc=<0 0 0>, rot=<0 0 0 1>}
 upperarm_twist_01_l {parent=upperarm_l, loc=<7.2135 -0.1134 0>, rot=<0 0 0 1>}
 deform_upperarm_bk_base_l {parent=upperarm_twist_01_l, loc=<-4.5254 -4.2759 -1.3968>, rot=<0 0.0007 0.0191 -0.9998>}
 deform_upperarm_bk_l {parent=deform_upperarm_bk_base_l, loc=<0 0 0>, rot=<0 0 0 1>}
 deform_upperarm_up_base_l {parent=upperarm_twist_01_l, loc=<-8.1601 0.06 3.1574>, rot=<0 0.0007 0.0191 -0.9998>}
 deform_upperarm_up_l {parent=deform_upperarm_up_base_l, loc=<0 0 0>, rot=<0 0 0 1>}
 deform_upperarm_fr_base_l {parent=upperarm_twist_01_l, loc=<-5.0343 4.3504 -0.7751>, rot=<0 0.0007 0.0191 -0.9998>}
 deform_upperarm_fr_l {parent=deform_upperarm_fr_base_l, loc=<0 0 0>, rot=<0 0 0 1>}
 upperarm_twist_02_l {parent=upperarm_l, loc=<19.7998 -0.1134 0>, rot=<0 0 0 1>}
 dyn_coat_sleeves_mid_MSTR_L {parent=clavicle_l, loc=<13.6376 -4.2292 -9.2822>, rot=<-0.5049 -0.3574 0.4811 -0.6212>}
 dyn_coat_sleeves_mid_01_L {parent=dyn_coat_sleeves_mid_MSTR_L, loc=<1.9962 0 0>, rot=<-0.0567 0 0 -0.9984>}
 dyn_coat_sleeves_mid_02_L {parent=dyn_coat_sleeves_mid_01_L, loc=<11.5511 0 0.1203>, rot=<0 0.0192 0 -0.9998>}
 dyn_coat_sleeves_mid_03_L {parent=dyn_coat_sleeves_mid_02_L, loc=<17.5076 0 0.3014>, rot=<0 0.0083 0 -1>}
 dyn_coat_sleeves_mid_04_L {parent=dyn_coat_sleeves_mid_03_L, loc=<19.8911 0 0>, rot=<0 0 0 1>}
 dyn_coat_sleeves_mid_04_L_01 {parent=dyn_coat_sleeves_mid_03_L, loc=<10.2239 1.8802 1.5466>, rot=<0 0 0 1>}
 dyn_coat_sleeves_mid_04_L_02 {parent=dyn_coat_sleeves_mid_04_L_01, loc=<8.4078 -0.3083 0.0379>, rot=<0 0 0 1>}
 clavicle_r {parent=spine_05, loc=<5.4831 -0.1053 3.2946>, rot=<0.7289 0.0465 -0.6829 0.009>}
 upperarm_r {parent=clavicle_r, loc=<-9.8658 1.2136 0.4105>, rot=<0.0164 -0.3245 -0.0269 -0.9454>}
 lowerarm_r {parent=upperarm_r, loc=<-26.465 0.8345 -0.259>, rot=<0.0036 0.0161 -0.2298 -0.9731>}
 hand_r {parent=lowerarm_r, loc=<-22.7355 0.3203 0.6106>, rot=<-0.6784 -0.0223 0.0735 -0.7307>}
 thumb_01_r {parent=hand_r, loc=<-1.3118 0.3191 2.459>, rot=<0.6546 -0.4042 0.0664 -0.6354>}
 thumb_02_r {parent=thumb_01_r, loc=<-5.9336 0.0448 -0.0401>, rot=<-0.0037 0.0213 0.298 -0.9543>}
 thumb_03_r {parent=thumb_02_r, loc=<-2.2223 -0.3682 0.0907>, rot=<0.0118 -0.013 0.0066 -0.9998>}
 index_metacarpal_r {parent=hand_r, loc=<-1.6535 -0.0183 2.96>, rot=<0.0158 -0.0124 -0.0173 -0.9996>}
 index_01_r {parent=index_metacarpal_r, loc=<-8.3008 -0.0357 0.0524>, rot=<-0.0044 0.0485 0.0924 -0.9945>}
 index_02_r {parent=index_01_r, loc=<-3.3587 0.3004 -0.0452>, rot=<-0.002 0.0027 0.3225 -0.9466>}
 index_03_r {parent=index_02_r, loc=<-2.5578 0.123 0.0252>, rot=<-0.002 0.0067 0.2558 -0.9667>}
 middle_metacarpal_r {parent=hand_r, loc=<-2.1856 0.1725 0.5984>, rot=<0.006 0.0168 -0.0329 -0.9993>}
 middle_01_r {parent=middle_metacarpal_r, loc=<-7.3275 0.2447 -0.0129>, rot=<-0.0208 0.0042 0.0858 -0.9961>}
 middle_02_r {parent=middle_01_r, loc=<-4.0123 -0.1401 -0.1175>, rot=<-0.0218 0.0258 0.3512 -0.9357>}
 middle_03_r {parent=middle_02_r, loc=<-2.6077 -0.0534 -0.1373>, rot=<0.0082 0.0188 0.1708 -0.9851>}
 ring_metacarpal_r {parent=hand_r, loc=<-2.2074 0.3625 -1.3638>, rot=<-0.0284 0.0722 -0.0458 -0.9959>}
 ring_01_r {parent=ring_metacarpal_r, loc=<-7.0295 0.6197 0.1446>, rot=<-0.0489 -0.0252 0.1892 -0.9804>}
 ring_02_r {parent=ring_01_r, loc=<-3.6174 -0.4647 0.1027>, rot=<0 0 0.2806 -0.9598>}
 ring_03_r {parent=ring_02_r, loc=<-2.4152 -0.114 -0.0199>, rot=<0 0 0.1189 -0.9929>}
 pinky_metacarpal_r {parent=hand_r, loc=<-1.7616 0.9753 -3.3814>, rot=<-0.0397 0.0814 -0.069 -0.9935>}
 pinky_01_r {parent=pinky_metacarpal_r, loc=<-7.1709 0.4831 0.1094>, rot=<-0.0654 -0.0035 0.2242 -0.9723>}
 pinky_02_r {parent=pinky_01_r, loc=<-2.8274 -0.0893 0.1043>, rot=<0.0037 0.0011 0.2854 -0.9584>}
 pinky_03_r {parent=pinky_02_r, loc=<-2.2046 -0.0307 0.0078>, rot=<-0.0002 0.0098 0.1445 -0.9895>}
 lowerarm_twist_02_r {parent=lowerarm_r, loc=<-10.6707 0 0>, rot=<0 0 0 -1>}
 lowerarm_twist_01_r {parent=lowerarm_r, loc=<-18.316 0 0>, rot=<0 0 0 -1>}
 base_sleeveFront_r {parent=lowerarm_twist_01_r, loc=<2.1443 -2.4799 0.4938>, rot=<0.9964 -0.0601 -0.0269 -0.0538>}
 dyn_sleeveFront_r {parent=base_sleeveFront_r, loc=<0 0 0>, rot=<0 0 0 -1>}
 base_sleeveBack_r {parent=lowerarm_twist_01_r, loc=<2.2974 2.8285 1.1134>, rot=<0.0314 -0.0308 0.0923 -0.9948>}
 dyn_sleeveBack_r {parent=base_sleeveBack_r, loc=<0 0 0>, rot=<0 0 0 -1>}
 base_sleeveBottom_r {parent=lowerarm_twist_01_r, loc=<2.3439 -0.3268 2.5712>, rot=<-0.7552 0.0204 0.0848 -0.6497>}
 dyn_sleeveBottom_r {parent=base_sleeveBottom_r, loc=<0 0 0>, rot=<0 0 0 1>}
 base_sleeveTop_r {parent=lowerarm_twist_01_r, loc=<1.7712 0.6642 -2.3459>, rot=<0.6955 -0.0473 0.0332 -0.7162>}
 dyn_sleeveTop_r {parent=base_sleeveTop_r, loc=<0 0 0>, rot=<0 0 0 1>}
 deform_elbow_in_base_r {parent=lowerarm_r, loc=<-0.68 -3.3096 0.3773>, rot=<0.0285 -0.0335 0.0024 -0.999>}
 deform_elbow_in_r {parent=deform_elbow_in_base_r, loc=<-0.0001 0 0.0002>, rot=<0 0 0 1>}
 deform_elbow_out_base_r {parent=lowerarm_r, loc=<0.6633 2.7819 -0.3092>, rot=<0.0285 -0.0335 0.0024 -0.999>}
 deform_elbow_out_r {parent=deform_elbow_out_base_r, loc=<0 0 0>, rot=<0 0 0 1>}
 upperarm_twist_01_r {parent=upperarm_r, loc=<-7.2135 0.1134 0>, rot=<0 0 0 -1>}
 deform_upperarm_bk_base_r {parent=upperarm_twist_01_r, loc=<5.9995 3.9524 2.1443>, rot=<0 0.0007 0.0191 -0.9998>}
 deform_upperarm_bk_r {parent=deform_upperarm_bk_base_r, loc=<0 0 0>, rot=<0 0 0 1>}
 deform_upperarm_up_base_r {parent=upperarm_twist_01_r, loc=<8.5633 -0.0986 -2.9852>, rot=<0 0.0007 0.0191 -0.9998>}
 deform_upperarm_up_r {parent=deform_upperarm_up_base_r, loc=<0 0 0>, rot=<0 0 0 -1>}
 deform_upperarm_fr_base_r {parent=upperarm_twist_01_r, loc=<6.1083 -5.252 2.5241>, rot=<0 0.0007 0.0191 -0.9998>}
 deform_upperarm_fr_r {parent=deform_upperarm_fr_base_r, loc=<0 0 0>, rot=<0 0 0 1>}
 upperarm_twist_02_r {parent=upperarm_r, loc=<-19.7998 0.1134 0>, rot=<0 0 0 -1>}
 dyn_coat_sleeves_mid_MSTR_R {parent=clavicle_r, loc=<-13.6988 4.3876 9.3435>, rot=<-0.5104 -0.5865 -0.5279 0.3418>}
 dyn_coat_sleeves_mid_01_R {parent=dyn_coat_sleeves_mid_MSTR_R, loc=<1.9961 0 0>, rot=<0 0 0 1>}
 dyn_coat_sleeves_mid_02_R {parent=dyn_coat_sleeves_mid_01_R, loc=<11.5517 0 0.1203>, rot=<0 0.0192 0 -0.9998>}
 dyn_coat_sleeves_mid_03_R {parent=dyn_coat_sleeves_mid_02_R, loc=<17.5076 0 0.3014>, rot=<0 0.0083 0 -1>}
 dyn_coat_sleeves_mid_04_R {parent=dyn_coat_sleeves_mid_03_R, loc=<19.8911 0 0>, rot=<0 0 0 1>}
 dyn_coat_sleeves_mid_04_R_01 {parent=dyn_coat_sleeves_mid_03_R, loc=<10.0132 -3.3245 1.2887>, rot=<0 0 0 1>}
 dyn_coat_sleeves_mid_04_R_02 {parent=dyn_coat_sleeves_mid_04_R_01, loc=<8.4135 0 0>, rot=<0 0 0 1>}
 attach_backpack {parent=spine_05, loc=<-0.0172 -9.4345 0>, rot=<-0.0093 -0.707 0.0093 -0.707>}
 Backpack_BR {parent=attach_backpack, type=Null, loc=<0 1.6957 -1.4887>, rot=<-0.0569 0 0 0.9984>}
 attach_cape {parent=spine_05, loc=<8.645 -9.6043 0>, rot=<0.0616 -0.7044 -0.0616 -0.7044>}
 cape {parent=attach_cape, type=Null, loc=<0 0.3873 -0.6664>, rot=<0.098 0 0 0.9952>}
 dyn_coat_collar_MSTR_fL {parent=spine_05, loc=<-0.3038 5.6552 -10.3592>, rot=<0.1211 -0.9868 -0.107 0.0116>}
 dyn_coat_collar_01_fL {parent=dyn_coat_collar_MSTR_fL, loc=<1 0 0>, rot=<0 0 0 1>}
 dyn_coat_collar_MSTR_fR {parent=spine_05, loc=<-0.3038 5.6552 10.3592>, rot=<-0.0098 0.1024 0.9874 -0.12>}
 dyn_coat_collar_01_fR {parent=dyn_coat_collar_MSTR_fR, loc=<1 0 0>, rot=<0 0 0 1>}
 dyn_coat_collar01_MSTR_fL {parent=spine_05, loc=<5.629 5.0638 -8.2803>, rot=<-0.1687 -0.9855 0.0198 0.0003>}
 dyn_coat_collar01_01_fL {parent=dyn_coat_collar01_MSTR_fL, loc=<1 0 0>, rot=<0 0 0 1>}
 dyn_coat_collar01_MSTR_fR {parent=spine_05, loc=<5.6293 5.0639 8.2803>, rot=<-0.0006 -0.0197 0.9855 0.1687>}
 dyn_coat_collar01_01_fR {parent=dyn_coat_collar01_MSTR_fR, loc=<1.0006 0.0002 0.0007>, rot=<0 0 0 1>}
 dyn_coat_button_MSTR_L {parent=spine_05, loc=<1.3109 -7.5835 -8.163>, rot=<0.6607 0.0309 -0.0259 -0.7496>}
 dyn_coat_button_MSTR_R {parent=spine_05, loc=<1.3114 -7.5834 8.163>, rot=<0.0309 -0.6607 0.7496 -0.0259>}
 dyn_rope_MSTR_fR {parent=spine_05, loc=<6.9343 3.9873 2.8204>, rot=<0.1053 -0.041 0.9191 0.3775>}
 dyn_rope_fR_01 {parent=dyn_rope_MSTR_fR, loc=<1.9236 0.0002 0.0007>, rot=<0 0 0 1>}
 dyn_rope_fR_02 {parent=dyn_rope_fR_01, loc=<6.5783 0 0>, rot=<0 0 0.0313 -0.9995>}
 dyn_coat_bk_MSTR {parent=spine_04, loc=<2.2656 -8.492 0.0747>, rot=<0.0175 -0.6982 0.7154 -0.0195>}
 dyn_coat_bk_01 {parent=dyn_coat_bk_MSTR, loc=<1 0 0>, rot=<0 0 0 1>}
 dyn_coat_bk_02 {parent=dyn_coat_bk_01, loc=<8.954 0 0>, rot=<0 0.1042 0 -0.9946>}
 dyn_coat_bk_03 {parent=dyn_coat_bk_02, loc=<14.4011 0 0>, rot=<0 0.067 0 -0.9978>}
 dyn_coat_bk_04 {parent=dyn_coat_bk_03, loc=<16.7654 0 0>, rot=<0 0 0 1>}
 dyn_coat_bk_04_L_02 {parent=dyn_coat_bk_04, loc=<-0.7027 5.4648 -2.0696>, rot=<0 0 -0.0605 -0.9982>}
 dyn_coat_bk_04_R_02 {parent=dyn_coat_bk_04, loc=<-0.6885 -5.3614 -2.0696>, rot=<0 0 0.0542 -0.9985>}
 dyn_coat_bk_04_L_01 {parent=dyn_coat_bk_03, loc=<0 4.3 0>, rot=<0 0 -0.0605 -0.9982>}
 dyn_coat_bk_04_R_01 {parent=dyn_coat_bk_03, loc=<0 -4.3 0>, rot=<0 0 0.0542 -0.9985>}
 dyn_coat_MSTR_fR {parent=spine_04, loc=<-0.8036 1.8267 12.8394>, rot=<-0.01 0.1024 0.9872 -0.1215>}
 dyn_coat_01_fR {parent=dyn_coat_MSTR_fR, loc=<1 0 0>, rot=<0 0 0 1>}
 dyn_coat_02_fR {parent=dyn_coat_01_fR, loc=<8.0406 0 0>, rot=<0 0.1042 0 -0.9946>}
 dyn_coat_03_fR {parent=dyn_coat_02_fR, loc=<14.4011 0 0>, rot=<0 0.067 0 -0.9978>}
 dyn_coat_04_fR {parent=dyn_coat_03_fR, loc=<16.7654 0 0>, rot=<0 0 0 1>}
 dyn_coat_MSTR_fL {parent=spine_04, loc=<-0.8036 1.8267 -12.8394>, rot=<0.1225 -0.9866 -0.107 0.0118>}
 dyn_coat_01_fL {parent=dyn_coat_MSTR_fL, loc=<1 0 0>, rot=<0 0 0 1>}
 dyn_coat_02_fL {parent=dyn_coat_01_fL, loc=<8.0399 0 0>, rot=<0 0.1042 0 -0.9946>}
 dyn_coat_03_fL {parent=dyn_coat_02_fL, loc=<14.4011 0 0>, rot=<0 0.067 0 -0.9978>}
 dyn_coat_04_fL {parent=dyn_coat_03_fL, loc=<16.7654 0 0>, rot=<0 0 0 1>}
 dyn_coat_MSTR_L {parent=spine_04, loc=<0.4242 -7.3939 -8.6116>, rot=<0.0428 -0.9963 0.071 0.0246>}
 dyn_coat_01_L {parent=dyn_coat_MSTR_L, loc=<1.1297 0.0919 0.1265>, rot=<-0.5179 0.0266 -0.0634 -0.8527>}
 dyn_coat_02_L {parent=dyn_coat_01_L, loc=<8.0399 0 0>, rot=<0 0.1042 0 -0.9946>}
 dyn_coat_03_L {parent=dyn_coat_02_L, loc=<14.4011 0 0>, rot=<0 0.067 0 -0.9978>}
 dyn_coat_04_L {parent=dyn_coat_03_L, loc=<16.7654 0 0>, rot=<0 0 0 1>}
 dyn_coat_MSTR_R {parent=spine_04, loc=<0.4242 -7.3939 8.6116>, rot=<0.0018 -0.5575 0.8301 -0.0114>}
 dyn_coat_01_R {parent=dyn_coat_MSTR_R, loc=<1 0 0>, rot=<0 0 0 1>}
 dyn_coat_02_R {parent=dyn_coat_01_R, loc=<8.0406 0 0>, rot=<0 0.1362 0 -0.9907>}
 dyn_coat_03_R {parent=dyn_coat_02_R, loc=<13.7413 0 0>, rot=<0 0 0 1>}
 dyn_coat_04_R {parent=dyn_coat_03_R, loc=<16.3136 0 1.6384>, rot=<0 0.05 0 -0.9987>}
 dyn_coat_ribbon_MSTR_R {parent=spine_04, loc=<7.0884 -8.2349 8.039>, rot=<0.0898 -0.6566 0.7448 -0.0781>}
 dyn_coat_ribbon_01_R {parent=dyn_coat_ribbon_MSTR_R, loc=<1 0 0>, rot=<0 0 0 1>}
 dyn_coat_ribbon_02_R {parent=dyn_coat_ribbon_01_R, loc=<7.0628 0.0194 -0.1423>, rot=<-0.0021 0.0629 0.0097 -0.998>}
 dyn_coat_ribbon_03_R {parent=dyn_coat_ribbon_02_R, loc=<10.7737 0 0.1761>, rot=<0 -0.0296 0 -0.9996>}
 dyn_coat_ribbon_04_R {parent=dyn_coat_ribbon_03_R, loc=<13.3555 0 1.3413>, rot=<0 0.05 0 -0.9987>}
 dyn_upperTorso_l_base {parent=spine_04, loc=<6.6531 5.824 -5.67>, rot=<-0.4316 -0.5601 -0.4767 -0.5223>}
 dyn_upperTorso_l {parent=dyn_upperTorso_l_base, loc=<0 0 0>, rot=<0 0 0 1>}
 dyn_upperTorso_r_base {parent=spine_04, loc=<6.6531 5.824 5.67>, rot=<0.501 0.499 0.5408 0.4556>}
 dyn_upperTorso_r {parent=dyn_upperTorso_r_base, loc=<0 0 0>, rot=<0 0 0 1>}
 dyn_spinekerchief_MSTR_fL {parent=spine_02, loc=<5.0265 10.21 -6.1398>, rot=<0.0566 -0.8575 -0.5113 0.0001>}
 dyn_spinekerchief_01_fL {parent=dyn_spinekerchief_MSTR_fL, loc=<2.3772 0 0>, rot=<0 0 0 1>}
 dyn_spinekerchief_02_fL {parent=dyn_spinekerchief_01_fL, loc=<8.1921 0 -0.0042>, rot=<0 -0.0137 0 -0.9999>}
 thigh_l {parent=pelvis, loc=<-1.0855 -0.2109 -8.7053>, rot=<0.0009 0.0355 0.0155 -0.9993>}
 calf_l {parent=thigh_l, loc=<-42.5284 -0.0019 -0.7297>, rot=<0.0006 -0.0161 -0.0391 -0.9991>}
 foot_l {parent=calf_l, loc=<-46.7401 -2.5014 -0.592>, rot=<0.0427 -0.0238 0.0353 -0.9982>}
 ball_l {parent=foot_l, loc=<-3.5658 10.8066 -0.1544>, rot=<0.0001 -0.0001 -0.7063 -0.7079>}
 calf_twist_02_l {parent=calf_l, loc=<-19.1546 -1.7943 -0.2638>, rot=<0.0237 0.0095 -0.0035 -0.9997>}
 calf_twist_01_l {parent=calf_l, loc=<-35.8483 -1.9197 -0.5794>, rot=<0.0237 0.0095 -0.0035 -0.9997>}
 deform_knee_in_base_l {parent=calf_l, loc=<-0.434 -5.1646 0.3336>, rot=<0.0017 0.0216 0.0001 -0.9998>}
 deform_knee_in_l {parent=deform_knee_in_base_l, loc=<0 0 0>, rot=<0 0 0 1>}
 deform_knee_out_base_l {parent=calf_l, loc=<-0.2933 4.012 0.2393>, rot=<0.0017 0.0216 0.0001 -0.9998>}
 deform_knee_out_l {parent=deform_knee_out_base_l, loc=<0 0 0>, rot=<0 0 0 1>}
 calf_l_ctrl {controlType=EULER_TRANSFORM, parent=calf_l, type=Control, loc=<0 0 0>, rot=<0 0 0 1>}
 thigh_twist_01_l {parent=thigh_l, loc=<-27.952 0 0>, rot=<0.0013 0 0 -1>}
 deform_groin_base_l {parent=thigh_l, loc=<0.8494 4.7477 -1.093>, rot=<0 0.0217 0 -0.9998>}
 deform_groin_l {parent=deform_groin_base_l, loc=<0 0 0>, rot=<0 0 0 1>}
 thigh_l_ctrl {controlType=EULER_TRANSFORM, parent=thigh_l, type=Control, loc=<0 0 0>, rot=<0 0 0 1>}
 thigh_r {parent=pelvis, loc=<-1.0855 -0.2109 8.7053>, rot=<0.0355 -0.0009 0.9993 0.0155>}
 calf_r {parent=thigh_r, loc=<42.5284 0.0019 0.7297>, rot=<0.0006 -0.0161 -0.0391 -0.9991>}
 foot_r {parent=calf_r, loc=<46.7401 2.5014 0.592>, rot=<0.0427 -0.0238 0.0353 -0.9982>}
 ball_r {parent=foot_r, loc=<3.5658 -10.8066 0.1544>, rot=<0.0001 -0.0001 -0.7063 -0.7079>}
 calf_twist_02_r {parent=calf_r, loc=<19.1546 1.7943 0.2638>, rot=<0.0237 0.0095 -0.0035 -0.9997>}
 calf_twist_01_r {parent=calf_r, loc=<35.8483 1.9197 0.5794>, rot=<0.0237 0.0095 -0.0035 -0.9997>}
 deform_knee_in_base_r {parent=calf_r, loc=<0.4339 5.1646 -0.3336>, rot=<0.0017 0.0216 0.0001 -0.9998>}
 deform_knee_in_r {parent=deform_knee_in_base_r, loc=<0 0 0>, rot=<0 0 0 1>}
 deform_knee_out_base_r {parent=calf_r, loc=<0.2933 -4.012 -0.2393>, rot=<0.0017 0.0216 0.0001 -0.9998>}
 deform_knee_out_r {parent=deform_knee_out_base_r, loc=<0 0 0>, rot=<0 0 0 1>}
 calf_r_ctrl {controlType=EULER_TRANSFORM, parent=calf_r, type=Control, loc=<0 0 0>, rot=<0 0 0 1>}
 dyn_knee_zipper_r {parent=calf_r, loc=<2.0752 -5.0206 -0.6476>, rot=<0.0017 0.0216 0.0001 -0.9998>}
 dyn_knee_R_boots_01_base {parent=calf_r, loc=<-1.4496 0.2673 -6.2312>, rot=<0.0017 0.0216 0.0001 -0.9998>}
 dyn_knee_R_boots_01 {parent=dyn_knee_R_boots_01_base, loc=<0.6743 0 0>, rot=<0 0 0 1>}
 dyn_knee_R_boots_02_base {parent=calf_r, loc=<-1.9208 0.1619 4.6447>, rot=<0.0017 0.0216 0.0001 -0.9998>}
 dyn_knee_R_boots_02 {parent=dyn_knee_R_boots_02_base, loc=<0.6743 0 0>, rot=<0 0 0 1>}
 thigh_twist_01_r {parent=thigh_r, loc=<27.952 0 0>, rot=<0.0013 0 0 -1>}
 deform_groin_base_r {parent=thigh_r, loc=<-0.8494 -4.7477 1.093>, rot=<0 0.0217 0 -0.9998>}
 deform_groin_r {parent=deform_groin_base_r, loc=<0 0 0>, rot=<0 0 0 1>}
 thigh_r_ctrl {controlType=EULER_TRANSFORM, parent=thigh_r, type=Control, loc=<0 0 0>, rot=<0 0 0 1>}
 deform_glute_base_l {parent=pelvis, loc=<-5.0422 -7.3482 -8.4287>, rot=<0.0275 0.0501 -0.4472 -0.8926>}
 deform_glute_l {parent=deform_glute_base_l, loc=<-0.0131 -0.0327 0>, rot=<0 0 0 1>}
 deform_glute_base_r {parent=pelvis, loc=<-5.042 -7.3482 8.4287>, rot=<0.0501 -0.0275 0.8926 -0.4472>}
 deform_glute_r {parent=deform_glute_base_r, loc=<0.0131 0.0327 0>, rot=<0 0 0 1>}
]
variables [
 DBNO {type=bool, default=False}
 DBNO_Carry {type=bool, default=False}
 isBaseballSliding {type=bool, default=False}
 isSliding {type=bool, default=False}
 spr_thigh_l_f_alpha {type=double, default=0}
 spr_thigh_r_f_alpha {type=double, default=0}
 spr_upperarm_l_bk_alpha {type=double, default=0}
 spr_upperarm_l_dn_alpha {type=double, default=0}
 spr_upperarm_l_f_alpha {type=double, default=0}
 spr_upperarm_l_up_alpha {type=double, default=0}
 spr_upperarm_r_bk_alpha {type=double, default=0}
 spr_upperarm_r_dn_alpha {type=double, default=0}
 spr_upperarm_r_f_alpha {type=double, default=0}
 spr_upperarm_r_up_alpha {type=double, default=0}
]
