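# Physics asset for the F_MED_TweakSmirk player body cosmetic, covering its dangling coat panels, sleeves, right-side ribbon and left-front kerchief.
# All bodies are single capsules, on those dyn_ bones plus pelvis, spine, arms and thighs.
# Sleeve, ribbon and kerchief chains use limited joints below tiny MSTR root capsules, which attach with free rotation to spine_05 (sleeves) or spine_04.
# Coat panel joints are angularly free, with each panel chain rooted on spine_04.
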PhysicsAsset /BRCosmetics/Characters/Player/Female/Medium/Bodies/F_MED_TweakSmirk/Meshes/F_MED_TweakSmirk_Physics
bodies {
 dyn_coat_02_fR [{
   shapeType=Capsule
   length=21.4308
   radius=16.0888
   shapeName=dyn_coat_02_fR_capsule
   center=<14.5759 6.2732 10.6003>
   rotation=<rot 0 -81.7302 -89.4366>
  }]
 dyn_coat_03_L [{
   shapeType=Capsule
   length=13.3623
   radius=9.5494
   shapeName=dyn_coat_03_L_capsule
   center=<8.7616 -1.11 -5.6642>
   rotation=<rot 0.0225 0.8685 -89.9417>
  }]
 dyn_coat_03_R [{
   shapeType=Capsule
   length=13.3623
   radius=9.5494
   shapeName=dyn_coat_03_L_capsule
   center=<8.7616 -1.11 -5.6642>
   rotation=<rot 0.019 4.2793 -89.9404>
  }]
 dyn_coat_03_fL [{
   shapeType=Capsule
   length=11.4614
   radius=7.8636
   shapeName=dyn_coat_03_fL_capsule
   center=<8.8353 1.5298 -5.2989>
   rotation=<rot 11.1757 166.1636 -47.2426>
  }]
 dyn_coat_03_fR [{
   shapeType=Capsule
   length=11.4614
   radius=7.8636
   shapeName=dyn_coat_03_fL_capsule
   center=<8.295 -0.4588 -5.547>
   rotation=<rot 11.1757 166.1636 -118.7618>
  }]
 dyn_coat_04_L [{
   shapeType=Capsule
   length=12.9777
   radius=10.47
   shapeName=dyn_coat_04_L_capsule
   center=<5.0749 -0.0215 -9.6781>
   rotation=<rot 0 -5.8618 -86.9103>
  }]
 dyn_coat_04_R [{
   shapeType=Capsule
   length=12.9777
   radius=10.47
   shapeName=dyn_coat_04_L_capsule
   center=<5.0749 -0.0215 -9.6781>
   rotation=<rot -0.2284 -1.6263 -86.9187>
  }]
 dyn_coat_04_fL [{
   shapeType=Capsule
   length=11.5354
   radius=9.4116
   shapeName=dyn_coat_04_fL_capsule
   center=<4.9088 2.7293 -8.2981>
   rotation=<rot 12.8429 178.4453 -47.7757>
  }]
 dyn_coat_04_fR [{
   shapeType=Capsule
   length=11.5354
   radius=9.4116
   shapeName=dyn_coat_04_fL_capsule
   center=<4.7373 -0.5453 -8.6603>
   rotation=<rot 12.8429 178.4453 -119.295>
  }]
 dyn_coat_bk_02 [{
   shapeType=Capsule
   length=8.0744
   radius=10.075
   shapeName=dyn_coat_bk_03_capsule
   center=<8.7466 0.0655 -9.0959>
   rotation=<rot 3.1517 179.0348 -90.0683>
  }]
 dyn_coat_bk_03 [{
   shapeType=Capsule
   length=13.0744
   radius=10.075
   shapeName=dyn_coat_bk_03_capsule
   center=<5.452 0.1248 -5.6983>
   rotation=<rot 3.1517 179.0348 -90.0683>
  }]
 dyn_coat_bk_04 [{
   shapeType=Capsule
   length=16.5081
   radius=10.7127
   shapeName=dyn_coat_bk_04_capsule
   center=<6.3148 -0.0061 -10.1498>
   rotation=<rot 0 0.3514 -88.6356>
  }]
 dyn_coat_ribbon_01_R [{
   shapeType=Capsule
   length=7.5
   radius=1.5
   shapeName=dyn_coat_ribbon_01_R_capsule
   center=<5.4852 0.6464 1.5091>
   rotation=<rot -1.6967 87.1603 -85.4559>
  }]
 dyn_coat_ribbon_02_R [{
   shapeType=Capsule
   length=12
   radius=1.5
   shapeName=dyn_coat_ribbon_02_R_capsule
   center=<7.5384 0.0989 1.3774>
   rotation=<rot 0 89.8595 -88.9253>
  }]
 dyn_coat_ribbon_03_R [{
   shapeType=Capsule
   length=12
   radius=1.5
   shapeName=dyn_coat_ribbon_03_R_capsule
   center=<6.3154 -0.01 2.0714>
   rotation=<rot 0 89.9533 -82.0845>
  }]
 dyn_coat_ribbon_04_R [{
   shapeType=Capsule
   length=12
   radius=1.5
   shapeName=dyn_coat_ribbon_04_R_capsule
   center=<5.2743 -0.213 1.4606>
   rotation=<rot -0.0177 88.4295 -92.4379>
  }]
 dyn_coat_ribbon_MSTR_R [{shapeType=Capsule, length=0.505, radius=0.505, shapeName=dyn_coat_ribbon_MSTR_R_capsule, rotation=<rot 0 -90 0>}]
 dyn_coat_sleeves_mid_01_L [{
   shapeType=Capsule
   length=1.4694
   radius=11
   shapeName=dyn_coat_sleeves_mid_01_R_capsule
   center=<6.9407 0 11.9165>
   rotation=<rot -89.3584 -179.9964 179.9964>
  }]
 dyn_coat_sleeves_mid_01_R [{
   shapeType=Capsule
   length=1.4694
   radius=11
   shapeName=dyn_coat_sleeves_mid_01_R_capsule
   center=<0.1316 0 11.7486>
   rotation=<rot -89.3584 -179.9964 179.9964>
  }]
 dyn_coat_sleeves_mid_02_L [{
   shapeType=Capsule
   length=10
   radius=6.5
   shapeName=dyn_coat_sleeves_mid_02_R_capsule
   center=<7.2166 0 6.0708>
   rotation=<rot -89.3584 -179.9965 179.9965>
  }]
 dyn_coat_sleeves_mid_02_R [{
   shapeType=Capsule
   length=9.5
   radius=7
   shapeName=dyn_coat_sleeves_mid_02_R_capsule
   center=<0.062 0 5.5329>
   rotation=<rot -89.3584 -179.9965 179.9965>
  }]
 dyn_coat_sleeves_mid_03_L [{
   shapeType=Capsule
   length=10.8707
   radius=7.5
   shapeName=dyn_coat_sleeves_mid_03_R_capsule
   center=<8.5165 0 7.3673>
   rotation=<rot -89.3584 -179.9966 179.9966>
  }]
 dyn_coat_sleeves_mid_03_R [{
   shapeType=Capsule
   length=11.3707
   radius=7
   shapeName=dyn_coat_sleeves_mid_03_R_capsule
   center=<0.0671 0 5.9892>
   rotation=<rot -89.3584 -179.9966 179.9966>
  }]
 dyn_coat_sleeves_mid_04_L [{
   shapeType=Capsule
   length=2.5
   radius=8
   shapeName=dyn_coat_sleeves_mid_04_R_capsule
   center=<2.7539 0 7.3>
   rotation=<rot -89.3584 -179.9966 179.9966>
  }]
 dyn_coat_sleeves_mid_04_L_01 [{
   shapeType=Capsule
   length=3.505
   radius=2.5
   shapeName=dyn_coat_sleeves_mid_04_L_01_capsule
   center=<3.2044 0 -0.6248>
   rotation=<rot -78.9671 180 -180>
  }]
 dyn_coat_sleeves_mid_04_L_02 [{
   shapeType=Capsule
   length=3.7937
   radius=2.5
   shapeName=dyn_coat_sleeves_mid_04_L_02_capsule
   center=<3.2044 0 -0.6248>
   rotation=<rot -78.9671 180 -180>
  }]
 dyn_coat_sleeves_mid_04_R [{
   shapeType=Capsule
   length=6
   radius=11
   shapeName=dyn_coat_sleeves_mid_04_R_capsule
   center=<0.0671 -2.155 5.9892>
   rotation=<rot -89.3584 -179.9966 179.9966>
  }]
 dyn_coat_sleeves_mid_04_R_01 [{
   shapeType=Capsule
   length=3.505
   radius=2.5
   shapeName=dyn_coat_sleeves_mid_04_L_01_capsule
   center=<3.1634 0 0.1782>
   rotation=<rot -87.0826 0 0>
  }]
 dyn_coat_sleeves_mid_04_R_02 [{
   shapeType=Capsule
   length=3.7937
   radius=2.5
   shapeName=dyn_coat_sleeves_mid_04_L_02_capsule
   center=<3.1743 0 0.2111>
   rotation=<rot -87.9391 0 0>
  }]
 dyn_coat_sleeves_mid_MSTR_L [{shapeType=Capsule, length=0.505, radius=0.505, shapeName=dyn_coat_sleeves_mid_MSTR_L_capsule, rotation=<rot 0 -90 0>}]
 dyn_coat_sleeves_mid_MSTR_R [{shapeType=Capsule, length=0.505, radius=0.505, shapeName=dyn_coat_sleeves_mid_MSTR_R_capsule, rotation=<rot 0 -90 0>}]
 dyn_spinekerchief_01_fL [{
   shapeType=Capsule
   length=6.0026
   radius=5.2507
   shapeName=dyn_spinekerchief_01_fL_capsule
   center=<4.2459 0.3206 3.1611>
   rotation=<rot 0 85.6117 -84.4147>
  }]
 dyn_spinekerchief_02_fL [{
   shapeType=Capsule
   length=4.0522
   radius=3.3128
   shapeName=dyn_spinekerchief_02_fL_capsule
   center=<4.9038 0.3626 1.9006>
   rotation=<rot -0.0907 -89.8608 -89.3328>
  }]
 dyn_spinekerchief_MSTR_fL [{shapeType=Capsule, length=0.505, radius=0.505, shapeName=dyn_spinekerchief_MSTR_fL_capsule, rotation=<rot 0 -90 0>}]
 lowerarm_l [{shapeType=Capsule, length=24.4066, radius=8.4285, center=<15.9433 1.7255 0.0931>, rotation=<rot 0 89.5914 -86.4792>}]
 lowerarm_r [{shapeType=Capsule, length=29.0989, radius=8.0998, center=<-14.7532 -1.2494 -0.092>, rotation=<rot 0 88.7368 -91.2727>}]
 pelvis [{shapeType=Capsule, length=10.8253, radius=8.8183, shapeName=pelvis_capsule}]
 spine_02 [{shapeType=Capsule, length=12.6242, radius=13.452, shapeName=spine_04_capsule, center=<-5.0689 0.1484 -0.4062>}]
 spine_03 [{shapeType=Capsule, length=12.6242, radius=13.452, shapeName=spine_04_capsule, center=<-2.6753 2.4225 -0.4062>}]
 spine_04 [{shapeType=Capsule, length=20.1242, radius=13.452, shapeName=spine_04_capsule, center=<-2.2167 6.0854 -0.4062>}]
 spine_05 [{shapeType=Capsule, length=11.262, radius=10.2669, shapeName=spine_05_capsule, center=<4.1248 3.2588 -0.3978>}]
 thigh_l [{
   shapeType=Capsule
   length=20.7206
   radius=10.5299
   shapeName=thigh_l_capsule
   center=<-24.1144 -0.3407 0.4813>
   rotation=<rot 0 -86.5888 -91.0201>
  }]
 thigh_r [{
   shapeType=Capsule
   length=20.7495
   radius=7.9963
   shapeName=thigh_r_capsule
   center=<23.9365 0.0796 0.6477>
   rotation=<rot 0 -91.6432 -85.5049>
  }]
 upperarm_l [{
   shapeType=Capsule
   length=14.9438
   radius=8
   shapeName=upperarm_l_capsule
   center=<14.9359 0.6618 -1.045>
   rotation=<rot -54.8976 88.0613 -84.0489>
  }]
 upperarm_r [{shapeType=Capsule, length=16.5654, radius=8.5, center=<-12.0211 0 -0.5876>, rotation=<rot 82.2635 -179.9992 -177.3773>}]
}
constraints [
 {bone1Name=lowerarm_l, bone2Name=upperarm_l, swing1LimitDegrees=45, swing2LimitDegrees=45, twistLimitDegrees=45}
 {bone1Name=lowerarm_r, bone2Name=upperarm_r, swing1LimitDegrees=45, swing2LimitDegrees=45, twistLimitDegrees=45}
 {
  bone1Name=upperarm_r
  bone2Name=spine_05
  swing1LimitDegrees=45
  swing1Motion=Free
  swing2LimitDegrees=45
  swing2Motion=Free
  twistLimitDegrees=45
  twistMotion=Free
 }
 {bone1Name=dyn_coat_sleeves_mid_04_R, bone2Name=dyn_coat_sleeves_mid_03_R, swing1LimitDegrees=12, swing2LimitDegrees=45, twistLimitDegrees=5}
 {bone1Name=dyn_coat_sleeves_mid_03_R, bone2Name=dyn_coat_sleeves_mid_02_R, swing1LimitDegrees=12, swing2LimitDegrees=45, twistLimitDegrees=5}
 {bone1Name=dyn_coat_sleeves_mid_02_R, bone2Name=dyn_coat_sleeves_mid_01_R, swing1LimitDegrees=12, swing2LimitDegrees=45, twistLimitDegrees=5}
 {
  bone1Name=dyn_coat_sleeves_mid_01_R
  bone2Name=dyn_coat_sleeves_mid_MSTR_R
  swing1LimitDegrees=21
  swing2LimitDegrees=45
  twistLimitDegrees=5
 }
 {
  bone1Name=dyn_coat_sleeves_mid_MSTR_R
  bone2Name=spine_05
  swing1LimitDegrees=45
  swing1Motion=Free
  swing2LimitDegrees=45
  swing2Motion=Free
  twistLimitDegrees=45
  twistMotion=Free
 }
 {
  bone1Name=upperarm_l
  bone2Name=spine_05
  swing1LimitDegrees=45
  swing1Motion=Free
  swing2LimitDegrees=45
  swing2Motion=Free
  twistLimitDegrees=45
  twistMotion=Free
 }
 {bone1Name=dyn_coat_sleeves_mid_04_L, bone2Name=dyn_coat_sleeves_mid_03_L, swing1LimitDegrees=10, swing2LimitDegrees=45, twistLimitDegrees=5}
 {bone1Name=dyn_coat_sleeves_mid_03_L, bone2Name=dyn_coat_sleeves_mid_02_L, swing1LimitDegrees=10, swing2LimitDegrees=45, twistLimitDegrees=5}
 {bone1Name=dyn_coat_sleeves_mid_02_L, bone2Name=dyn_coat_sleeves_mid_01_L, swing1LimitDegrees=10, swing2LimitDegrees=45, twistLimitDegrees=5}
 {bone1Name=dyn_coat_sleeves_mid_01_L, bone2Name=dyn_coat_sleeves_mid_MSTR_L, swing1LimitDegrees=5, swing2LimitDegrees=45, twistLimitDegrees=5}
 {
  bone1Name=dyn_coat_sleeves_mid_MSTR_L
  bone2Name=spine_05
  swing1LimitDegrees=45
  swing1Motion=Free
  swing2LimitDegrees=45
  swing2Motion=Free
  twistLimitDegrees=45
  twistMotion=Free
 }
 {
  bone1Name=dyn_coat_bk_04
  bone2Name=dyn_coat_bk_03
  swing1LimitDegrees=45
  swing1Motion=Free
  swing2LimitDegrees=45
  swing2Motion=Free
  twistLimitDegrees=45
  twistMotion=Free
 }
 {
  bone1Name=dyn_coat_bk_03
  bone2Name=dyn_coat_bk_02
  swing1LimitDegrees=45
  swing1Motion=Free
  swing2LimitDegrees=45
  swing2Motion=Free
  twistLimitDegrees=45
  twistMotion=Free
 }
 {
  bone1Name=dyn_coat_04_fR
  bone2Name=dyn_coat_03_fR
  swing1LimitDegrees=45
  swing1Motion=Free
  swing2LimitDegrees=45
  swing2Motion=Free
  twistLimitDegrees=45
  twistMotion=Free
 }
 {
  bone1Name=dyn_coat_03_fR
  bone2Name=dyn_coat_02_fR
  swing1LimitDegrees=45
  swing1Motion=Free
  swing2LimitDegrees=45
  swing2Motion=Free
  twistLimitDegrees=45
  twistMotion=Free
 }
 {
  bone1Name=dyn_coat_04_fL
  bone2Name=dyn_coat_03_fL
  swing1LimitDegrees=45
  swing1Motion=Free
  swing2LimitDegrees=45
  swing2Motion=Free
  twistLimitDegrees=45
  twistMotion=Free
 }
 {
  bone1Name=dyn_coat_03_fL
  bone2Name=spine_04
  swing1LimitDegrees=45
  swing1Motion=Free
  swing2LimitDegrees=45
  swing2Motion=Free
  twistLimitDegrees=45
  twistMotion=Free
 }
 {
  bone1Name=dyn_coat_04_R
  bone2Name=dyn_coat_03_R
  swing1LimitDegrees=45
  swing1Motion=Free
  swing2LimitDegrees=45
  swing2Motion=Free
  twistLimitDegrees=45
  twistMotion=Free
 }
 {
  bone1Name=dyn_coat_03_R
  bone2Name=spine_04
  swing1LimitDegrees=45
  swing1Motion=Free
  swing2LimitDegrees=45
  swing2Motion=Free
  twistLimitDegrees=45
  twistMotion=Free
 }
 {
  bone1Name=dyn_coat_04_L
  bone2Name=dyn_coat_03_L
  swing1LimitDegrees=45
  swing1Motion=Free
  swing2LimitDegrees=45
  swing2Motion=Free
  twistLimitDegrees=45
  twistMotion=Free
 }
 {
  bone1Name=dyn_coat_03_L
  bone2Name=spine_04
  swing1LimitDegrees=45
  swing1Motion=Free
  swing2LimitDegrees=45
  swing2Motion=Free
  twistLimitDegrees=45
  twistMotion=Free
 }
 {
  bone1Name=spine_05
  bone2Name=spine_04
  swing1LimitDegrees=45
  swing1Motion=Free
  swing2LimitDegrees=45
  swing2Motion=Free
  twistLimitDegrees=45
  twistMotion=Free
 }
 {
  bone1Name=spine_04
  bone2Name=spine_03
  swing1LimitDegrees=45
  swing1Motion=Free
  swing2LimitDegrees=45
  swing2Motion=Free
  twistLimitDegrees=45
  twistMotion=Free
 }
 {
  bone1Name=thigh_l
  bone2Name=pelvis
  swing1LimitDegrees=45
  swing1Motion=Free
  swing2LimitDegrees=45
  swing2Motion=Free
  twistLimitDegrees=45
  twistMotion=Free
 }
 {
  bone1Name=thigh_r
  bone2Name=pelvis
  swing1LimitDegrees=45
  swing1Motion=Free
  swing2LimitDegrees=45
  swing2Motion=Free
  twistLimitDegrees=45
  twistMotion=Free
 }
 {
  bone1Name=dyn_coat_sleeves_mid_04_L_01
  bone2Name=dyn_coat_sleeves_mid_03_L
  swing1LimitDegrees=45
  swing2LimitDegrees=5
  twistLimitDegrees=0
 }
 {
  bone1Name=dyn_coat_sleeves_mid_04_L_02
  bone2Name=dyn_coat_sleeves_mid_04_L_01
  swing1LimitDegrees=45
  swing2LimitDegrees=5
  twistLimitDegrees=0
 }
 {
  bone1Name=dyn_coat_sleeves_mid_04_R_01
  bone2Name=dyn_coat_sleeves_mid_03_R
  swing1LimitDegrees=45
  swing2LimitDegrees=5
  twistLimitDegrees=0
 }
 {
  bone1Name=dyn_coat_sleeves_mid_04_R_02
  bone2Name=dyn_coat_sleeves_mid_04_R_01
  swing1LimitDegrees=45
  swing2LimitDegrees=5
  twistLimitDegrees=0
 }
 {
  bone1Name=dyn_coat_ribbon_MSTR_R
  bone2Name=spine_04
  swing1LimitDegrees=45
  swing1Motion=Free
  swing2LimitDegrees=45
  swing2Motion=Free
  twistLimitDegrees=45
  twistMotion=Free
 }
 {
  bone1Name=spine_02
  bone2Name=pelvis
  swing1LimitDegrees=45
  swing1Motion=Free
  swing2LimitDegrees=45
  swing2Motion=Free
  twistLimitDegrees=45
  twistMotion=Free
 }
 {
  bone1Name=dyn_spinekerchief_MSTR_fL
  bone2Name=spine_04
  swing1LimitDegrees=45
  swing1Motion=Free
  swing2LimitDegrees=45
  swing2Motion=Free
  twistLimitDegrees=45
  twistMotion=Free
 }
 {bone1Name=dyn_spinekerchief_01_fL, bone2Name=dyn_spinekerchief_MSTR_fL, swing1LimitDegrees=25, swing2LimitDegrees=55, twistLimitDegrees=5}
 {bone1Name=dyn_spinekerchief_02_fL, bone2Name=dyn_spinekerchief_01_fL, swing1LimitDegrees=25, swing2LimitDegrees=55, twistLimitDegrees=5}
 {bone1Name=dyn_coat_ribbon_01_R, bone2Name=dyn_coat_ribbon_MSTR_R, swing1LimitDegrees=15, swing2LimitDegrees=20, twistLimitDegrees=1}
 {bone1Name=dyn_coat_ribbon_02_R, bone2Name=dyn_coat_ribbon_01_R, swing1LimitDegrees=15, swing2LimitDegrees=30, twistLimitDegrees=1}
 {bone1Name=dyn_coat_ribbon_03_R, bone2Name=dyn_coat_ribbon_02_R, swing1LimitDegrees=15, swing2LimitDegrees=30, twistLimitDegrees=1}
 {bone1Name=dyn_coat_ribbon_04_R, bone2Name=dyn_coat_ribbon_03_R, swing1LimitDegrees=15, swing2LimitDegrees=30, twistLimitDegrees=1}
 {
  bone1Name=dyn_coat_02_fR
  bone2Name=spine_04
  swing1LimitDegrees=45
  swing1Motion=Free
  swing2LimitDegrees=45
  swing2Motion=Free
  twistLimitDegrees=45
  twistMotion=Free
 }
 {
  bone1Name=spine_03
  bone2Name=spine_02
  swing1LimitDegrees=45
  swing1Motion=Free
  swing2LimitDegrees=45
  swing2Motion=Free
  twistLimitDegrees=45
  twistMotion=Free
 }
 {
  bone1Name=dyn_coat_bk_02
  bone2Name=spine_04
  swing1LimitDegrees=45
  swing1Motion=Free
  swing2LimitDegrees=45
  swing2Motion=Free
  twistLimitDegrees=45
  twistMotion=Free
 }
]
