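# Physics asset for the TweakSmirk medium female body (WearingPackBool variant), providing capsule bodies for its dynamic coat, sleeves, right-side ribbon and left-front kerchief.
# Pelvis, spine, arm and thigh capsules cover the body, with most cloth chains rooted at spine_04, several through MSTR bones.
# Coat panels and ribbon use free angular constraints, while sleeve and kerchief chains have angular limits, tightest along the sleeves.
# Each sleeve also branches at segment 03 into a small two-bone, zero-twist tip chain.
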PhysicsAsset /BRCosmetics/Characters/Player/Female/Medium/Bodies/F_MED_TweakSmirk/Meshes/F_MED_TweakSmirk_WearingPackBool_Physics
bodies {
 dyn_coat_02_fR [{
   shapeType=Capsule
   length=14.4308
   radius=10.0888
   shapeName=dyn_coat_02_fR_capsule
   center=<7.7206 -0.8131 3.8704>
   rotation=<rot 0 -81.7302 -89.4366>
  }]
 dyn_coat_03_L [{
   shapeType=Capsule
   length=11.3623
   radius=7.5494
   shapeName=dyn_coat_03_L_capsule
   center=<8.7614 -1.1097 -5.3342>
   rotation=<rot 0.0225 0.8685 -89.9417>
  }]
 dyn_coat_03_R [{
   shapeType=Capsule
   length=11.3623
   radius=7.5494
   shapeName=dyn_coat_03_L_capsule
   center=<8.7608 -1.1081 -3.7994>
   rotation=<rot 0.0225 0.8685 -89.9417>
  }]
 dyn_coat_03_fL [{
   shapeType=Capsule
   length=9.4614
   radius=5.8636
   shapeName=dyn_coat_03_fL_capsule
   center=<8.8273 1.3011 -5.0613>
   rotation=<rot 11.1757 166.1636 -47.2426>
  }]
 dyn_coat_03_fR [{
   shapeType=Capsule
   length=9.4614
   radius=5.8636
   shapeName=dyn_coat_03_fL_capsule
   center=<8.3874 -0.318 -5.2633>
   rotation=<rot 11.1757 166.1636 -118.7618>
  }]
 dyn_coat_04_L [{
   shapeType=Capsule
   length=10.9777
   radius=8.47
   shapeName=dyn_coat_04_L_capsule
   center=<5.0825 0.0523 -8.3033>
   rotation=<rot 0 -5.8618 -86.9103>
  }]
 dyn_coat_04_R [{
   shapeType=Capsule
   length=10.9777
   radius=8.47
   shapeName=dyn_coat_04_L_capsule
   center=<5.0825 0.0523 -8.3033>
   rotation=<rot -0.2284 -1.6263 -86.9187>
  }]
 dyn_coat_04_fL [{
   shapeType=Capsule
   length=9.5354
   radius=7.4116
   shapeName=dyn_coat_04_fL_capsule
   center=<5.1102 1.7983 -7.3041>
   rotation=<rot 12.8429 178.4453 -47.7757>
  }]
 dyn_coat_04_fR [{
   shapeType=Capsule
   length=9.5354
   radius=7.4116
   shapeName=dyn_coat_04_fL_capsule
   center=<5.0224 0.1209 -7.4896>
   rotation=<rot 12.8429 178.4453 -119.295>
  }]
 dyn_coat_bk_03 [{
   shapeType=Capsule
   length=14.5081
   radius=8.7127
   shapeName=dyn_coat_bk_04_capsule
   center=<6.3141 0.1117 -5.2069>
   rotation=<rot 0 0.3514 -88.6356>
  }]
 dyn_coat_bk_04 [{
   shapeType=Capsule
   length=14.5081
   radius=8.7127
   shapeName=dyn_coat_bk_04_capsule
   center=<6.3144 0.0501 -7.7913>
   rotation=<rot 0 0.3514 -88.6356>
  }]
 dyn_coat_ribbon_01_R [{
   shapeType=Capsule
   length=0
   radius=12.5
   shapeName=dyn_coat_ribbon_02_R_capsule
   center=<12.2921 16.2576 2.9437>
   rotation=<rot 0 -85.0902 -89.6445>
  }]
 dyn_coat_ribbon_02_R [{
   shapeType=Capsule
   length=0
   radius=12.5
   shapeName=dyn_coat_ribbon_02_R_capsule
   center=<21.0839 13.7883 6.6303>
   rotation=<rot 0 -85.0902 -89.6445>
  }]
 dyn_coat_ribbon_03_R [{
   shapeType=Capsule
   length=0
   radius=12.5
   shapeName=dyn_coat_ribbon_02_R_capsule
   center=<21.3499 10.9015 9.5354>
   rotation=<rot 0 -85.0902 -89.6445>
  }]
 dyn_coat_ribbon_04_R [{
   shapeType=Capsule
   length=0
   radius=12.5
   shapeName=dyn_coat_ribbon_02_R_capsule
   center=<21.3866 10.3299 7.5325>
   rotation=<rot 0 -85.0902 -89.6445>
  }]
 dyn_coat_ribbon_MSTR_R [{shapeType=Capsule, length=0.005, radius=1.505, shapeName=dyn_coat_ribbon_MSTR_R_capsule, rotation=<rot 0 -90 0>}]
 dyn_coat_sleeves_mid_01_L [{
   shapeType=Capsule
   length=1.4694
   radius=11
   shapeName=dyn_coat_sleeves_mid_01_R_capsule
   center=<6.9442 0 12.2252>
   rotation=<rot -89.3584 -179.9964 179.9964>
  }]
 dyn_coat_sleeves_mid_01_R [{
   shapeType=Capsule
   length=1.4694
   radius=11
   shapeName=dyn_coat_sleeves_mid_01_R_capsule
   center=<6.9543 -0.322 13.1252>
   rotation=<rot -70.3025 91.7928 -91.9041>
  }]
 dyn_coat_sleeves_mid_02_L [{
   shapeType=Capsule
   length=7.5
   radius=9
   shapeName=dyn_coat_sleeves_mid_02_R_capsule
   center=<11.3554 0 6.4048>
   rotation=<rot -89.3584 -179.9965 179.9965>
  }]
 dyn_coat_sleeves_mid_02_R [{
   shapeType=Capsule
   length=7.5
   radius=9
   shapeName=dyn_coat_sleeves_mid_02_R_capsule
   center=<11.3478 0 5.729>
   rotation=<rot -89.3584 -179.9965 179.9965>
  }]
 dyn_coat_sleeves_mid_03_L [{
   shapeType=Capsule
   length=8.3707
   radius=10
   shapeName=dyn_coat_sleeves_mid_03_R_capsule
   center=<12.6465 0 6.9136>
   rotation=<rot -89.3584 -179.9966 179.9966>
  }]
 dyn_coat_sleeves_mid_03_R [{
   shapeType=Capsule
   length=8.3707
   radius=10
   shapeName=dyn_coat_sleeves_mid_03_R_capsule
   center=<12.6465 0 6.9136>
   rotation=<rot -89.3584 -179.9966 179.9966>
  }]
 dyn_coat_sleeves_mid_04_L [{
   shapeType=Capsule
   length=0
   radius=10.5
   shapeName=dyn_coat_sleeves_mid_04_R_capsule
   center=<6.8883 0 7.2329>
   rotation=<rot -89.3584 -179.9966 179.9966>
  }]
 dyn_coat_sleeves_mid_04_L_01 [{
   shapeType=Capsule
   length=3.505
   radius=2.5
   shapeName=dyn_coat_sleeves_mid_04_L_01_capsule
   center=<3.2044 0 -0.6248>
   rotation=<rot -78.9671 180 -180>
  }]
 dyn_coat_sleeves_mid_04_L_02 [{
   shapeType=Capsule
   length=3.7937
   radius=2.5
   shapeName=dyn_coat_sleeves_mid_04_L_02_capsule
   center=<3.2044 0 -0.6248>
   rotation=<rot -78.9671 180 -180>
  }]
 dyn_coat_sleeves_mid_04_R [{
   shapeType=Capsule
   length=0
   radius=10.5
   shapeName=dyn_coat_sleeves_mid_04_R_capsule
   center=<6.8883 0 7.2329>
   rotation=<rot -89.3584 -179.9966 179.9966>
  }]
 dyn_coat_sleeves_mid_04_R_01 [{
   shapeType=Capsule
   length=3.505
   radius=2.5
   shapeName=dyn_coat_sleeves_mid_04_L_01_capsule
   center=<3.1634 0 0.1782>
   rotation=<rot -87.0826 0 0>
  }]
 dyn_coat_sleeves_mid_04_R_02 [{
   shapeType=Capsule
   length=3.7937
   radius=2.5
   shapeName=dyn_coat_sleeves_mid_04_L_02_capsule
   center=<3.1743 0 0.2111>
   rotation=<rot -87.9391 0 0>
  }]
 dyn_coat_sleeves_mid_MSTR_L [{shapeType=Capsule, length=0.505, radius=0.505, shapeName=dyn_coat_sleeves_mid_MSTR_L_capsule, rotation=<rot 0 -90 0>}]
 dyn_coat_sleeves_mid_MSTR_R [{shapeType=Capsule, length=0.505, radius=0.505, shapeName=dyn_coat_sleeves_mid_MSTR_R_capsule, rotation=<rot 0 -90 0>}]
 dyn_spinekerchief_01_fL [{
   shapeType=Capsule
   length=6.0026
   radius=5.2507
   shapeName=dyn_spinekerchief_01_fL_capsule
   center=<4.2459 0.3206 3.1611>
   rotation=<rot 0 85.6117 -84.4147>
  }]
 dyn_spinekerchief_02_fL [{
   shapeType=Capsule
   length=4.0522
   radius=3.3128
   shapeName=dyn_spinekerchief_02_fL_capsule
   center=<4.9038 0.3626 1.9006>
   rotation=<rot -0.0907 -89.8608 -89.3328>
  }]
 dyn_spinekerchief_MSTR_fL [{shapeType=Capsule, length=0.505, radius=0.505, shapeName=dyn_spinekerchief_MSTR_fL_capsule, rotation=<rot 0 -90 0>}]
 lowerarm_l [{shapeType=Capsule, length=26.6566, radius=6.1785, center=<15.9433 1.7255 0.0931>, rotation=<rot 0 89.5914 -94.9167>}]
 lowerarm_r [{shapeType=Capsule, length=31.8489, radius=5.3498, center=<-14.7532 -1.2494 -0.092>, rotation=<rot 0.0001 88.7368 -96.8977>}]
 pelvis [{shapeType=Capsule, length=10.8253, radius=8.8183, shapeName=pelvis_capsule}]
 spine_02 [{shapeType=Capsule, length=12.6242, radius=13.452, shapeName=spine_04_capsule, center=<-5.0689 5.5456 -0.4062>}]
 spine_04 [{shapeType=Capsule, length=20.1242, radius=13.452, shapeName=spine_04_capsule, center=<-2.2167 6.9039 -0.4062>}]
 thigh_l [{
   shapeType=Capsule
   length=20.7206
   radius=10.5299
   shapeName=thigh_l_capsule
   center=<-24.1144 -0.3407 0.4813>
   rotation=<rot 0 -86.5888 -91.0201>
  }]
 thigh_r [{
   shapeType=Capsule
   length=20.7495
   radius=7.9963
   shapeName=thigh_r_capsule
   center=<23.9365 0.0796 0.6477>
   rotation=<rot 0 -91.6432 -85.5049>
  }]
 upperarm_l [{
   shapeType=Capsule
   length=18.6938
   radius=4.25
   shapeName=upperarm_l_capsule
   center=<14.9359 0.6618 -1.045>
   rotation=<rot -54.8976 88.0613 -84.0489>
  }]
 upperarm_r [{
   shapeType=Capsule
   length=18.6938
   radius=4.25
   shapeName=upperarm_l_capsule
   center=<-17.5827 1.2141 -0.187>
   rotation=<rot -86.3754 22.398 -19.4261>
  }]
}
constraints [
 {bone1Name=lowerarm_l, bone2Name=upperarm_l, swing1LimitDegrees=45, swing2LimitDegrees=45, twistLimitDegrees=45}
 {bone1Name=lowerarm_r, bone2Name=upperarm_r, swing1LimitDegrees=45, swing2LimitDegrees=45, twistLimitDegrees=45}
 {
  bone1Name=upperarm_r
  bone2Name=spine_04
  swing1LimitDegrees=45
  swing1Motion=Free
  swing2LimitDegrees=45
  swing2Motion=Free
  twistLimitDegrees=45
  twistMotion=Free
 }
 {bone1Name=dyn_coat_sleeves_mid_04_R, bone2Name=dyn_coat_sleeves_mid_03_R, swing1LimitDegrees=3, swing2LimitDegrees=23, twistLimitDegrees=5}
 {bone1Name=dyn_coat_sleeves_mid_03_R, bone2Name=dyn_coat_sleeves_mid_02_R, swing1LimitDegrees=1, swing2LimitDegrees=15, twistLimitDegrees=5}
 {bone1Name=dyn_coat_sleeves_mid_02_R, bone2Name=dyn_coat_sleeves_mid_01_R, swing1LimitDegrees=1, swing2LimitDegrees=15, twistLimitDegrees=5}
 {bone1Name=dyn_coat_sleeves_mid_01_R, bone2Name=dyn_coat_sleeves_mid_MSTR_R, swing1LimitDegrees=5, swing2LimitDegrees=20, twistLimitDegrees=5}
 {
  bone1Name=dyn_coat_sleeves_mid_MSTR_R
  bone2Name=spine_04
  swing1LimitDegrees=45
  swing1Motion=Free
  swing2LimitDegrees=45
  swing2Motion=Free
  twistLimitDegrees=45
  twistMotion=Free
 }
 {
  bone1Name=upperarm_l
  bone2Name=spine_04
  swing1LimitDegrees=45
  swing1Motion=Free
  swing2LimitDegrees=45
  swing2Motion=Free
  twistLimitDegrees=45
  twistMotion=Free
 }
 {bone1Name=dyn_coat_sleeves_mid_04_L, bone2Name=dyn_coat_sleeves_mid_03_L, swing1LimitDegrees=2.5, swing2LimitDegrees=23, twistLimitDegrees=5}
 {bone1Name=dyn_coat_sleeves_mid_03_L, bone2Name=dyn_coat_sleeves_mid_02_L, swing1LimitDegrees=1, swing2LimitDegrees=15, twistLimitDegrees=5}
 {bone1Name=dyn_coat_sleeves_mid_02_L, bone2Name=dyn_coat_sleeves_mid_01_L, swing1LimitDegrees=1, swing2LimitDegrees=15, twistLimitDegrees=5}
 {bone1Name=dyn_coat_sleeves_mid_01_L, bone2Name=dyn_coat_sleeves_mid_MSTR_L, swing1LimitDegrees=5, swing2LimitDegrees=20, twistLimitDegrees=5}
 {
  bone1Name=dyn_coat_sleeves_mid_MSTR_L
  bone2Name=spine_04
  swing1LimitDegrees=45
  swing1Motion=Free
  swing2LimitDegrees=45
  swing2Motion=Free
  twistLimitDegrees=45
  twistMotion=Free
 }
 {
  bone1Name=dyn_coat_bk_04
  bone2Name=dyn_coat_bk_03
  swing1LimitDegrees=45
  swing1Motion=Free
  swing2LimitDegrees=45
  swing2Motion=Free
  twistLimitDegrees=45
  twistMotion=Free
 }
 {
  bone1Name=dyn_coat_bk_03
  bone2Name=spine_04
  swing1LimitDegrees=45
  swing1Motion=Free
  swing2LimitDegrees=45
  swing2Motion=Free
  twistLimitDegrees=45
  twistMotion=Free
 }
 {
  bone1Name=dyn_coat_04_fR
  bone2Name=dyn_coat_03_fR
  swing1LimitDegrees=45
  swing1Motion=Free
  swing2LimitDegrees=45
  swing2Motion=Free
  twistLimitDegrees=45
  twistMotion=Free
 }
 {
  bone1Name=dyn_coat_03_fR
  bone2Name=dyn_coat_02_fR
  swing1LimitDegrees=45
  swing1Motion=Free
  swing2LimitDegrees=45
  swing2Motion=Free
  twistLimitDegrees=45
  twistMotion=Free
 }
 {
  bone1Name=dyn_coat_04_fL
  bone2Name=dyn_coat_03_fL
  swing1LimitDegrees=45
  swing1Motion=Free
  swing2LimitDegrees=45
  swing2Motion=Free
  twistLimitDegrees=45
  twistMotion=Free
 }
 {
  bone1Name=dyn_coat_03_fL
  bone2Name=spine_04
  swing1LimitDegrees=45
  swing1Motion=Free
  swing2LimitDegrees=45
  swing2Motion=Free
  twistLimitDegrees=45
  twistMotion=Free
 }
 {
  bone1Name=dyn_coat_04_R
  bone2Name=dyn_coat_03_R
  swing1LimitDegrees=45
  swing1Motion=Free
  swing2LimitDegrees=45
  swing2Motion=Free
  twistLimitDegrees=45
  twistMotion=Free
 }
 {
  bone1Name=dyn_coat_03_R
  bone2Name=spine_04
  swing1LimitDegrees=45
  swing1Motion=Free
  swing2LimitDegrees=45
  swing2Motion=Free
  twistLimitDegrees=45
  twistMotion=Free
 }
 {
  bone1Name=dyn_coat_04_L
  bone2Name=dyn_coat_03_L
  swing1LimitDegrees=45
  swing1Motion=Free
  swing2LimitDegrees=45
  swing2Motion=Free
  twistLimitDegrees=45
  twistMotion=Free
 }
 {
  bone1Name=dyn_coat_03_L
  bone2Name=spine_04
  swing1LimitDegrees=45
  swing1Motion=Free
  swing2LimitDegrees=45
  swing2Motion=Free
  twistLimitDegrees=45
  twistMotion=Free
 }
 {
  bone1Name=spine_04
  bone2Name=spine_02
  swing1LimitDegrees=45
  swing1Motion=Free
  swing2LimitDegrees=45
  swing2Motion=Free
  twistLimitDegrees=45
  twistMotion=Free
 }
 {
  bone1Name=thigh_l
  bone2Name=pelvis
  swing1LimitDegrees=45
  swing1Motion=Free
  swing2LimitDegrees=45
  swing2Motion=Free
  twistLimitDegrees=45
  twistMotion=Free
 }
 {
  bone1Name=thigh_r
  bone2Name=pelvis
  swing1LimitDegrees=45
  swing1Motion=Free
  swing2LimitDegrees=45
  swing2Motion=Free
  twistLimitDegrees=45
  twistMotion=Free
 }
 {
  bone1Name=dyn_coat_sleeves_mid_04_L_01
  bone2Name=dyn_coat_sleeves_mid_03_L
  swing1LimitDegrees=45
  swing2LimitDegrees=13
  twistLimitDegrees=0
 }
 {
  bone1Name=dyn_coat_sleeves_mid_04_L_02
  bone2Name=dyn_coat_sleeves_mid_04_L_01
  swing1LimitDegrees=45
  swing2LimitDegrees=13
  twistLimitDegrees=0
 }
 {
  bone1Name=dyn_coat_sleeves_mid_04_R_01
  bone2Name=dyn_coat_sleeves_mid_03_R
  swing1LimitDegrees=45
  swing2LimitDegrees=13
  twistLimitDegrees=0
 }
 {
  bone1Name=dyn_coat_sleeves_mid_04_R_02
  bone2Name=dyn_coat_sleeves_mid_04_R_01
  swing1LimitDegrees=45
  swing2LimitDegrees=13
  twistLimitDegrees=0
 }
 {
  bone1Name=dyn_coat_ribbon_MSTR_R
  bone2Name=spine_04
  swing1LimitDegrees=45
  swing1Motion=Free
  swing2LimitDegrees=45
  swing2Motion=Free
  twistLimitDegrees=45
  twistMotion=Free
 }
 {
  bone1Name=spine_02
  bone2Name=pelvis
  swing1LimitDegrees=45
  swing1Motion=Free
  swing2LimitDegrees=45
  swing2Motion=Free
  twistLimitDegrees=45
  twistMotion=Free
 }
 {
  bone1Name=dyn_spinekerchief_MSTR_fL
  bone2Name=spine_04
  swing1LimitDegrees=45
  swing1Motion=Free
  swing2LimitDegrees=45
  swing2Motion=Free
  twistLimitDegrees=45
  twistMotion=Free
 }
 {bone1Name=dyn_spinekerchief_01_fL, bone2Name=dyn_spinekerchief_MSTR_fL, swing1LimitDegrees=25, swing2LimitDegrees=55, twistLimitDegrees=5}
 {bone1Name=dyn_spinekerchief_02_fL, bone2Name=dyn_spinekerchief_01_fL, swing1LimitDegrees=25, swing2LimitDegrees=55, twistLimitDegrees=5}
 {
  bone1Name=dyn_coat_02_fR
  bone2Name=spine_04
  swing1LimitDegrees=45
  swing1Motion=Free
  swing2LimitDegrees=45
  swing2Motion=Free
  twistLimitDegrees=45
  twistMotion=Free
 }
 {
  bone1Name=dyn_coat_ribbon_01_R
  bone2Name=dyn_coat_ribbon_MSTR_R
  swing1LimitDegrees=45
  swing1Motion=Free
  swing2LimitDegrees=45
  swing2Motion=Free
  twistLimitDegrees=45
  twistMotion=Free
 }
 {
  bone1Name=dyn_coat_ribbon_02_R
  bone2Name=dyn_coat_ribbon_01_R
  swing1LimitDegrees=45
  swing1Motion=Free
  swing2LimitDegrees=45
  swing2Motion=Free
  twistLimitDegrees=45
  twistMotion=Free
 }
 {
  bone1Name=dyn_coat_ribbon_03_R
  bone2Name=dyn_coat_ribbon_02_R
  swing1LimitDegrees=45
  swing1Motion=Free
  swing2LimitDegrees=45
  swing2Motion=Free
  twistLimitDegrees=45
  twistMotion=Free
 }
 {
  bone1Name=dyn_coat_ribbon_04_R
  bone2Name=dyn_coat_ribbon_03_R
  swing1LimitDegrees=45
  swing1Motion=Free
  swing2LimitDegrees=45
  swing2Motion=Free
  twistLimitDegrees=45
  twistMotion=Free
 }
]
