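# Post-process style cosmetic AnimBlueprint for the F_MED_TweakSmirk female body, adding secondary motion to coat, ribbons, sleeves, upper torso and knee details.
# The parent mesh pose passes through a Control Rig driven by slide and DBNO states, then chained bone offsets, constraints, AnimDynamics and coat RigidBody.
# That cached result feeds two RigidBody branches, selected by a backbling check; the pack branch constrains right ribbons to back-coat bones.
# EventGraph updates DBNO and parameter-asset gravity settings.
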AnimBlueprint /BRCosmetics/Characters/Player/Female/Medium/Bodies/F_MED_TweakSmirk/Meshes/F_MED_TweakSmirk_AnimBP
target_skeleton=&/F_MED_TweakSmirk_Skeleton
graphs [
 AnimGraph {nodes [
   "Transform (Modify) Bone - Bone: dyn_coat_ribbon_03_R"(Alpha=1, Translation="0,0,-2", Rotation="0,0,-15", Scale="1,1,1") {
    rotationMode=BMM_Additive
    rotationSpace=BCS_BoneSpace
    translationMode=BMM_Additive
    translationSpace=BCS_BoneSpace
    boneToModify {boneName=dyn_coat_ribbon_03_R}
    rotation {roll=-15}
    translation {z=-2}
   }
   n1 = "Constraint - dyn_coat_ribbon_01_R (R)"(ComponentPose=Local To Component(LocalPose=Use cached pose 'AnimDyn_Pose'()), Alpha=1, ConstraintWeights_0=1) {
    boneToModify {boneName=dyn_coat_ribbon_01_R}
    constraintSetup [{transformType=Rotation, targetBone {boneName=dyn_coat_bk_01}}]
    constraintWeights [1]
   }
   n2 = "Constraint - dyn_coat_ribbon_02_R (R)"(ComponentPose=@n1, Alpha=1, ConstraintWeights_0=1) {
    boneToModify {boneName=dyn_coat_ribbon_02_R}
    constraintSetup [{transformType=Rotation, targetBone {boneName=dyn_coat_bk_01}}]
    constraintWeights [1]
   }
   n3 = "Constraint - dyn_coat_ribbon_03_R (R)"(ComponentPose=@n2, Alpha=1, ConstraintWeights_0=1) {
    boneToModify {boneName=dyn_coat_ribbon_03_R}
    constraintSetup [{transformType=Rotation, targetBone {boneName=dyn_coat_bk_02}}]
    constraintWeights [1]
   }
   n4 = "Constraint - dyn_coat_ribbon_04_R (R)"(ComponentPose=@n3, Alpha=1, ConstraintWeights_0=1) {
    boneToModify {boneName=dyn_coat_ribbon_04_R}
    constraintSetup [{transformType=Rotation, targetBone {boneName=dyn_coat_bk_03}}]
    constraintWeights [1]
   }
   n5 = RigidBody!(
    ComponentPose=@n4
    bAlphaBoolEnabled:bool
    OverrideWorldGravity:Vector="0,0,-980"
    ExternalForce
    SimSpaceSettings:Sim Space Settings Structure=(WorldAlpha=0,VelocityScaleZ=1,DampingAlpha=1,MaxLinearVelocity=10000,MaxAngularVelocity=10000,MaxLinearAcceleration=10000,MaxAngularAcceleration=10000,ExternalLinearDragV=(X=0,Y=0,Z=0),ExternalLinearVelocity=(X=0,Y=0,Z=0),ExternalAngularVelocity=(X=0,Y=0,Z=0))
   ) -> (Pose) {
    alphaInputType=Bool
    bClampLinearTranslationLimitToRefPose=true
    bOverrideWorldGravity=true
    overridePhysicsAsset=&/F_MED_TweakSmirk_WearingPackBool_Physics
    simulationSpace=BaseBoneSpace
    alphaBoolBlend {blendOption=Linear}
    baseBoneRef {boneName=spine_05}
    componentAppliedLinearAccClamp=<3000 3000 2700>
    componentLinearAccScale=<0.01 0.4 0.35>
    componentLinearVelScale=<0.08 0.4 0.4>
    overrideWorldGravity {z=-980}
   }
   n6 = RigidBody(
    ComponentPose=Local To Component(LocalPose=Use cached pose 'AnimDyn_Pose'())
    OverrideWorldGravity="0,0,-980"
    SimSpaceSettings=(WorldAlpha=0,VelocityScaleZ=1,DampingAlpha=1,MaxLinearVelocity=10000,MaxAngularVelocity=10000,MaxLinearAcceleration=10000,MaxAngularAcceleration=10000,ExternalLinearDragV=(X=0,Y=0,Z=0),ExternalLinearVelocity=(X=0,Y=0,Z=0),ExternalAngularVelocity=(X=0,Y=0,Z=0))
   ) {
    alphaInputType=Bool
    bClampLinearTranslationLimitToRefPose=true
    bOverrideWorldGravity=true
    overridePhysicsAsset=&/F_MED_TweakSmirk_Physics
    simulationSpace=BaseBoneSpace
    alphaBoolBlend {blendOption=Linear}
    baseBoneRef {boneName=spine_05}
    componentAppliedLinearAccClamp=<3000 3000 2700>
    componentLinearAccScale=<0.01 0.4 0.35>
    componentLinearVelScale=<0.08 0.4 0.4>
    overrideWorldGravity {z=-980}
   }
   Output Pose(Result=Blend Poses by bool(
     BlendPose_0=Component To Local(ComponentPose=@n5)
     BlendPose_1=Component To Local(ComponentPose=@n6)
     bActiveValue=Get isWearingPackBool() -> (isWearingPackBool:bool)
    ) {blendType=Linear}) {layerGroup=None}
   n7 = Control Rig!(
    Source=Copy Pose From Mesh() {bUseAttachedParent=true}
    Alpha
    isSliding:bool=Get bIsSliding()
    isBaseballSliding:bool=Get bIsBaseballSliding()
    DBNO:bool=Get DBNO() -> (DBNO:bool)
    DBNO_Carry:bool=Get bIsDBNOCarried()
   ) -> (Pose) {
    alphaBoolBlend {blendOption=Linear}
    controlRigAssetReference {blueprintRigClass=&/F_MED_TweakSmirk_CtrlRig.F_MED_TweakSmirk_CtrlRig_C}
   }
   Save cached pose 'SkirtCachePose'(Pose=@n7) {cacheName=SkirtCachePose}
   n8 = "Transform (Modify) Bone - Bone: attach_backpack"!(
    ComponentPose=Local To Component(LocalPose=Use cached pose 'SkirtCachePose'())
    bAlphaBoolEnabled:bool=True
    Translation="0,0,0"
    Rotation="0,0,7"
    Scale="1,1,1"
   ) -> (Pose) {
    alphaInputType=Bool
    rotationMode=BMM_Additive
    rotationSpace=BCS_BoneSpace
    scaleMode=BMM_Additive
    scaleSpace=BCS_BoneSpace
    translationMode=BMM_Additive
    translationSpace=BCS_BoneSpace
    alphaBoolBlend {blendInTime=0.2, blendOutTime=0.2}
    boneToModify {boneName=attach_backpack}
    rotation {roll=7}
   }
   n9 = "Transform (Modify) Bone - Bone: dyn_upperTorso_l"(ComponentPose=@n8, Alpha=1, Translation="0,0,1", Rotation="0,0,0", Scale="1,1,1") {translationMode=BMM_Additive, translationSpace=BCS_BoneSpace, boneToModify {boneName=dyn_upperTorso_l}, translation {z=1}}
   n10 = "Transform (Modify) Bone - Bone: dyn_upperTorso_r"(ComponentPose=@n9, Alpha=1, Translation="0,0,1", Rotation="0,0,0", Scale="1,1,1") {translationMode=BMM_Additive, translationSpace=BCS_BoneSpace, boneToModify {boneName=dyn_upperTorso_r}, translation {z=1}}
   n11 = Anim Dynamics - Bone: dyn_upperTorso_r!(ComponentPose=@n10, Alpha=1, GravityOverride:Vector="0,0,-980", ExternalForce="0,0,0") -> (Pose) {
    angularBiasOverride=1
    angularDampingOverride=0.9999
    angularSpringConstant=5000
    bAngularSpring=true
    bGravityOverrideInSimSpace=true
    bLinearSpring=true
    bOverrideAngularBias=true
    bOverrideAngularDamping=true
    bOverrideLinearDamping=true
    bUseGravityOverride=true
    linearDampingOverride=0.9999
    linearSpringConstant=100
    simulationSpace=RootRelative
    boundBone {boneName=dyn_upperTorso_r}
    componentAppliedLinearAccClamp=<2 2 2>
    gravityOverride {z=-980}
    physicsBodyDefinitions [{
      boundBone {boneName=dyn_upperTorso_r}
      boxExtents=<41 41 41>
      constraintSetup {angularLimitsMax=<0 1 3>, angularLimitsMin=<0 -4 -3>, angularTarget=<1 0 0>}
      localJointOffset=<2 0 0>
     }]
    relativeSpaceBone {boneName=spine_04}
    simSpaceSettings {maxAngularAcceleration=45, maxAngularVelocity=8, simSpaceAngularAlpha=0.25}
   }
   n12 = Anim Dynamics - Bone: dyn_upperTorso_l!(ComponentPose=@n11, Alpha=1, GravityOverride:Vector="0,0,-980", ExternalForce="0,0,0") -> (Pose) {
    angularBiasOverride=1
    angularDampingOverride=0.9999
    angularSpringConstant=5000
    bAngularSpring=true
    bGravityOverrideInSimSpace=true
    bLinearSpring=true
    bOverrideAngularBias=true
    bOverrideAngularDamping=true
    bOverrideLinearDamping=true
    bUseGravityOverride=true
    linearDampingOverride=0.9999
    linearSpringConstant=100
    simulationSpace=RootRelative
    boundBone {boneName=dyn_upperTorso_l}
    componentAppliedLinearAccClamp=<2 2 2>
    gravityOverride {z=-980}
    physicsBodyDefinitions [{
      boundBone {boneName=dyn_upperTorso_l}
      boxExtents=<40 40 40>
      constraintSetup {angularLimitsMax=<0 1 3>, angularLimitsMin=<0 -4 -3>, angularTarget=<1 0 0>}
      localJointOffset=<2 0 0>
     }]
    relativeSpaceBone {boneName=spine_04}
    simSpaceSettings {maxAngularAcceleration=45, maxAngularVelocity=8, simSpaceAngularAlpha=0.25}
   }
   n13 = Anim Dynamics - Bone: dyn_upperTorso_r(ComponentPose=@n12, Alpha=1, GravityOverride="0,0,-200", ExternalForce="0,0,0") {
    angularBiasOverride=1
    angularDampingOverride=0.99
    angularSpringConstant=10000
    bGravityOverrideInSimSpace=true
    bLinearSpring=true
    bOverrideAngularBias=true
    bOverrideLinearDamping=true
    bUseGravityOverride=true
    linearDampingOverride=1
    linearSpringConstant=180
    simulationSpace=RootRelative
    boundBone {boneName=dyn_upperTorso_r}
    componentAppliedLinearAccClamp=<5 5 5>
    gravityOverride {z=-200}
    physicsBodyDefinitions [{
      boundBone {boneName=dyn_upperTorso_r}
      boxExtents=<100 100 100>
      constraintSetup {angularLimitsMax=<0 0 0>, angularLimitsMin=<0 0 0>, angularTarget=<1 0 0>, linearAxesMax=<0 0.3 0.7>, linearAxesMin=<0 -0.3 -0.3>}
      localJointOffset=<0.2 0 0>
     }]
    relativeSpaceBone {boneName=spine_04}
    simSpaceSettings {maxAngularAcceleration=200, maxAngularVelocity=200, simSpaceAngularAlpha=0.1}
   }
   n14 = Anim Dynamics - Bone: dyn_upperTorso_l(ComponentPose=@n13, Alpha=1, GravityOverride="0,0,-200", ExternalForce="0,0,0") {
    angularBiasOverride=1
    angularDampingOverride=0.99
    angularSpringConstant=10000
    bGravityOverrideInSimSpace=true
    bLinearSpring=true
    bOverrideAngularBias=true
    bOverrideLinearDamping=true
    bUseGravityOverride=true
    linearDampingOverride=1
    linearSpringConstant=180
    simulationSpace=RootRelative
    boundBone {boneName=dyn_upperTorso_l}
    componentAppliedLinearAccClamp=<5 5 5>
    gravityOverride {z=-200}
    physicsBodyDefinitions [{
      boundBone {boneName=dyn_upperTorso_l}
      boxExtents=<110 110 110>
      constraintSetup {angularLimitsMax=<0 0 0>, angularLimitsMin=<0 0 0>, angularTarget=<1 0 0>, linearAxesMax=<0 0.37 0.757>, linearAxesMin=<0 -0.37 -0.37>}
      localJointOffset=<0.2 0 0>
     }]
    relativeSpaceBone {boneName=spine_04}
    simSpaceSettings {maxAngularAcceleration=200, maxAngularVelocity=200, simSpaceAngularAlpha=0.1}
   }
   n15 = "Constraint - dyn_rope_fR_01 (PPP)"!(ComponentPose=@n14, Alpha=1, ConstraintWeights_0=0.8, ConstraintWeights_1:float=0.2, ConstraintWeights_2:float=0) -> (Pose) {
    boneToModify {boneName=dyn_rope_fR_01}
    constraintSetup [{targetBone {boneName=spine_05}}, {targetBone {boneName=spine_04}}, {targetBone {boneName=calf_r}}]
    constraintWeights [1,1,1]
   }
   n16 = "Constraint - dyn_rope_fR_02 (PPP)"!(ComponentPose=@n15, Alpha=1, ConstraintWeights_0=0.8, ConstraintWeights_1:float=0.2, ConstraintWeights_2:float=0) -> (Pose) {
    boneToModify {boneName=dyn_rope_fR_02}
    constraintSetup [{targetBone {boneName=spine_04}}, {targetBone {boneName=spine_03}}, {targetBone {boneName=calf_r}}]
    constraintWeights [1,1,1]
   }
   n17 = Anim Dynamics - Bone: dyn_rope_fR_01!(ComponentPose=@n16, Alpha=1, GravityOverride:Vector="0,0,-980", ExternalForce="0,0,0") -> (Pose) {
    angularDampingOverride=0.99
    angularSpringConstant=800
    bAngularSpring=true
    bGravityOverrideInSimSpace=true
    bOverrideAngularDamping=true
    bUseGravityOverride=true
    simulationSpace=BoneRelative
    alphaBoolBlend {blendOption=Linear}
    boundBone {boneName=dyn_rope_fR_01}
    componentAppliedLinearAccClamp=<2000 2000 1600>
    componentLinearAccScale=<0.4 0.4 0.4>
    componentLinearVelScale=<0.45 0.45 0.55>
    gravityOverride {z=-980}
    physicsBodyDefinitions [{
      sphereCollisionRadius=2
      boundBone {boneName=dyn_rope_fR_01}
      boxExtents=<17 17 17>
      constraintSetup {angularLimitsMax=<0 10 0>, angularLimitsMin=<0 -13 0>, angularTarget=<0 0 0>}
      localJointOffset=<5 0 0>
     }]
    relativeSpaceBone {boneName=root}
    retargetingSettings {bEnabled=true}
   }
   n18 = Anim Dynamics - Bone: dyn_rope_fR_02!(ComponentPose=@n17, Alpha=1, GravityOverride:Vector="0,0,-980", ExternalForce="0,0,0") -> (Pose) {
    angularDampingOverride=0.99
    angularSpringConstant=800
    bAngularSpring=true
    bGravityOverrideInSimSpace=true
    bOverrideAngularDamping=true
    bUseGravityOverride=true
    simulationSpace=BoneRelative
    alphaBoolBlend {blendOption=Linear}
    boundBone {boneName=dyn_rope_fR_02}
    componentAppliedLinearAccClamp=<2000 2000 1600>
    componentLinearAccScale=<0.4 0.4 0.4>
    componentLinearVelScale=<0.45 0.45 0.55>
    gravityOverride {z=-980}
    physicsBodyDefinitions [{
      sphereCollisionRadius=2
      boundBone {boneName=dyn_rope_fR_02}
      boxExtents=<17 17 17>
      constraintSetup {angularLimitsMax=<0 7 -3>, angularLimitsMin=<0 -7 -3>, angularTarget=<0 0 0>}
      localJointOffset=<5 0 0>
     }]
    relativeSpaceBone {boneName=root}
    retargetingSettings {bEnabled=true}
   }
   n19 = "Constraint - dyn_knee_R_boots_01_base (PPP)"!(ComponentPose=@n18, Alpha=1, ConstraintWeights_0=1, ConstraintWeights_1:float=0, ConstraintWeights_2:float=0) -> (Pose) {
    boneToModify {boneName=dyn_knee_R_boots_01_base}
    constraintSetup [{targetBone {boneName=thigh_r}}, {targetBone {boneName=deform_knee_out_r}}, {targetBone {boneName=calf_r}}]
    constraintWeights [1,1,1]
   }
   n20 = "Constraint - dyn_knee_R_boots_02_base (PPP)"!(ComponentPose=@n19, Alpha=1, ConstraintWeights_0=1, ConstraintWeights_1:float=0, ConstraintWeights_2:float=0) -> (Pose) {
    boneToModify {boneName=dyn_knee_R_boots_02_base}
    constraintSetup [{targetBone {boneName=thigh_r}}, {targetBone {boneName=deform_knee_out_r}}, {targetBone {boneName=calf_r}}]
    constraintWeights [1,1,1]
   }
   n21 = Anim Dynamics - Bone: dyn_knee_R_boots_01!(ComponentPose=@n20, Alpha=1, GravityOverride:Vector="0,0,-980", ExternalForce="0,0,0") -> (Pose) {
    angularDampingOverride=0.99
    angularSpringConstant=800
    bAngularSpring=true
    bOverrideAngularDamping=true
    bUseGravityOverride=true
    gravityScale=0.3
    simulationSpace=RootRelative
    alphaBoolBlend {blendOption=Linear}
    boundBone {boneName=dyn_knee_R_boots_01}
    componentAppliedLinearAccClamp=<2000 2000 1600>
    componentLinearAccScale=<0.4 0.4 0.4>
    componentLinearVelScale=<0.45 0.45 0.55>
    gravityOverride {z=-980}
    physicsBodyDefinitions [{
      sphereCollisionRadius=2
      boundBone {boneName=dyn_knee_R_boots_01}
      boxExtents=<30 30 30>
      constraintSetup {angularLimitsMax=<0 0 7>, angularLimitsMin=<0 -10 -7>, angularTarget=<0 0 0>}
      localJointOffset=<5 0 0>
     }]
    relativeSpaceBone {boneName=root}
    retargetingSettings {bEnabled=true, source {rotation {pitch=-2}}}
   }
   n22 = Anim Dynamics - Bone: dyn_knee_R_boots_02!(ComponentPose=@n21, Alpha=1, GravityOverride:Vector="0,0,-980", ExternalForce="0,0,0") -> (Pose) {
    angularDampingOverride=0.99
    angularSpringConstant=800
    bAngularSpring=true
    bOverrideAngularDamping=true
    bUseGravityOverride=true
    gravityScale=0.3
    simulationSpace=RootRelative
    alphaBoolBlend {blendOption=Linear}
    boundBone {boneName=dyn_knee_R_boots_02}
    componentAppliedLinearAccClamp=<2000 2000 1600>
    componentLinearAccScale=<0.4 0.4 0.4>
    componentLinearVelScale=<0.45 0.45 0.55>
    gravityOverride {z=-980}
    physicsBodyDefinitions [{
      sphereCollisionRadius=2
      boundBone {boneName=dyn_knee_R_boots_02}
      boxExtents=<30 30 30>
      constraintSetup {angularLimitsMax=<0 10 7>, angularLimitsMin=<0 0 -7>, angularTarget=<0 0 0>}
      localJointOffset=<5 0 0>
     }]
    relativeSpaceBone {boneName=root}
    retargetingSettings {bEnabled=true, source {rotation {pitch=-2}}}
   }
   n23 = "Constraint - dyn_knee_zipper_r (PPP)"!(ComponentPose=@n22, Alpha=1, ConstraintWeights_0=0.5208, ConstraintWeights_1:float=0.3007, ConstraintWeights_2:float=0.1785) -> (Pose) {
    boneToModify {boneName=dyn_knee_zipper_r}
    constraintSetup [{targetBone {boneName=calf_r}}, {targetBone {boneName=thigh_r}}, {targetBone {boneName=deform_knee_out_r}}]
    constraintWeights [1,1,1]
   }
   n24 = Anim Dynamics - Bone: dyn_knee_zipper_r!(ComponentPose=@n23, Alpha=1, GravityOverride:Vector="0,0,-980", ExternalForce="0,0,0") -> (Pose) {
    angularDampingOverride=0.99
    angularSpringConstant=800
    bAngularSpring=true
    bOverrideAngularDamping=true
    bUseGravityOverride=true
    gravityScale=0.3
    simulationSpace=RootRelative
    alphaBoolBlend {blendOption=Linear}
    boundBone {boneName=dyn_knee_zipper_r}
    componentAppliedLinearAccClamp=<2000 2000 1600>
    componentLinearAccScale=<0.4 0.4 0.4>
    componentLinearVelScale=<0.45 0.45 0.55>
    gravityOverride {z=-980}
    physicsBodyDefinitions [{
      sphereCollisionRadius=2
      boundBone {boneName=dyn_knee_zipper_r}
      boxExtents=<17 17 17>
      constraintSetup {angularLimitsMax=<0 1 10>, angularLimitsMin=<0 -1 -20>, angularTarget=<0 0 0>}
      localJointOffset=<5 0 0>
     }]
    relativeSpaceBone {boneName=root}
    retargetingSettings {bEnabled=true, source {rotation {pitch=-2}}}
   }
   n25 = Anim Dynamics - Bone: dyn_coat_01_fL!(ComponentPose=@n24, Alpha=1, GravityOverride:Vector="0,0,-980", ExternalForce="0,0,0") -> (Pose) {
    angularDampingOverride=0.99
    angularSpringConstant=800
    bAngularSpring=true
    bOverrideAngularDamping=true
    bUseGravityOverride=true
    gravityScale=0.3
    simulationSpace=RootRelative
    alphaBoolBlend {blendOption=Linear}
    boundBone {boneName=dyn_coat_01_fL}
    componentAppliedLinearAccClamp=<2000 2000 1600>
    componentLinearAccScale=<0.4 0.4 0.4>
    componentLinearVelScale=<0.45 0.45 0.55>
    gravityOverride {z=-980}
    physicsBodyDefinitions [{
      sphereCollisionRadius=2
      boundBone {boneName=dyn_coat_01_fL}
      boxExtents=<30 30 30>
      constraintSetup {angularLimitsMax=<0 2 0>, angularLimitsMin=<0 -15 -10>, angularTarget=<0 0 0>}
      localJointOffset=<5 0 0>
     }]
    relativeSpaceBone {boneName=root}
    retargetingSettings {bEnabled=true, source {rotation {pitch=-2}}}
   }
   n26 = Anim Dynamics - Bone: dyn_coat_01_L!(ComponentPose=@n25, Alpha=1, GravityOverride:Vector="0,0,-980") -> (Pose) {
    angularDampingOverride=0.99
    angularSpringConstant=800
    bOverrideAngularDamping=true
    bUseGravityOverride=true
    gravityScale=0.3
    linearSpringConstant=800
    simulationSpace=RootRelative
    alphaBoolBlend {blendOption=Linear}
    boundBone {boneName=dyn_coat_01_L}
    componentAppliedLinearAccClamp=<2000 2000 1600>
    componentLinearAccScale=<0.4 0.4 0.4>
    componentLinearVelScale=<0.45 0.45 0.55>
    gravityOverride {z=-980}
    physicsBodyDefinitions [{
      sphereCollisionRadius=2
      boundBone {boneName=dyn_coat_01_L}
      boxExtents=<30 30 30>
      constraintSetup {angularLimitsMax=<0 0 2>, angularLimitsMin=<0 -15 -2>, angularTarget=<0 0 0>}
      localJointOffset=<5 0 0>
     }]
    relativeSpaceBone {boneName=root}
    retargetingSettings {bEnabled=true, sourceMaximum=70}
   }
   n27 = Anim Dynamics - Bone: dyn_coat_02_L!(ComponentPose=@n26, Alpha=1, GravityOverride:Vector="0,0,-980") -> (Pose) {
    angularDampingOverride=0.99
    angularSpringConstant=800
    bAngularSpring=true
    bOverrideAngularDamping=true
    bUseGravityOverride=true
    gravityScale=0.3
    simulationSpace=RootRelative
    alphaBoolBlend {blendOption=Linear}
    boundBone {boneName=dyn_coat_02_L}
    componentAppliedLinearAccClamp=<2000 2000 1600>
    componentLinearAccScale=<0.4 0.4 0.4>
    componentLinearVelScale=<0.45 0.45 0.55>
    gravityOverride {z=-980}
    physicsBodyDefinitions [{
      sphereCollisionRadius=2
      boundBone {boneName=dyn_coat_02_L}
      boxExtents=<20 20 20>
      constraintSetup {angularLimitsMax=<0 0 2>, angularLimitsMin=<0 -20 -2>, angularTarget=<0 0 0>}
      localJointOffset=<5 0 0>
     }]
    relativeSpaceBone {boneName=root}
    retargetingSettings {bEnabled=true, sourceMaximum=70}
   }
   n28 = Anim Dynamics - Bone: dyn_coat_bk_01!(ComponentPose=@n27, Alpha=1, GravityOverride:Vector="0,0,-980") -> (Pose) {
    angularDampingOverride=0.99
    angularSpringConstant=800
    bAngularSpring=true
    bOverrideAngularDamping=true
    bUseGravityOverride=true
    gravityScale=0.3
    simulationSpace=RootRelative
    alphaBoolBlend {blendOption=Linear}
    boundBone {boneName=dyn_coat_bk_01}
    componentAppliedLinearAccClamp=<2000 2000 1600>
    componentLinearAccScale=<0.4 0.4 0.4>
    componentLinearVelScale=<0.45 0.45 0.55>
    gravityOverride {z=-980}
    physicsBodyDefinitions [{
      sphereCollisionRadius=2
      boundBone {boneName=dyn_coat_bk_01}
      boxExtents=<30 30 30>
      constraintSetup {angularLimitsMax=<0 0 2>, angularLimitsMin=<0 -15 -2>, angularTarget=<0 0 0>}
      localJointOffset=<5 0 0>
     }]
    relativeSpaceBone {boneName=root}
    retargetingSettings {bEnabled=true, sourceMaximum=52}
   }
   n29 = Anim Dynamics - Bone: dyn_coat_bk_02!(ComponentPose=@n28, Alpha=1, GravityOverride:Vector="0,0,-980") -> (Pose) {
    angularDampingOverride=0.99
    angularSpringConstant=800
    bAngularSpring=true
    bOverrideAngularDamping=true
    bUseGravityOverride=true
    gravityScale=0.3
    simulationSpace=RootRelative
    alphaBoolBlend {blendOption=Linear}
    boundBone {boneName=dyn_coat_bk_02}
    componentAppliedLinearAccClamp=<2000 2000 1600>
    componentLinearAccScale=<0.4 0.4 0.4>
    componentLinearVelScale=<0.45 0.45 0.55>
    gravityOverride {z=-980}
    physicsBodyDefinitions [{
      sphereCollisionRadius=2
      boundBone {boneName=dyn_coat_bk_02}
      boxExtents=<20 20 20>
      constraintSetup {angularLimitsMax=<0 0 2>, angularLimitsMin=<0 -20 -2>, angularTarget=<0 0 0>}
      localJointOffset=<5 0 0>
     }]
    relativeSpaceBone {boneName=root}
    retargetingSettings {bEnabled=true, sourceMaximum=52}
   }
   n30 = Anim Dynamics - Bone: dyn_coat_01_R!(ComponentPose=@n29, Alpha=1, GravityOverride:Vector="0,0,-980") -> (Pose) {
    angularDampingOverride=0.99
    angularSpringConstant=800
    bAngularSpring=true
    bOverrideAngularDamping=true
    bUseGravityOverride=true
    gravityScale=0.3
    simulationSpace=RootRelative
    alphaBoolBlend {blendOption=Linear}
    boundBone {boneName=dyn_coat_01_R}
    componentAppliedLinearAccClamp=<2000 2000 1600>
    componentLinearAccScale=<0.4 0.4 0.4>
    componentLinearVelScale=<0.45 0.45 0.55>
    gravityOverride {z=-980}
    physicsBodyDefinitions [{
      sphereCollisionRadius=2
      boundBone {boneName=dyn_coat_01_R}
      boxExtents=<30 30 30>
      constraintSetup {angularLimitsMax=<0 0 2>, angularLimitsMin=<0 -15 -2>, angularTarget=<0 0 0>}
      localJointOffset=<5 0 0>
     }]
    relativeSpaceBone {boneName=root}
    retargetingSettings {bEnabled=true, sourceMaximum=52}
   }
   n31 = Anim Dynamics - Bone: dyn_coat_02_R!(ComponentPose=@n30, Alpha=1, GravityOverride:Vector="0,0,-980") -> (Pose) {
    angularDampingOverride=0.99
    angularSpringConstant=800
    bAngularSpring=true
    bOverrideAngularDamping=true
    bUseGravityOverride=true
    gravityScale=0.3
    simulationSpace=RootRelative
    alphaBoolBlend {blendOption=Linear}
    boundBone {boneName=dyn_coat_02_R}
    componentAppliedLinearAccClamp=<2000 2000 1600>
    componentLinearAccScale=<0.4 0.4 0.4>
    componentLinearVelScale=<0.45 0.45 0.55>
    gravityOverride {z=-980}
    physicsBodyDefinitions [{
      sphereCollisionRadius=2
      boundBone {boneName=dyn_coat_02_R}
      boxExtents=<20 20 20>
      constraintSetup {angularLimitsMax=<0 0 2>, angularLimitsMin=<0 -20 -2>, angularTarget=<0 0 0>}
      localJointOffset=<5 0 0>
     }]
    relativeSpaceBone {boneName=root}
    retargetingSettings {bEnabled=true}
   }
   n32 = Anim Dynamics - Bone: dyn_coat_01_fR!(ComponentPose=@n31, Alpha=1, GravityOverride:Vector="0,0,-980", ExternalForce="0,0,0") -> (Pose) {
    angularDampingOverride=0.99
    angularSpringConstant=800
    bAngularSpring=true
    bOverrideAngularDamping=true
    bUseGravityOverride=true
    gravityScale=0.3
    simulationSpace=RootRelative
    alphaBoolBlend {blendOption=Linear}
    boundBone {boneName=dyn_coat_01_fR}
    componentAppliedLinearAccClamp=<2000 2000 1600>
    componentLinearAccScale=<0.4 0.4 0.4>
    componentLinearVelScale=<0.45 0.45 0.55>
    gravityOverride {z=-980}
    physicsBodyDefinitions [{
      sphereCollisionRadius=2
      boundBone {boneName=dyn_coat_01_fR}
      boxExtents=<30 30 30>
      constraintSetup {angularLimitsMax=<0 2 10>, angularLimitsMin=<0 -15 0>, angularTarget=<0 0 0>}
      localJointOffset=<5 0 0>
     }]
    relativeSpaceBone {boneName=root}
    retargetingSettings {bEnabled=true, source {rotation {pitch=-2}}}
   }
   n33 = RigidBody(
    ComponentPose=@n32
    OverrideWorldGravity="0,0,-980"
    SimSpaceSettings=(WorldAlpha=0,VelocityScaleZ=1,DampingAlpha=1,MaxLinearVelocity=10000,MaxAngularVelocity=10000,MaxLinearAcceleration=10000,MaxAngularAcceleration=10000,ExternalLinearDragV=(X=0,Y=0,Z=0),ExternalLinearVelocity=(X=0,Y=0,Z=0),ExternalAngularVelocity=(X=0,Y=0,Z=0))
   ) {
    alphaInputType=Bool
    bClampLinearTranslationLimitToRefPose=true
    bOverrideWorldGravity=true
    overridePhysicsAsset=&/F_MED_TweakSmirk_Coat_Physics
    simulationSpace=BaseBoneSpace
    alphaBoolBlend {blendOption=Linear}
    baseBoneRef {boneName=spine_05}
    componentAppliedLinearAccClamp=<3000 3000 2700>
    componentLinearAccScale=<0.01 0.4 0.35>
    componentLinearVelScale=<0.08 0.4 0.4>
    overrideWorldGravity {z=-980}
   }
   n34 = Anim Dynamics - Bone: dyn_coat_bk_04_L_01!(ComponentPose=@n33, Alpha=1, GravityOverride:Vector="0,0,-980", ExternalForce="0,0,0") -> (Pose) {
    angularDampingOverride=0.99
    angularSpringConstant=800
    bAngularSpring=true
    bOverrideAngularDamping=true
    bUseGravityOverride=true
    gravityScale=0.3
    simulationSpace=RootRelative
    alphaBoolBlend {blendOption=Linear}
    boundBone {boneName=dyn_coat_bk_04_L_01}
    componentAppliedLinearAccClamp=<2000 2000 1600>
    componentLinearAccScale=<0.4 0.4 0.4>
    componentLinearVelScale=<0.45 0.45 0.55>
    gravityOverride {z=-980}
    physicsBodyDefinitions [{
      sphereCollisionRadius=2
      boundBone {boneName=dyn_coat_bk_04_L_01}
      boxExtents=<4 8 5>
      constraintSetup {angularLimitsMax=<0 2 5>, angularLimitsMin=<0 -35 0>, angularTarget=<0 0 0>}
      localJointOffset=<3 0 0>
     }]
    relativeSpaceBone {boneName=root}
    retargetingSettings {bEnabled=true, source {rotation {pitch=-2}}}
   }
   n35 = Anim Dynamics - Bone: dyn_coat_bk_04_R_01!(ComponentPose=@n34, Alpha=1, GravityOverride:Vector="0,0,-980", ExternalForce="0,0,0") -> (Pose) {
    angularDampingOverride=0.99
    angularSpringConstant=800
    bAngularSpring=true
    bOverrideAngularDamping=true
    bUseGravityOverride=true
    gravityScale=0.3
    simulationSpace=RootRelative
    alphaBoolBlend {blendOption=Linear}
    boundBone {boneName=dyn_coat_bk_04_R_01}
    componentAppliedLinearAccClamp=<2000 2000 1600>
    componentLinearAccScale=<0.4 0.4 0.4>
    componentLinearVelScale=<0.45 0.45 0.55>
    gravityOverride {z=-980}
    physicsBodyDefinitions [{
      sphereCollisionRadius=2
      boundBone {boneName=dyn_coat_bk_04_R_01}
      boxExtents=<4 8 5>
      constraintSetup {angularLimitsMax=<0 2 5>, angularLimitsMin=<0 -35 0>, angularTarget=<0 0 0>}
      localJointOffset=<3 0 0>
     }]
    relativeSpaceBone {boneName=root}
    retargetingSettings {bEnabled=true, source {rotation {pitch=-2}}}
   }
   n36 = Anim Dynamics - Bone: dyn_coat_bk_04_L_02!(ComponentPose=@n35, Alpha=1, GravityOverride:Vector="0,0,-980", ExternalForce="0,0,0") -> (Pose) {
    angularDampingOverride=0.99
    angularSpringConstant=800
    bAngularSpring=true
    bOverrideAngularDamping=true
    bUseGravityOverride=true
    gravityScale=0.3
    simulationSpace=RootRelative
    alphaBoolBlend {blendOption=Linear}
    boundBone {boneName=dyn_coat_bk_04_L_02}
    componentAppliedLinearAccClamp=<2000 2000 1600>
    componentLinearAccScale=<0.4 0.4 0.4>
    componentLinearVelScale=<0.45 0.45 0.55>
    gravityOverride {z=-980}
    physicsBodyDefinitions [{
      sphereCollisionRadius=2
      boundBone {boneName=dyn_coat_bk_04_L_02}
      boxExtents=<8 12 10>
      constraintSetup {angularLimitsMax=<0 2 15>, angularLimitsMin=<0 -15 -15>, angularTarget=<0 0 0>}
      localJointOffset=<3 0 0>
     }]
    relativeSpaceBone {boneName=root}
    retargetingSettings {bEnabled=true, source {rotation {pitch=-2}}}
   }
   n37 = Anim Dynamics - Bone: dyn_coat_bk_04_R_02!(ComponentPose=@n36, Alpha=1, GravityOverride:Vector="0,0,-980", ExternalForce="0,0,0") -> (Pose) {
    angularDampingOverride=0.99
    angularSpringConstant=800
    bAngularSpring=true
    bOverrideAngularDamping=true
    bUseGravityOverride=true
    gravityScale=0.3
    simulationSpace=RootRelative
    alphaBoolBlend {blendOption=Linear}
    boundBone {boneName=dyn_coat_bk_04_R_02}
    componentAppliedLinearAccClamp=<2000 2000 1600>
    componentLinearAccScale=<0.4 0.4 0.4>
    componentLinearVelScale=<0.45 0.45 0.55>
    gravityOverride {z=-980}
    physicsBodyDefinitions [{
      sphereCollisionRadius=2
      boundBone {boneName=dyn_coat_bk_04_R_02}
      boxExtents=<8 12 10>
      constraintSetup {angularLimitsMax=<0 2 15>, angularLimitsMin=<0 -15 -15>, angularTarget=<0 0 0>}
      localJointOffset=<3 0 0>
     }]
    relativeSpaceBone {boneName=root}
    retargetingSettings {bEnabled=true, source {rotation {pitch=-2}}}
   }
   n38 = "Transform (Modify) Bone - Bone: base_sleeveTop_l"(ComponentPose=@n37, Alpha=1, Translation="0,0,0", Rotation="-2.543,-1,4.608", Scale="1,1,1") {
    rotationMode=BMM_Additive
    rotationSpace=BCS_BoneSpace
    translationMode=BMM_Additive
    translationSpace=BCS_BoneSpace
    alphaBoolBlend {blendOption=Linear}
    boneToModify {boneName=base_sleeveTop_l}
    rotation=<rot -2.543 -1 4.608>
   }
   n39 = "Transform (Modify) Bone - Bone: base_sleeveBottom_l"(ComponentPose=@n38, Alpha=1, Translation="-0.415,1.375,-0.405", Rotation="0,0,0", Scale="1,1,1") {
    rotationMode=BMM_Additive
    rotationSpace=BCS_BoneSpace
    translationMode=BMM_Additive
    translationSpace=BCS_BoneSpace
    alphaBoolBlend {blendOption=Linear}
    boneToModify {boneName=base_sleeveBottom_l}
    translation=<-0.415 1.375 -0.405>
   }
   n40 = "Transform (Modify) Bone - Bone: base_sleeveBack_l"(ComponentPose=@n39, Alpha=1, Translation="0,0,0", Rotation="-1,0,0", Scale="1,1,1") {
    rotationMode=BMM_Additive
    rotationSpace=BCS_BoneSpace
    translationMode=BMM_Additive
    translationSpace=BCS_BoneSpace
    alphaBoolBlend {blendOption=Linear}
    boneToModify {boneName=base_sleeveBack_l}
    rotation {pitch=-1}
   }
   n41 = "Transform (Modify) Bone - Bone: base_sleeveFront_l"(ComponentPose=@n40, Alpha=1, Translation="0,-0.5,0", Rotation="0,-3,0.161", Scale="1,1,1") {
    rotationMode=BMM_Additive
    rotationSpace=BCS_BoneSpace
    translationMode=BMM_Additive
    translationSpace=BCS_BoneSpace
    alphaBoolBlend {blendOption=Linear}
    boneToModify {boneName=base_sleeveFront_l}
    rotation {roll=0.161, yaw=-3}
    translation {y=-0.5}
   }
   n42 = "Transform (Modify) Bone - Bone: base_sleeveTop_r"(ComponentPose=@n41, Alpha=1, Translation="1.704,0.188,-0.083", Rotation="-0.960,6.851,12.920", Scale="1,1,1") {
    rotationMode=BMM_Additive
    rotationSpace=BCS_BoneSpace
    translationMode=BMM_Additive
    translationSpace=BCS_BoneSpace
    alphaBoolBlend {blendOption=Linear}
    boneToModify {boneName=base_sleeveTop_r}
    rotation=<rot -0.9599 6.851 12.92>
    translation=<1.704 0.188 -0.083>
   }
   n43 = "Transform (Modify) Bone - Bone: base_sleeveBottom_r"(ComponentPose=@n42, Alpha=1, Translation="0,0,0", Rotation="0,0,0", Scale="1,1,1") {
    rotationMode=BMM_Additive
    rotationSpace=BCS_BoneSpace
    translationMode=BMM_Additive
    translationSpace=BCS_BoneSpace
    alphaBoolBlend {blendOption=Linear}
    boneToModify {boneName=base_sleeveBottom_r}
   }
   n44 = "Transform (Modify) Bone - Bone: base_sleeveBack_r"(ComponentPose=@n43, Alpha=1, Translation="0,0,0", Rotation="0,0,0", Scale="1,1,1") {
    rotationMode=BMM_Additive
    rotationSpace=BCS_BoneSpace
    translationMode=BMM_Additive
    translationSpace=BCS_BoneSpace
    alphaBoolBlend {blendOption=Linear}
    boneToModify {boneName=base_sleeveBack_r}
   }
   n45 = "Transform (Modify) Bone - Bone: base_sleeveFront_r"(ComponentPose=@n44, Alpha=1, Translation="0,0.5,0", Rotation="-7.108,6.200,-0.989", Scale="1,1,1") {
    rotationMode=BMM_Additive
    rotationSpace=BCS_BoneSpace
    translationMode=BMM_Additive
    translationSpace=BCS_BoneSpace
    alphaBoolBlend {blendOption=Linear}
    boneToModify {boneName=base_sleeveFront_r}
    rotation=<rot -7.108 6.2 -0.989>
    translation {y=0.5}
   }
   n46 = Anim Dynamics - Bone: dyn_sleeveTop_l(ComponentPose=@n45, Alpha=1, ExternalForce="0,0,0") {
    angularDampingOverride=1
    angularSpringConstant=1000
    bAngularSpring=true
    bGravityOverrideInSimSpace=true
    bOverrideAngularDamping=true
    bUseGravityOverride=true
    gravityScale=0.35
    simulationSpace=BoneRelative
    alphaBoolBlend {blendOption=Linear}
    boundBone {boneName=dyn_sleeveTop_l}
    componentAppliedLinearAccClamp=<5 5 5>
    gravityOverride {z=-980}
    physicsBodyDefinitions [{
      boundBone {boneName=dyn_sleeveTop_l}
      boxExtents=<30 30 30>
      constraintSetup {angularLimitsMax=<5 0 2>, angularLimitsMin=<-5 0 -15>, angularTarget=<1 0 0>}
      localJointOffset=<2 0 0>
     }]
    relativeSpaceBone {boneName=lowerarm_twist_01_l}
    simSpaceSettings {maxAngularAcceleration=200, maxAngularVelocity=200, simSpaceAngularAlpha=0.1}
   }
   n47 = Anim Dynamics - Bone: dyn_sleeveBottom_l(ComponentPose=@n46, Alpha=1, ExternalForce="0,0,0") {
    angularDampingOverride=1
    angularSpringConstant=1000
    bAngularSpring=true
    bGravityOverrideInSimSpace=true
    bOverrideAngularDamping=true
    bUseGravityOverride=true
    gravityScale=0.35
    simulationSpace=BoneRelative
    alphaBoolBlend {blendOption=Linear}
    boundBone {boneName=dyn_sleeveBottom_l}
    componentAppliedLinearAccClamp=<5 5 5>
    gravityOverride {z=-980}
    physicsBodyDefinitions [{
      boundBone {boneName=dyn_sleeveBottom_l}
      boxExtents=<30 30 30>
      constraintSetup {angularLimitsMax=<5 2 3>, angularLimitsMin=<-5 -2 -2>, angularTarget=<1 0 0>}
      localJointOffset=<2 0 0>
     }]
    relativeSpaceBone {boneName=lowerarm_twist_01_l}
    simSpaceSettings {maxAngularAcceleration=200, maxAngularVelocity=200, simSpaceAngularAlpha=0.1}
   }
   n48 = Anim Dynamics - Bone: dyn_sleeveFront_l(ComponentPose=@n47, Alpha=1, ExternalForce="0,0,0") {
    angularDampingOverride=1
    angularSpringConstant=1000
    bAngularSpring=true
    bGravityOverrideInSimSpace=true
    bOverrideAngularDamping=true
    bUseGravityOverride=true
    gravityScale=0.35
    simulationSpace=BoneRelative
    alphaBoolBlend {blendOption=Linear}
    boundBone {boneName=dyn_sleeveFront_l}
    componentAppliedLinearAccClamp=<5 5 5>
    gravityOverride {z=-980}
    physicsBodyDefinitions [{
      boundBone {boneName=dyn_sleeveFront_l}
      boxExtents=<30 30 30>
      constraintSetup {angularLimitsMax=<5 0 0>, angularLimitsMin=<-5 0 -26>, angularTarget=<1 0 0>}
      localJointOffset=<2 0 0>
     }]
    relativeSpaceBone {boneName=lowerarm_twist_01_l}
    simSpaceSettings {maxAngularAcceleration=200, maxAngularVelocity=200, simSpaceAngularAlpha=0.1}
   }
   n49 = Anim Dynamics - Bone: dyn_sleeveBack_l(ComponentPose=@n48, Alpha=1, ExternalForce="0,0,0") {
    angularDampingOverride=1
    angularSpringConstant=1000
    bAngularSpring=true
    bGravityOverrideInSimSpace=true
    bOverrideAngularDamping=true
    bUseGravityOverride=true
    gravityScale=0.35
    simulationSpace=BoneRelative
    alphaBoolBlend {blendOption=Linear}
    boundBone {boneName=dyn_sleeveBack_l}
    componentAppliedLinearAccClamp=<5 5 5>
    gravityOverride {z=-980}
    physicsBodyDefinitions [{
      boundBone {boneName=dyn_sleeveBack_l}
      boxExtents=<30 30 30>
      constraintSetup {angularLimitsMax=<1 2 -2>, angularLimitsMin=<-1 -2 -14>, angularTarget=<1 0 0>}
      localJointOffset=<2 0 0>
     }]
    relativeSpaceBone {boneName=lowerarm_twist_01_l}
    simSpaceSettings {maxAngularAcceleration=200, maxAngularVelocity=200, simSpaceAngularAlpha=0.1}
   }
   n50 = Anim Dynamics - Bone: dyn_sleeveTop_r(ComponentPose=@n49, Alpha=1, ExternalForce="0,0,0") {
    angularDampingOverride=1
    angularSpringConstant=1000
    bAngularSpring=true
    bOverrideAngularDamping=true
    bUseGravityOverride=true
    gravityScale=0.35
    simulationSpace=BoneRelative
    alphaBoolBlend {blendOption=Linear}
    boundBone {boneName=dyn_sleeveTop_r}
    componentAppliedLinearAccClamp=<5 5 5>
    gravityOverride {z=-980}
    physicsBodyDefinitions [{
      boundBone {boneName=dyn_sleeveTop_r}
      boxExtents=<30 30 30>
      constraintSetup {angularLimitsMax=<5 0 2>, angularLimitsMin=<-5 0 -13>, angularTarget=<1 0 0>}
      localJointOffset=<-2 0 0>
     }]
    relativeSpaceBone {boneName=lowerarm_twist_01_r}
    simSpaceSettings {maxAngularAcceleration=200, maxAngularVelocity=200, simSpaceAngularAlpha=0.1}
   }
   n51 = Anim Dynamics - Bone: dyn_sleeveBottom_r(ComponentPose=@n50, Alpha=1, ExternalForce="0,0,0") {
    angularDampingOverride=1
    angularSpringConstant=1000
    bAngularSpring=true
    bOverrideAngularDamping=true
    bUseGravityOverride=true
    gravityScale=0.35
    simulationSpace=BoneRelative
    alphaBoolBlend {blendOption=Linear}
    boundBone {boneName=dyn_sleeveBottom_r}
    componentAppliedLinearAccClamp=<5 5 5>
    gravityOverride {z=-980}
    physicsBodyDefinitions [{
      boundBone {boneName=dyn_sleeveBottom_r}
      boxExtents=<30 30 30>
      constraintSetup {angularLimitsMax=<5 2 3>, angularLimitsMin=<-5 -2 -2>, angularTarget=<1 0 0>}
      localJointOffset=<-2 0 0>
     }]
    relativeSpaceBone {boneName=lowerarm_twist_01_r}
    simSpaceSettings {maxAngularAcceleration=200, maxAngularVelocity=200, simSpaceAngularAlpha=0.1}
   }
   n52 = Anim Dynamics - Bone: dyn_sleeveFront_r(ComponentPose=@n51, Alpha=1, ExternalForce="0,0,0") {
    angularDampingOverride=1
    angularSpringConstant=1000
    bAngularSpring=true
    bOverrideAngularDamping=true
    bUseGravityOverride=true
    gravityScale=0.35
    simulationSpace=BoneRelative
    alphaBoolBlend {blendOption=Linear}
    boundBone {boneName=dyn_sleeveFront_r}
    componentAppliedLinearAccClamp=<5 5 5>
    gravityOverride {z=-980}
    physicsBodyDefinitions [{
      boundBone {boneName=dyn_sleeveFront_r}
      boxExtents=<30 30 30>
      constraintSetup {angularLimitsMax=<8 0 0>, angularLimitsMin=<-8 0 -20>, angularTarget=<1 0 0>}
      localJointOffset=<-2 0 0>
     }]
    relativeSpaceBone {boneName=lowerarm_twist_01_r}
    simSpaceSettings {maxAngularAcceleration=200, maxAngularVelocity=200, simSpaceAngularAlpha=0.1}
   }
   n53 = Anim Dynamics - Bone: dyn_sleeveBack_r(ComponentPose=@n52, Alpha=1, ExternalForce="0,0,0") {
    angularDampingOverride=1
    angularSpringConstant=1000
    bAngularSpring=true
    bOverrideAngularDamping=true
    bUseGravityOverride=true
    gravityScale=0.35
    simulationSpace=BoneRelative
    alphaBoolBlend {blendOption=Linear}
    boundBone {boneName=dyn_sleeveBack_r}
    componentAppliedLinearAccClamp=<5 5 5>
    gravityOverride {z=-980}
    physicsBodyDefinitions [{
      boundBone {boneName=dyn_sleeveBack_r}
      boxExtents=<30 30 30>
      constraintSetup {angularLimitsMax=<1 2 2>, angularLimitsMin=<-1 -2 -14>, angularTarget=<1 0 0>}
      localJointOffset=<-2 0 0>
     }]
    relativeSpaceBone {boneName=lowerarm_twist_01_r}
    simSpaceSettings {maxAngularAcceleration=200, maxAngularVelocity=200, simSpaceAngularAlpha=0.1}
   }
   Save cached pose 'AnimDyn_Pose'(Pose=Component To Local(ComponentPose=@n53)) {cacheName=AnimDyn_Pose}
  ]}
 EventGraph {nodes [
   Event BlueprintUpdateAnimation()
   > n1 = Sequence!(execute) -> (then_0, then_1, then_2:Exec)
   > n2 = "Gravity Override State Function (Parameter Asset) (Nativized)"(
    CharacterPartAnimInstance=Self-Reference()
    GameTimeInSeconds=GetGameTimeInSeconds()
    Parameters=&/F_MED_TweakSmirk_Parameters
   )
   > Set Gravity Override Cape(Gravity Override Cape:Vector=@n2) -> (Output_Get:Vector)
   > Set RBAN Settings Out_0(RBAN Settings Out_0:Fort Rigid Body Anim Node Parameters Structure=@n2.OutRigidBodyAnimNodeParameters) -> (Output_Get:Fort Rigid Body Anim Node Parameters Structure)
   Set DBNO(execute=@n1.then_1, DBNO:bool=Get bIsDBNO()) -> (Output_Get:bool)
   Event OnCharacterCustomizationLoaded()
   n3 = CheckForBackbling()
   > Set isWearingPackBool(isWearingPackBool:bool=OR Boolean!(A=@n3, B=@n3.bIsBackblingACape, C:bool=@n3.bIsBackblingLong) -> (ReturnValue)) -> (Output_Get:bool)
   > Set is_WearingCape(is_WearingCape:bool=@n3.bIsBackblingLong) -> (Output_Get:bool)
   AND Boolean(A=@n3, B=NOT Boolean(A=@n3.bIsBackblingLong))
  ]}
]
variables [
 DBNO {type=boolean}
 Gravity Override Cape {type=Vector}
 is_WearingCape {type=boolean}
 isWearingPackBool {type=boolean}
 RBAN Settings Out_0 {type=FortRigidBodyAnimNodeParameters}
]
