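# Control Rig for the F_MED_TweakSmirk face accessory mesh that adjusts the back hair as the head moves.
# A spherical pose reader measures the head bone's orientation relative to spine_05, using an offset, asymmetric active region with falloff.
# Its output weights a Modify Transforms step that rotates and shifts dyn_Hair_Back_Base, carrying the child back-hair chains along.
# The front hair chains are in the hierarchy but are not driven here.
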ControlRigBlueprint /BRCosmetics/Characters/Player/Female/Medium/Bodies/F_MED_TweakSmirk/Meshes/Parts/F_MED_TweakSmirk_FaceAcc_CtrlRig
graphs [
 RigVMFunctionLibrary {links [], nodes []}
 RigVMModel {nodes [
   RigUnit_BeginExecution = &RigUnit_BeginExecution()
   n1 = &RigUnit_SphericalPoseReader(
    ActiveRegionScaleFactors=(PositiveWidth=5,NegativeWidth=2,PositiveHeight=5,NegativeHeight=5)
    ActiveRegionSize=0.1
    Debug=(bDrawDebug=false,DebugScale=25,DebugSegments=20,DebugThickness=0.25)
    DriverAxis=(X=1,Y=1,Z=0)
    DriverItem=(Type=Bone,Name="head")
    FalloffRegionScaleFactors=(PositiveWidth=1.5,NegativeWidth=0.1,PositiveHeight=2,NegativeHeight=2)
    FalloffSize=0.6
    FlipHeightScaling=true
    FlipWidthScaling=false
    OptionalParentItem=(Type=Bone,Name="spine_05")
    RotationOffset=(X=-30,Y=0,Z=-10)
    ExecutePin=@RigUnit_BeginExecution.ExecutePin
   )
   > &RigUnit_ModifyTransforms(
    ItemToModify=((Item=(Type=Bone,Name="dyn_Hair_Back_Base"),Transform=(Rotation=(X=0,Y=0,Z=0.1736,W=0.9848),Translation=(X=-1,Y=1,Z=0),Scale3D=(X=1,Y=1,Z=1))))
    Weight=@n1.OutputParam
   )
  ]}
]
hierarchy [
 root {loc=<0 0 0>, rot=<0 0 0 1>}
 pelvis {parent=root, loc=<0 -0.5301 98.0571>, rot=<0.0127 0.707 -0.0127 -0.707>}
 spine_01 {parent=pelvis, loc=<6.1208 -0.9328 0>, rot=<0 0 0.034 -0.9994>}
 spine_02 {parent=spine_01, loc=<6.5997 -0.2095 0>, rot=<0 0 0.0238 -0.9997>}
 spine_03 {parent=spine_02, loc=<6.4914 -0.1237 0>, rot=<0 0 0.0149 -0.9999>}
 spine_04 {parent=spine_03, loc=<6.9598 0.0794 0>, rot=<0 0 -0.0132 -0.9999>}
 spine_05 {parent=spine_04, loc=<6.9987 0 0>, rot=<0 0 -0.0015 -1>}
 neck_01 {parent=spine_05, loc=<11.3138 0.1529 0>, rot=<0 0 -0.2003 -0.9797>}
 neck_02 {parent=neck_01, loc=<4.5627 0 0>, rot=<0 0 0.0504 -0.9987>}
 head {parent=neck_02, loc=<5.3914 0 0>, rot=<0 0 0.115 -0.9934>}
 dyn_Hair_Fn_l_01 {parent=head, loc=<-1.0461 1.5293 -6.5822>, rot=<0.0341 -0.0691 0.7031 -0.7069>}
 dyn_Hair_Fn_l_02 {parent=dyn_Hair_Fn_l_01, loc=<0 -6.5072 0>, rot=<0 0 0.1732 -0.9849>}
 dyn_Hair_Fn_l_03 {parent=dyn_Hair_Fn_l_02, loc=<0 -7.2735 0>, rot=<0 0 0.1852 -0.9827>}
 dyn_Hair_Fn_l_04 {parent=dyn_Hair_Fn_l_03, loc=<0.092 -5.8854 0>, rot=<-0.1126 0.0122 0.1067 -0.9878>}
 dyn_Hair_Fn_r_01 {parent=head, loc=<-1.0459 1.5293 6.5822>, rot=<0.0691 0.0341 -0.7069 -0.7031>}
 dyn_Hair_Fn_r_02 {parent=dyn_Hair_Fn_r_01, loc=<0 6.5071 0>, rot=<0 0 0.1732 -0.9849>}
 dyn_Hair_Fn_r_03 {parent=dyn_Hair_Fn_r_02, loc=<-0.0001 7.2738 0.0001>, rot=<0 0 0.1852 -0.9827>}
 dyn_Hair_Fn_r_04 {parent=dyn_Hair_Fn_r_03, loc=<-0.0918 5.8852 -0.0001>, rot=<-0.1126 0.0122 0.1067 -0.9878>}
 dyn_Hair_Back_Base {parent=head, loc=<2.347 -8.6377 0>, rot=<0 0 0.7082 -0.7061>}
 dyn_Hair_Back_02 {parent=dyn_Hair_Back_Base, loc=<5.4794 1.5907 0>, rot=<0 0 -0.196 -0.9806>}
 dyn_Hair_Back_03 {parent=dyn_Hair_Back_02, loc=<1.4313 -1.7375 0>, rot=<0 0 0.1419 -0.9899>}
 dyn_Hair_Back_04 {parent=dyn_Hair_Back_03, loc=<0 -4.1014 -0.3554>, rot=<0 0 0 1>}
 dyn_Hair_Back_05 {parent=dyn_Hair_Back_04, loc=<0 -4.6882 0>, rot=<0 0 -0.0897 -0.996>}
 dyn_Hair_Back_06 {parent=dyn_Hair_Back_05, loc=<0.1033 -5.3508 0>, rot=<0 0 0.1325 -0.9912>}
 dyn_Hair_Back_07 {parent=dyn_Hair_Back_06, loc=<0 -5.8708 0>, rot=<0 0 0.3054 -0.9522>}
 dyn_Hair_Back_06a_01 {parent=dyn_Hair_Back_06, loc=<0.8071 0.5977 2.6785>, rot=<0 0 0.0576 -0.9983>}
 dyn_Hair_Back_06a_02 {parent=dyn_Hair_Back_06a_01, loc=<0 -4.1332 0>, rot=<0 0 0 1>}
 dyn_Hair_Back_06b_01 {parent=dyn_Hair_Back_06, loc=<-2.2634 -2.0345 -0.7844>, rot=<-0.1239 -0.038 0.2304 -0.9644>}
 dyn_Hair_Back_06b_02 {parent=dyn_Hair_Back_06b_01, loc=<0 -4.1332 0>, rot=<0 0 -0.3175 -0.9483>}
 dyn_Hair_Back_05a_01 {parent=dyn_Hair_Back_05, loc=<0.1043 -0.8956 -2.2047>, rot=<-0.1667 0 0 -0.986>}
 dyn_Hair_Back_05a_02 {parent=dyn_Hair_Back_05a_01, loc=<0.1033 -5.3508 0>, rot=<0.0707 0.0031 -0.0431 -0.9966>}
 dyn_Hair_Back_05a_03 {parent=dyn_Hair_Back_05a_02, loc=<0 -5.0975 0>, rot=<0 0 0 1>}
 dyn_Hair_Back_04a_01 {parent=dyn_Hair_Back_04, loc=<0.5286 -0.2995 1.7524>, rot=<0.1591 -0.0133 -0.1551 -0.9749>}
 dyn_Hair_Back_04a_02 {parent=dyn_Hair_Back_04a_01, loc=<0 -4.6882 0>, rot=<0 0 -0.0005 -1>}
 dyn_Hair_Back_04a_03 {parent=dyn_Hair_Back_04a_02, loc=<0.1033 -5.3508 0>, rot=<0 0 0.1325 -0.9912>}
 dyn_Hair_Back_La_base {parent=dyn_Hair_Back_Base, loc=<-0.8701 0.7211 -6.2083>, rot=<-0.4426 0.0157 0.0318 -0.896>}
 dyn_Hair_Back_La_01 {parent=dyn_Hair_Back_La_base, loc=<0 -0.4588 0>, rot=<0 0 0 1>}
 dyn_Hair_Back_La_02 {parent=dyn_Hair_Back_La_01, loc=<0 -6.1016 0>, rot=<0 0 0 1>}
 dyn_Hair_Back_L_base {parent=dyn_Hair_Back_Base, loc=<-1.9636 -3.1661 -3.386>, rot=<-0.2489 0.1746 -0.0386 -0.9519>}
 dyn_Hair_Back_L_01 {parent=dyn_Hair_Back_L_base, loc=<0 -0.4985 0>, rot=<0 0 0 1>}
 dyn_Hair_Back_L_02 {parent=dyn_Hair_Back_L_01, loc=<0 -7.8236 0>, rot=<0 0 0 1>}
 dyn_Hair_Back_R_base {parent=dyn_Hair_Back_Base, loc=<-1.933 -3.3324 3.4738>, rot=<0.1746 0.2489 0.9519 -0.0386>}
 dyn_Hair_Back_R_01 {parent=dyn_Hair_Back_R_base, loc=<0 0.3078 0>, rot=<0 0 0 1>}
 dyn_Hair_Back_R_02 {parent=dyn_Hair_Back_R_01, loc=<0 7.824 -0.0002>, rot=<0 0 0 -1>}
]
