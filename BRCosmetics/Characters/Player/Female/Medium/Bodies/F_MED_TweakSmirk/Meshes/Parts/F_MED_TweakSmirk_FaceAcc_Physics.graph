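# Physics asset for the TweakSmirk female medium face accessory mesh, defining collision and joints for dynamic hair bones.
# Capsules on the head, upper spine, clavicles and upper arms form body volumes around the hair chains.
# Two front strands and a branching back chain hang from the head with limited swing and tight twist, mostly loosening toward the tips.
# Left and right back side pieces use free-rotation base joints followed by twist-dominant limited joints.
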PhysicsAsset /BRCosmetics/Characters/Player/Female/Medium/Bodies/F_MED_TweakSmirk/Meshes/Parts/F_MED_TweakSmirk_FaceAcc_Physics
bodies {
 clavicle_l [{
   shapeType=Capsule
   length=3.2691
   radius=7.0665
   shapeName=dyn_Hair_Back_R_01_capsule
   center=<10.4318 -0.2735 -3.1544>
   rotation=<rot 85.6257 -0.0003 -0.0004>
  }]
 clavicle_r [{
   shapeType=Capsule
   length=3.2691
   radius=7.0665
   shapeName=dyn_Hair_Back_R_01_capsule
   center=<-9.7073 0 2.6832>
   rotation=<rot 85.6257 -0.0003 -0.0004>
  }]
 dyn_Hair_Back_02 [{shapeType=Capsule, length=2, radius=1.5, shapeName=dyn_Hair_Back_02_capsule, center=<0 -0.9942 0.1343>, rotation=<rot 0 -0.0001 97.6927>}]
 dyn_Hair_Back_03 [{
   shapeType=Capsule
   length=3.6422
   radius=1.5
   shapeName=dyn_Hair_Back_03_capsule
   center=<0 -0.9942 0.1343>
   rotation=<rot 0 -0.0001 97.6927>
  }]
 dyn_Hair_Back_04 [{
   shapeType=Capsule
   length=5.7057
   radius=1.5
   shapeName=dyn_Hair_Back_04_capsule
   center=<0 -0.9942 0.1343>
   rotation=<rot 0 -0.0001 97.6927>
  }]
 dyn_Hair_Back_04a_01 [{
   shapeType=Capsule
   length=3.0948
   radius=1.5
   shapeName=dyn_Hair_Back_04a_01_capsule
   center=<0 -0.9942 0.1343>
   rotation=<rot 0 -0.0001 97.6927>
  }]
 dyn_Hair_Back_04a_02 [{
   shapeType=Capsule
   length=2.4215
   radius=1.5
   shapeName=dyn_Hair_Back_04a_02_capsule
   center=<0 -0.9942 0.1343>
   rotation=<rot 0 -0.0001 97.6927>
  }]
 dyn_Hair_Back_04a_03 [{
   shapeType=Capsule
   length=2.8877
   radius=1.5
   shapeName=dyn_Hair_Back_04a_03_capsule
   center=<0 -0.9942 0.1343>
   rotation=<rot 0 -0.0001 97.6927>
  }]
 dyn_Hair_Back_05 [{
   shapeType=Capsule
   length=6.4294
   radius=1.5
   shapeName=dyn_Hair_Back_05_capsule
   center=<0 -0.9942 0.1343>
   rotation=<rot 0 -0.0001 97.6927>
  }]
 dyn_Hair_Back_05a_01 [{
   shapeType=Capsule
   length=2.3873
   radius=1.5
   shapeName=dyn_Hair_Back_05a_01_capsule
   center=<0 -0.9942 0.1343>
   rotation=<rot 0 -0.0001 97.6927>
  }]
 dyn_Hair_Back_05a_02 [{
   shapeType=Capsule
   length=2.22
   radius=1.5
   shapeName=dyn_Hair_Back_05a_02_capsule
   center=<0 -0.9942 0.1343>
   rotation=<rot 0 -0.0001 97.6927>
  }]
 dyn_Hair_Back_05a_03 [{
   shapeType=Capsule
   length=2.2783
   radius=1.5
   shapeName=dyn_Hair_Back_05a_03_capsule
   center=<0 -0.9942 0.1343>
   rotation=<rot 0 -0.0001 97.6927>
  }]
 dyn_Hair_Back_06 [{
   shapeType=Capsule
   length=4.9787
   radius=1.5
   shapeName=dyn_Hair_Back_06_capsule
   center=<0 -0.9942 0.1343>
   rotation=<rot 0 -0.0001 97.6927>
  }]
 dyn_Hair_Back_06a_01 [{
   shapeType=Capsule
   length=2.1942
   radius=1.5
   shapeName=dyn_Hair_Back_06a_01_capsule
   center=<0 -0.9942 0.1343>
   rotation=<rot 0 -0.0001 97.6927>
  }]
 dyn_Hair_Back_06a_02 [{
   shapeType=Capsule
   length=2.6984
   radius=1.5
   shapeName=dyn_Hair_Back_06a_02_capsule
   center=<0 -0.9942 0.1343>
   rotation=<rot 0 -0.0001 97.6927>
  }]
 dyn_Hair_Back_06b_01 [{
   shapeType=Capsule
   length=2.2596
   radius=1.5
   shapeName=dyn_Hair_Back_06b_01_capsule
   center=<0 -0.9942 0.1343>
   rotation=<rot 0 -0.0001 97.6927>
  }]
 dyn_Hair_Back_06b_02 [{shapeType=Capsule, length=2, radius=1.5, shapeName=dyn_Hair_Back_06b_02_capsule, center=<0 -0.9942 0.1343>, rotation=<rot 0 -0.0001 97.6927>}]
 dyn_Hair_Back_07 [{
   shapeType=Capsule
   length=3.4248
   radius=1.5
   shapeName=dyn_Hair_Back_07_capsule
   center=<0 -0.9942 0.1343>
   rotation=<rot 0 -0.0001 97.6927>
  }]
 dyn_Hair_Back_Base [{
   shapeType=Capsule
   length=8.3626
   radius=8.4906
   shapeName=dyn_Hair_Back_Base_capsule
   center=<-3.6028 4.7176 -0.62>
   rotation=<rot 0 -41.3762 -176.9706>
  }]
 dyn_Hair_Back_L_01 [{
   shapeType=Capsule
   length=2.4916
   radius=5.1693
   shapeName=dyn_Hair_Back_L_01_capsule
   center=<0.2434 -4.3208 -2.2992>
   rotation=<rot 1.3149 -164.5437 -90.1226>
  }]
 dyn_Hair_Back_L_base [{shapeType=Capsule, length=0.505, radius=0.505, shapeName=dyn_Hair_Back_L_base_capsule, rotation=<rot 0 -90 0>}]
 dyn_Hair_Back_La_01 [{
   shapeType=Capsule
   length=4.9916
   radius=2.6693
   shapeName=dyn_Hair_Back_L_01_capsule
   center=<0.1301 -1.9057 -0.3447>
   rotation=<rot 1.3149 -164.5437 -90.1226>
  }]
 dyn_Hair_Back_La_base [{shapeType=Capsule, length=0.505, radius=0.505, shapeName=dyn_Hair_Back_La_base_capsule, rotation=<rot 0 -90 0>}]
 dyn_Hair_Back_R_01 [{
   shapeType=Capsule
   length=0
   radius=7.5665
   shapeName=dyn_Hair_Back_R_01_capsule
   center=<-0.6714 3.8602 5.2612>
   rotation=<rot 3.9963 -25.7686 -93.2563>
  }]
 dyn_Hair_Back_R_base [{shapeType=Capsule, length=0.505, radius=0.505, shapeName=dyn_Hair_Back_R_base_capsule, rotation=<rot 0 -90 0>}]
 dyn_Hair_Fn_l_01 [{
   shapeType=Capsule
   length=5.3772
   radius=1
   shapeName=dyn_Hair_Fn_l_01_capsule
   center=<-0.0003 -2.5171 -0.5625>
   rotation=<rot -0.001 -0.0071 -83.617>
  }]
 dyn_Hair_Fn_l_02 [{
   shapeType=Capsule
   length=5.1679
   radius=1
   shapeName=dyn_Hair_Fn_l_02_capsule
   center=<-0.0003 -2.4238 0.2712>
   rotation=<rot -0.0011 -0.0071 -83.6166>
  }]
 dyn_Hair_Fn_l_03 [{
   shapeType=Capsule
   length=5
   radius=1
   shapeName=dyn_Hair_Fn_l_03_capsule
   center=<-0.0003 -2.4238 0.2712>
   rotation=<rot -0.001 -0.0071 -83.6162>
  }]
 dyn_Hair_Fn_l_04 [{
   shapeType=Capsule
   length=5.2855
   radius=1.5
   shapeName=dyn_Hair_Fn_l_04_capsule
   center=<-0.0003 -2.4238 0.2712>
   rotation=<rot 0.0007 -0.0071 -83.616>
  }]
 dyn_Hair_Fn_r_01 [{
   shapeType=Capsule
   length=5.1117
   radius=1
   shapeName=dyn_Hair_Fn_r_01_capsule
   center=<0.0008 2.3177 -0.256>
   rotation=<rot -0.0019 -0.0132 96.3436>
  }]
 dyn_Hair_Fn_r_02 [{
   shapeType=Capsule
   length=5.2643
   radius=1
   shapeName=dyn_Hair_Fn_r_02_capsule
   center=<0.0008 2.3177 -0.256>
   rotation=<rot -0.002 -0.0131 96.3443>
  }]
 dyn_Hair_Fn_r_03 [{
   shapeType=Capsule
   length=5
   radius=1
   shapeName=dyn_Hair_Fn_r_03_capsule
   center=<0.0008 2.3177 -0.2561>
   rotation=<rot -0.0019 -0.013 96.345>
  }]
 dyn_Hair_Fn_r_04 [{
   shapeType=Capsule
   length=5.2855
   radius=1.5
   shapeName=dyn_Hair_Fn_l_04_capsule
   center=<0.4495 2.3044 0.3273>
   rotation=<rot -0.0365 2.8049 -94.8669>
  }]
 head [{shapeType=Capsule, length=6.5292, radius=6.1732, center=<0.4437 -0.561 0.0321>, rotation=<rot 1.4726 -88.5401 -92.3223>}]
 spine_04 [{shapeType=Capsule, length=11, radius=14, shapeName=spine_05_capsule, center=<-2.4381 1.3438 0>, rotation=<rot 0 -90 0>}]
 spine_05 [{shapeType=Capsule, length=18, radius=10, shapeName=spine_05_capsule, center=<0.9784 -2.3583 0>, rotation=<rot 0 -90 0>}]
 upperarm_l [{
   shapeType=Capsule
   length=25.7691
   radius=7.0665
   shapeName=dyn_Hair_Back_R_01_capsule
   center=<13.5258 0 -1.0346>
   rotation=<rot 85.6257 -0.0003 -0.0004>
  }]
 upperarm_r [{
   shapeType=Capsule
   length=25.7691
   radius=7.0665
   shapeName=dyn_Hair_Back_R_01_capsule
   center=<-12.1626 0 0.9304>
   rotation=<rot 85.6257 -0.0003 -0.0004>
  }]
}
constraints [
 {bone1Name=dyn_Hair_Back_Base, bone2Name=head, swing1LimitDegrees=35, swing2LimitDegrees=1, twistLimitDegrees=1}
 {bone1Name=dyn_Hair_Back_02, bone2Name=dyn_Hair_Back_Base, swing1LimitDegrees=10, swing2LimitDegrees=5, twistLimitDegrees=1}
 {bone1Name=dyn_Hair_Back_03, bone2Name=dyn_Hair_Back_02, swing1LimitDegrees=15, swing2LimitDegrees=5, twistLimitDegrees=1}
 {bone1Name=dyn_Hair_Back_04, bone2Name=dyn_Hair_Back_03, swing1LimitDegrees=20, swing2LimitDegrees=15, twistLimitDegrees=1}
 {bone1Name=dyn_Hair_Back_05, bone2Name=dyn_Hair_Back_04, swing1LimitDegrees=25, swing2LimitDegrees=15, twistLimitDegrees=1}
 {bone1Name=dyn_Hair_Back_06, bone2Name=dyn_Hair_Back_05, swing1LimitDegrees=35, swing2LimitDegrees=15, twistLimitDegrees=1}
 {bone1Name=dyn_Hair_Back_07, bone2Name=dyn_Hair_Back_06, swing1LimitDegrees=35, swing2LimitDegrees=15, twistLimitDegrees=1}
 {bone1Name=dyn_Hair_Back_06a_01, bone2Name=dyn_Hair_Back_06, swing1LimitDegrees=35, swing2LimitDegrees=5, twistLimitDegrees=1}
 {bone1Name=dyn_Hair_Back_06a_02, bone2Name=dyn_Hair_Back_06a_01, swing1LimitDegrees=35, swing2LimitDegrees=5, twistLimitDegrees=1}
 {bone1Name=dyn_Hair_Back_06b_01, bone2Name=dyn_Hair_Back_06, swing1LimitDegrees=35, swing2LimitDegrees=5, twistLimitDegrees=1}
 {bone1Name=dyn_Hair_Back_06b_02, bone2Name=dyn_Hair_Back_06b_01, swing1LimitDegrees=35, swing2LimitDegrees=5, twistLimitDegrees=1}
 {bone1Name=dyn_Hair_Back_05a_01, bone2Name=dyn_Hair_Back_05, swing1LimitDegrees=35, swing2LimitDegrees=5, twistLimitDegrees=1}
 {bone1Name=dyn_Hair_Back_05a_02, bone2Name=dyn_Hair_Back_05a_01, swing1LimitDegrees=35, swing2LimitDegrees=5, twistLimitDegrees=1}
 {bone1Name=dyn_Hair_Back_05a_03, bone2Name=dyn_Hair_Back_05a_02, swing1LimitDegrees=35, swing2LimitDegrees=5, twistLimitDegrees=1}
 {bone1Name=dyn_Hair_Back_04a_01, bone2Name=dyn_Hair_Back_04, swing1LimitDegrees=35, swing2LimitDegrees=5, twistLimitDegrees=1}
 {bone1Name=dyn_Hair_Back_04a_02, bone2Name=dyn_Hair_Back_04a_01, swing1LimitDegrees=35, swing2LimitDegrees=5, twistLimitDegrees=1}
 {bone1Name=dyn_Hair_Back_04a_03, bone2Name=dyn_Hair_Back_04a_02, swing1LimitDegrees=35, swing2LimitDegrees=5, twistLimitDegrees=1}
 {bone1Name=dyn_Hair_Back_La_01, bone2Name=dyn_Hair_Back_La_base, swing1LimitDegrees=5, swing2LimitDegrees=5, twistLimitDegrees=60}
 {bone1Name=dyn_Hair_Back_R_01, bone2Name=dyn_Hair_Back_R_base, swing1LimitDegrees=5, swing2LimitDegrees=5, twistLimitDegrees=45}
 {
  bone1Name=dyn_Hair_Back_R_base
  bone2Name=dyn_Hair_Back_Base
  swing1LimitDegrees=45
  swing1Motion=Free
  swing2LimitDegrees=45
  swing2Motion=Free
  twistLimitDegrees=45
  twistMotion=Free
 }
 {
  bone1Name=dyn_Hair_Back_La_base
  bone2Name=dyn_Hair_Back_Base
  swing1LimitDegrees=45
  swing1Motion=Free
  swing2LimitDegrees=45
  swing2Motion=Free
  twistLimitDegrees=45
  twistMotion=Free
 }
 {bone1Name=dyn_Hair_Back_L_01, bone2Name=dyn_Hair_Back_L_base, swing1LimitDegrees=5, swing2LimitDegrees=5, twistLimitDegrees=45}
 {
  bone1Name=dyn_Hair_Back_L_base
  bone2Name=dyn_Hair_Back_Base
  swing1LimitDegrees=45
  swing1Motion=Free
  swing2LimitDegrees=45
  swing2Motion=Free
  twistLimitDegrees=45
  twistMotion=Free
 }
 {bone1Name=dyn_Hair_Fn_l_01, bone2Name=head, swing1LimitDegrees=37, swing2LimitDegrees=10, twistLimitDegrees=2}
 {bone1Name=dyn_Hair_Fn_l_02, bone2Name=dyn_Hair_Fn_l_01, swing1LimitDegrees=40, swing2LimitDegrees=20, twistLimitDegrees=2}
 {bone1Name=dyn_Hair_Fn_l_03, bone2Name=dyn_Hair_Fn_l_02, swing1LimitDegrees=45, swing2LimitDegrees=25, twistLimitDegrees=2}
 {bone1Name=dyn_Hair_Fn_l_04, bone2Name=dyn_Hair_Fn_l_03, swing1LimitDegrees=45, swing2LimitDegrees=25, twistLimitDegrees=2}
 {bone1Name=dyn_Hair_Fn_r_01, bone2Name=head, swing1LimitDegrees=37, swing2LimitDegrees=10, twistLimitDegrees=2}
 {bone1Name=dyn_Hair_Fn_r_02, bone2Name=dyn_Hair_Fn_r_01, swing1LimitDegrees=40, swing2LimitDegrees=20, twistLimitDegrees=2}
 {bone1Name=dyn_Hair_Fn_r_03, bone2Name=dyn_Hair_Fn_r_02, swing1LimitDegrees=45, swing2LimitDegrees=25, twistLimitDegrees=2}
 {bone1Name=dyn_Hair_Fn_r_04, bone2Name=dyn_Hair_Fn_r_03, swing1LimitDegrees=45, swing2LimitDegrees=25, twistLimitDegrees=2}
]
