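# Post-process style AnimBP for the F_MED_TweakSmirk face accessory that adds secondary motion to its hair.
# Copied parent pose runs through Control Rig and chained front-hair AnimDynamics into a cached pose, then neck-relative RigidBody simulation.
# Backbling detection on customization load selects a pack-specific physics asset with back-hair bone offsets, or standard physics.
# Updates compute gravity and rigid-body settings from borrowed LoanFloat parameters, which the AnimGraph never reads.
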AnimBlueprint /BRCosmetics/Characters/Player/Female/Medium/Bodies/F_MED_TweakSmirk/Meshes/Parts/F_MED_TweakSmirk_FaceAcc_AnimBP
target_skeleton=&/F_MED_TweakSmirk_FaceAcc_Skeleton
graphs [
 AnimGraph {nodes [
   n1 = "Transform (Modify) Bone - Bone: dyn_Hair_Back_02"(ComponentPose=Local To Component(LocalPose=Use cached pose 'hair pose'()), Alpha=1, Translation="1,-1,0", Rotation="0,10,0", Scale="1,1,1") {
    rotationMode=BMM_Additive
    rotationSpace=BCS_BoneSpace
    scaleMode=BMM_Additive
    scaleSpace=BCS_BoneSpace
    translationMode=BMM_Additive
    translationSpace=BCS_BoneSpace
    alphaBoolBlend {blendInTime=0.2, blendOutTime=0.2}
    boneToModify {boneName=dyn_Hair_Back_02}
    rotation {yaw=10}
    translation=<1 -1>
   }
   n2 = RigidBody!(
    ComponentPose=@n1
    Alpha
    OverrideWorldGravity:Vector="0,0,-980"
    ExternalForce
    SimSpaceSettings:Sim Space Settings Structure=(WorldAlpha=0.1,VelocityScaleZ=0,DampingAlpha=1,MaxLinearVelocity=500,MaxAngularVelocity=10,MaxLinearAcceleration=700,MaxAngularAcceleration=1000,ExternalLinearDragV=(X=0.5,Y=0.5,Z=0),ExternalLinearVelocity=(X=0,Y=0,Z=0),ExternalAngularVelocity=(X=0,Y=0,Z=0))
   ) -> (Pose) {
    bClampLinearTranslationLimitToRefPose=true
    bOverrideWorldGravity=true
    overridePhysicsAsset=&/F_MED_TweakSmirk_FaceAcc_WearingPackBool_Physics1
    simulationSpace=BaseBoneSpace
    alphaBoolBlend {blendOption=Linear}
    baseBoneRef {boneName=neck_01}
    componentAppliedLinearAccClamp=<0 0 0>
    overrideWorldGravity {z=-980}
    simSpaceSettings {
     maxAngularAcceleration=1000
     maxAngularVelocity=10
     maxLinearAcceleration=700
     maxLinearVelocity=500
     velocityScaleZ=0
     worldAlpha=0.1
     externalLinearDragV=<0.5 0.5>
    }
   }
   n3 = "Transform (Modify) Bone - Bone: dyn_Hair_Back_05"(ComponentPose=@n2, Alpha=1, Translation="1,0,0", Rotation="0,-7,0", Scale="1,1,1") {
    rotationMode=BMM_Additive
    rotationSpace=BCS_BoneSpace
    scaleMode=BMM_Additive
    scaleSpace=BCS_BoneSpace
    translationMode=BMM_Additive
    translationSpace=BCS_BoneSpace
    alphaBoolBlend {blendInTime=0.2, blendOutTime=0.2}
    boneToModify {boneName=dyn_Hair_Back_05}
    rotation {yaw=-7}
    translation {x=1}
   }
   n4 = RigidBody(
    ComponentPose=Local To Component(LocalPose=Use cached pose 'hair pose'())
    OverrideWorldGravity="0,0,-980"
    SimSpaceSettings=(WorldAlpha=0.1,VelocityScaleZ=0,DampingAlpha=1,MaxLinearVelocity=500,MaxAngularVelocity=10,MaxLinearAcceleration=700,MaxAngularAcceleration=1000,ExternalLinearDragV=(X=0.5,Y=0.5,Z=0),ExternalLinearVelocity=(X=0,Y=0,Z=0),ExternalAngularVelocity=(X=0,Y=0,Z=0))
   ) {
    bClampLinearTranslationLimitToRefPose=true
    bOverrideWorldGravity=true
    overridePhysicsAsset=&/F_MED_TweakSmirk_FaceAcc_Physics
    simulationSpace=BaseBoneSpace
    alphaBoolBlend {blendOption=Linear}
    baseBoneRef {boneName=neck_01}
    componentAppliedLinearAccClamp=<0 0 0>
    overrideWorldGravity {z=-980}
    simSpaceSettings {
     maxAngularAcceleration=1000
     maxAngularVelocity=10
     maxLinearAcceleration=700
     maxLinearVelocity=500
     velocityScaleZ=0
     worldAlpha=0.1
     externalLinearDragV=<0.5 0.5>
    }
   }
   Output Pose(Result=Blend Poses by bool(
     BlendPose_0=Component To Local(ComponentPose=@n3)
     BlendPose_1=Component To Local(ComponentPose=@n4)
     bActiveValue=Get isWearingPackBool() -> (isWearingPackBool:bool)
    ) {blendType=Linear}) {layerGroup=None}
   n5 = Control Rig(Source=Copy Pose From Mesh() {bUseAttachedParent=true}) {
    alphaBoolBlend {blendOption=Linear}
    controlRigAssetReference {blueprintRigClass=&/F_MED_TweakSmirk_FaceAcc_CtrlRig.F_MED_TweakSmirk_FaceAcc_CtrlRig_C}
   }
   n6 = Anim Dynamics - Bone: dyn_Hair_Fn_a_01!(ComponentPose=Local To Component(LocalPose=@n5), Alpha=1, GravityOverride:Vector="0,0,500", ExternalForce="0,0,0") -> (Pose) {
    angularDampingOverride=0.9999
    angularSpringConstant=800
    bAngularSpring=true
    bOverrideAngularDamping=true
    bUseGravityOverride=true
    gravityScale=0.3
    simulationSpace=RootRelative
    alphaBoolBlend {blendOption=Linear}
    boundBone {boneName=dyn_Hair_Fn_a_01}
    componentAppliedLinearAccClamp=<2000 2000 1600>
    componentLinearAccScale=<0.4 0.4 0.4>
    componentLinearVelScale=<0.45 0.45 0.55>
    gravityOverride {z=500}
    physicsBodyDefinitions [{
      sphereCollisionRadius=2
      boundBone {boneName=dyn_Hair_Fn_a_01}
      boxExtents=<10 15 18>
      constraintSetup {twistAxis=AxisY, angularLimitsMax=<5 0 15>, angularLimitsMin=<5 0 -15>, angularTarget=<0 0 0>}
      localJointOffset=<0 -2 0>
     }]
    relativeSpaceBone {boneName=root}
   }
   n7 = Anim Dynamics - Bone: dyn_Hair_Fn_a_02!(ComponentPose=@n6, Alpha=1, GravityOverride:Vector="0,0,500", ExternalForce="0,0,0") -> (Pose) {
    angularDampingOverride=0.9999
    angularSpringConstant=800
    bAngularSpring=true
    bOverrideAngularDamping=true
    bUseGravityOverride=true
    gravityScale=0.3
    simulationSpace=RootRelative
    alphaBoolBlend {blendOption=Linear}
    boundBone {boneName=dyn_Hair_Fn_a_02}
    componentAppliedLinearAccClamp=<2000 2000 1600>
    componentLinearAccScale=<0.4 0.4 0.4>
    componentLinearVelScale=<0.45 0.45 0.55>
    gravityOverride {z=500}
    physicsBodyDefinitions [{
      sphereCollisionRadius=2
      boundBone {boneName=dyn_Hair_Fn_a_02}
      boxExtents=<8 10 15>
      constraintSetup {twistAxis=AxisY, angularLimitsMax=<5 0 15>, angularLimitsMin=<5 0 -15>, angularTarget=<0 0 0>}
      localJointOffset=<0 -2 0>
     }]
    relativeSpaceBone {boneName=root}
   }
   n8 = Anim Dynamics - Bone: dyn_Hair_Fn_b_01!(ComponentPose=@n7, Alpha=1, GravityOverride:Vector="0,0,500", ExternalForce="0,0,0") -> (Pose) {
    angularDampingOverride=0.9999
    angularSpringConstant=800
    bAngularSpring=true
    bOverrideAngularDamping=true
    bUseGravityOverride=true
    gravityScale=0.3
    simulationSpace=RootRelative
    alphaBoolBlend {blendOption=Linear}
    boundBone {boneName=dyn_Hair_Fn_b_01}
    componentAppliedLinearAccClamp=<2000 2000 1600>
    componentLinearAccScale=<0.4 0.4 0.4>
    componentLinearVelScale=<0.45 0.45 0.55>
    gravityOverride {z=500}
    physicsBodyDefinitions [{
      sphereCollisionRadius=2
      boundBone {boneName=dyn_Hair_Fn_b_01}
      boxExtents=<10 15 18>
      constraintSetup {twistAxis=AxisY, angularLimitsMax=<15 0 5>, angularLimitsMin=<-2 0 -10>, angularTarget=<0 0 0>}
      localJointOffset=<0 -2 0>
     }]
    relativeSpaceBone {boneName=root}
   }
   n9 = Anim Dynamics - Bone: dyn_Hair_Fn_b_02!(ComponentPose=@n8, Alpha=1, GravityOverride:Vector="0,0,500", ExternalForce="0,0,0") -> (Pose) {
    angularDampingOverride=0.9999
    angularSpringConstant=800
    bAngularSpring=true
    bOverrideAngularDamping=true
    bUseGravityOverride=true
    gravityScale=0.3
    simulationSpace=RootRelative
    alphaBoolBlend {blendOption=Linear}
    boundBone {boneName=dyn_Hair_Fn_b_02}
    componentAppliedLinearAccClamp=<2000 2000 1600>
    componentLinearAccScale=<0.4 0.4 0.4>
    componentLinearVelScale=<0.45 0.45 0.55>
    gravityOverride {z=500}
    physicsBodyDefinitions [{
      sphereCollisionRadius=2
      boundBone {boneName=dyn_Hair_Fn_b_02}
      boxExtents=<8 10 15>
      constraintSetup {twistAxis=AxisY, angularLimitsMax=<15 0 5>, angularLimitsMin=<-2 0 -10>, angularTarget=<0 0 0>}
      localJointOffset=<0 -2 0>
     }]
    relativeSpaceBone {boneName=root}
   }
   n10 = Anim Dynamics - Bone: dyn_Hair_Fn_c_01!(ComponentPose=@n9, Alpha=1, GravityOverride:Vector="0,0,500", ExternalForce="0,0,0") -> (Pose) {
    angularDampingOverride=0.9999
    angularSpringConstant=800
    bAngularSpring=true
    bOverrideAngularDamping=true
    bUseGravityOverride=true
    gravityScale=0.3
    simulationSpace=RootRelative
    alphaBoolBlend {blendOption=Linear}
    boundBone {boneName=dyn_Hair_Fn_c_01}
    componentAppliedLinearAccClamp=<2000 2000 1600>
    componentLinearAccScale=<0.4 0.4 0.4>
    componentLinearVelScale=<0.45 0.45 0.55>
    gravityOverride {z=500}
    physicsBodyDefinitions [{
      sphereCollisionRadius=2
      boundBone {boneName=dyn_Hair_Fn_c_01}
      boxExtents=<10 15 18>
      constraintSetup {twistAxis=AxisY, angularLimitsMax=<15 0 7>, angularLimitsMin=<-2 0 -2>, angularTarget=<0 0 0>}
      localJointOffset=<0 -2 0>
     }]
    relativeSpaceBone {boneName=root}
   }
   n11 = Anim Dynamics - Bone: dyn_Hair_Fn_c_02!(ComponentPose=@n10, Alpha=1, GravityOverride:Vector="0,0,500", ExternalForce="0,0,0") -> (Pose) {
    angularDampingOverride=0.9999
    angularSpringConstant=800
    bAngularSpring=true
    bOverrideAngularDamping=true
    bUseGravityOverride=true
    gravityScale=0.3
    simulationSpace=RootRelative
    alphaBoolBlend {blendOption=Linear}
    boundBone {boneName=dyn_Hair_Fn_c_02}
    componentAppliedLinearAccClamp=<2000 2000 1600>
    componentLinearAccScale=<0.4 0.4 0.4>
    componentLinearVelScale=<0.45 0.45 0.55>
    gravityOverride {z=500}
    physicsBodyDefinitions [{
      sphereCollisionRadius=2
      boundBone {boneName=dyn_Hair_Fn_c_02}
      boxExtents=<8 10 15>
      constraintSetup {twistAxis=AxisY, angularLimitsMax=<15 0 7>, angularLimitsMin=<-2 0 -2>, angularTarget=<0 0 0>}
      localJointOffset=<0 -2 0>
     }]
    relativeSpaceBone {boneName=root}
   }
   n12 = Anim Dynamics - Bone: dyn_Hair_Fn_d_01!(ComponentPose=@n11, Alpha=1, GravityOverride:Vector="0,0,500", ExternalForce="0,0,0") -> (Pose) {
    angularDampingOverride=0.9999
    angularSpringConstant=800
    bAngularSpring=true
    bOverrideAngularDamping=true
    bUseGravityOverride=true
    gravityScale=0.3
    simulationSpace=RootRelative
    alphaBoolBlend {blendOption=Linear}
    boundBone {boneName=dyn_Hair_Fn_d_01}
    componentAppliedLinearAccClamp=<2000 2000 1600>
    componentLinearAccScale=<0.4 0.4 0.4>
    componentLinearVelScale=<0.45 0.45 0.55>
    gravityOverride {z=500}
    physicsBodyDefinitions [{
      sphereCollisionRadius=2
      boundBone {boneName=dyn_Hair_Fn_d_01}
      boxExtents=<10 15 18>
      constraintSetup {twistAxis=AxisY, angularLimitsMax=<15 0 3>, angularLimitsMin=<-2 0 -5>, angularTarget=<0 0 0>}
      localJointOffset=<0 -2 0>
     }]
    relativeSpaceBone {boneName=root}
   }
   n13 = Anim Dynamics - Bone: dyn_Hair_Fn_d_02!(ComponentPose=@n12, Alpha=1, GravityOverride:Vector="0,0,500", ExternalForce="0,0,0") -> (Pose) {
    angularDampingOverride=0.9999
    angularSpringConstant=800
    bAngularSpring=true
    bOverrideAngularDamping=true
    bUseGravityOverride=true
    gravityScale=0.3
    simulationSpace=RootRelative
    alphaBoolBlend {blendOption=Linear}
    boundBone {boneName=dyn_Hair_Fn_d_02}
    componentAppliedLinearAccClamp=<2000 2000 1600>
    componentLinearAccScale=<0.4 0.4 0.4>
    componentLinearVelScale=<0.45 0.45 0.55>
    gravityOverride {z=500}
    physicsBodyDefinitions [{
      sphereCollisionRadius=2
      boundBone {boneName=dyn_Hair_Fn_d_02}
      boxExtents=<8 10 15>
      constraintSetup {twistAxis=AxisY, angularLimitsMax=<15 0 3>, angularLimitsMin=<-2 0 -5>, angularTarget=<0 0 0>}
      localJointOffset=<0 -2 0>
     }]
    relativeSpaceBone {boneName=root}
   }
   n14 = Anim Dynamics - Bone: dyn_Hair_Fn_e_01!(ComponentPose=@n13, Alpha=1, GravityOverride:Vector="0,0,500", ExternalForce="0,0,0") -> (Pose) {
    angularDampingOverride=0.9999
    angularSpringConstant=800
    bAngularSpring=true
    bOverrideAngularDamping=true
    bUseGravityOverride=true
    gravityScale=0.3
    simulationSpace=RootRelative
    alphaBoolBlend {blendOption=Linear}
    boundBone {boneName=dyn_Hair_Fn_e_01}
    componentAppliedLinearAccClamp=<2000 2000 1600>
    componentLinearAccScale=<0.4 0.4 0.4>
    componentLinearVelScale=<0.45 0.45 0.55>
    gravityOverride {z=500}
    physicsBodyDefinitions [{
      sphereCollisionRadius=2
      boundBone {boneName=dyn_Hair_Fn_e_01}
      boxExtents=<10 15 18>
      constraintSetup {twistAxis=AxisY, angularLimitsMax=<15 0 3>, angularLimitsMin=<-2 0 -5>, angularTarget=<0 0 0>}
      localJointOffset=<0 -2 0>
     }]
    relativeSpaceBone {boneName=root}
   }
   n15 = Anim Dynamics - Bone: dyn_Hair_Fn_e_02!(ComponentPose=@n14, Alpha=1, GravityOverride:Vector="0,0,500", ExternalForce="0,0,0") -> (Pose) {
    angularDampingOverride=0.9999
    angularSpringConstant=800
    bAngularSpring=true
    bOverrideAngularDamping=true
    bUseGravityOverride=true
    gravityScale=0.3
    simulationSpace=RootRelative
    alphaBoolBlend {blendOption=Linear}
    boundBone {boneName=dyn_Hair_Fn_e_02}
    componentAppliedLinearAccClamp=<2000 2000 1600>
    componentLinearAccScale=<0.4 0.4 0.4>
    componentLinearVelScale=<0.45 0.45 0.55>
    gravityOverride {z=500}
    physicsBodyDefinitions [{
      sphereCollisionRadius=2
      boundBone {boneName=dyn_Hair_Fn_e_02}
      boxExtents=<8 10 15>
      constraintSetup {twistAxis=AxisY, angularLimitsMax=<15 0 3>, angularLimitsMin=<-2 0 -5>, angularTarget=<0 0 0>}
      localJointOffset=<0 -2 0>
     }]
    relativeSpaceBone {boneName=root}
   }
   n16 = Anim Dynamics - Bone: dyn_Hair_Fn_f_01!(ComponentPose=@n15, Alpha=1, GravityOverride:Vector="0,0,500", ExternalForce="0,0,0") -> (Pose) {
    angularDampingOverride=0.9999
    angularSpringConstant=800
    bAngularSpring=true
    bOverrideAngularDamping=true
    bUseGravityOverride=true
    gravityScale=0.3
    simulationSpace=RootRelative
    alphaBoolBlend {blendOption=Linear}
    boundBone {boneName=dyn_Hair_Fn_f_01}
    componentAppliedLinearAccClamp=<2000 2000 1600>
    componentLinearAccScale=<0.4 0.4 0.4>
    componentLinearVelScale=<0.45 0.45 0.55>
    gravityOverride {z=500}
    physicsBodyDefinitions [{
      sphereCollisionRadius=2
      boundBone {boneName=dyn_Hair_Fn_f_01}
      boxExtents=<10 15 18>
      constraintSetup {twistAxis=AxisY, angularLimitsMax=<15 0 10>, angularLimitsMin=<-2 0 -3>, angularTarget=<0 0 0>}
      localJointOffset=<0 -2 0>
     }]
    relativeSpaceBone {boneName=root}
   }
   n17 = Anim Dynamics - Bone: dyn_Hair_Fn_f_02!(ComponentPose=@n16, Alpha=1, GravityOverride:Vector="0,0,500", ExternalForce="0,0,0") -> (Pose) {
    angularDampingOverride=0.9999
    angularSpringConstant=800
    bAngularSpring=true
    bOverrideAngularDamping=true
    bUseGravityOverride=true
    gravityScale=0.3
    simulationSpace=RootRelative
    alphaBoolBlend {blendOption=Linear}
    boundBone {boneName=dyn_Hair_Fn_f_02}
    componentAppliedLinearAccClamp=<2000 2000 1600>
    componentLinearAccScale=<0.4 0.4 0.4>
    componentLinearVelScale=<0.45 0.45 0.55>
    gravityOverride {z=500}
    physicsBodyDefinitions [{
      sphereCollisionRadius=2
      boundBone {boneName=dyn_Hair_Fn_f_02}
      boxExtents=<8 10 15>
      constraintSetup {twistAxis=AxisY, angularLimitsMax=<15 0 10>, angularLimitsMin=<-2 0 -3>, angularTarget=<0 0 0>}
      localJointOffset=<0 -2 0>
     }]
    relativeSpaceBone {boneName=root}
   }
   Save cached pose 'hair pose'(Pose=Component To Local(ComponentPose=@n17)) {cacheName=hair pose}
  ]}
 EventGraph {nodes [
   Event BlueprintUpdateAnimation()
   > Sequence!(execute) -> (then_0, then_1, then_2:Exec, then_3:Exec, then_4:Exec)
   > n1 = "Gravity Override State Function (Parameter Asset) (Nativized)"(
    CharacterPartAnimInstance=Self-Reference()
    GameTimeInSeconds=GetGameTimeInSeconds()
    Parameters=/BRCosmetics/Characters/Player/Female/Medium/Bodies/F_MED_LoanFloat/Meshes/Parts/F_MED_LoanFloat_FaceAcc_Parameters
   )
   > Set Gravity Override Cape(Gravity Override Cape:Vector=@n1) -> (Output_Get:Vector)
   > Set RBAN Settings Out_0(RBAN Settings Out_0:Fort Rigid Body Anim Node Parameters Structure=@n1.OutRigidBodyAnimNodeParameters) -> (Output_Get:Fort Rigid Body Anim Node Parameters Structure)
   Event OnCharacterCustomizationLoaded()
   n2 = CheckForBackbling()
   > Set isWearingPackBool(isWearingPackBool:bool=OR Boolean!(A=@n2, B=@n2.bIsBackblingACape, C:bool=@n2.bIsBackblingLong) -> (ReturnValue)) -> (Output_Get:bool)
   > Set is_WearingCape(is_WearingCape:bool=@n2.bIsBackblingLong) -> (Output_Get:bool)
   AND Boolean(A=@n2, B=NOT Boolean(A=@n2.bIsBackblingLong))
  ]}
]
variables [
 Gravity Override Cape {type=Vector}
 is_WearingCape {type=boolean}
 isWearingPackBool {type=boolean}
 RBAN Settings Out_0 {
  type=FortRigidBodyAnimNodeParameters
  default {simSpaceSettings {
    dampingAlpha=0.25
    maxAngularAcceleration=80
    maxAngularVelocity=1
    maxLinearAcceleration=1000
    maxLinearVelocity=250
    worldAlpha=1
    externalLinearDragV=<0.1 1 1>
   }}
 }
]
